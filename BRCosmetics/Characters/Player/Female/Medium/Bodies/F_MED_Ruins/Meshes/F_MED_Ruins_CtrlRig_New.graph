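# Control Rig for the F_MED_Ruins female body cosmetic that procedurally corrects jacket, belt, flask, bag and arm-chain bones as the body moves.
# A five-branch sequence first rotation-constrains jacket masters to thighs and pelvis without translation.
# Spherical pose readers on both thighs, right upper arm and upper spine then weight transform offsets on those bones.
# The arm-chain aim and constraint, plus three stray offset nodes, are unconnected to execution.
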ControlRigBlueprint /BRCosmetics/Characters/Player/Female/Medium/Bodies/F_MED_Ruins/Meshes/F_MED_Ruins_CtrlRig_New
graphs [
 ContainedGraph {nodes [
   n1 = RigVMFunctionEntryNode()
   SequenceAggregate = &RigVMFunction_Sequence(ExecuteContext=@n1.ExecuteContext)
   SequenceAggregate_1 = &RigVMFunction_Sequence(ExecuteContext=@SequenceAggregate.B)
   SequenceAggregate_2 = &RigVMFunction_Sequence(ExecuteContext=@SequenceAggregate_1.B)
   SequenceAggregate_3 = &RigVMFunction_Sequence(ExecuteContext=@SequenceAggregate_2.B)
   RigVMFunctionReturnNode(A=@SequenceAggregate.A, B=@SequenceAggregate_1.A, C=@SequenceAggregate_2.A, D=@SequenceAggregate_3.A, E=@SequenceAggregate_3.B)
  ]}
 RigVMFunctionLibrary {links [], nodes []}
 RigVMModel {nodes [
   /Script/ControlRig.RigUnit_AimItem(
    DebugSettings=(Scale=10,WorldOffset=(Rotation=(X=0,Y=0,Z=0,W=1),Translation=(X=0,Y=0,Z=0),Scale3D=(X=1,Y=1,Z=1)))
    ExecutePin=()
    Item=(Type=Bone,Name="dyn_arm_chain_tgt")
    Primary=(Weight=1,Axis=(X=1,Y=0,Z=0),Target=(X=1,Y=0,Z=0),Kind=Location,Space=(Type=Bone,Name="dyn_arm_chain_twist_lookat_01"))
    Secondary=(Weight=1,Axis=(X=0,Y=0,Z=1),Target=(X=1,Y=0,Z=0),Kind=Location,Space=(Type=Bone))
    Weight=1
   ) {is=struct RigVMUnitNode}
   > &RigUnit_ParentConstraint(
    AdvancedSettings=(RotationOrderForFilter=XZY)
    Child=(Type=Bone,Name="dyn_arm_chain_con")
    Parents=((Item=(Type=Bone,Name="dyn_arm_chain_tgt"),Weight=1))
    Weight=1
   )
   RigUnit_BeginExecution = &RigUnit_BeginExecution()
   SequenceAggregate = RigVMAggregateNode(ExecuteContext=@RigUnit_BeginExecution.ExecutePin)
   &RigUnit_ParentConstraint(
    AdvancedSettings=(RotationOrderForFilter=XZY)
    Child=(Type=Bone,Name="dyn_jacketFront_mstr_l")
    Filter=(TranslationFilter=(bX=false,bY=false,bZ=false),RotationFilter=(bX=True,bY=True,bZ=True),ScaleFilter=(bX=True,bY=True,bZ=True))
    Parents=((Item=(Type=Bone,Name="thigh_l"),Weight=1))
    Weight=1
    ExecutePin=@SequenceAggregate.A
   )
   > &RigUnit_ParentConstraint(
    AdvancedSettings=(RotationOrderForFilter=XZY)
    Child=(Type=Bone,Name="dyn_jacketSide_mstr_l")
    Filter=(TranslationFilter=(bX=false,bY=false,bZ=false),RotationFilter=(bX=True,bY=True,bZ=True),ScaleFilter=(bX=True,bY=True,bZ=True))
    Parents=((Item=(Type=Bone,Name="thigh_l"),Weight=0.6),(Item=(Type=Bone,Name="pelvis"),Weight=0.4))
    Weight=1
   )
   > &RigUnit_ParentConstraint(
    AdvancedSettings=(RotationOrderForFilter=XZY)
    Child=(Type=Bone,Name="dyn_jacketFront_mstr")
    Filter=(TranslationFilter=(bX=false,bY=false,bZ=false),RotationFilter=(bX=True,bY=True,bZ=True),ScaleFilter=(bX=True,bY=True,bZ=True))
    Parents=((Item=(Type=Bone,Name="thigh_r"),Weight=1))
    Weight=1
   )
   > &RigUnit_ParentConstraint(
    AdvancedSettings=(RotationOrderForFilter=XZY)
    Child=(Type=Bone,Name="dyn_jacketFront_mstr_r")
    Filter=(TranslationFilter=(bX=false,bY=false,bZ=false),RotationFilter=(bX=True,bY=True,bZ=True),ScaleFilter=(bX=True,bY=True,bZ=True))
    Parents=((Item=(Type=Bone,Name="thigh_r"),Weight=1))
    Weight=1
   )
   > &RigUnit_ParentConstraint(
    AdvancedSettings=(RotationOrderForFilter=XZY)
    Child=(Type=Bone,Name="dyn_jacketSide_mstr_r")
    Filter=(TranslationFilter=(bX=false,bY=false,bZ=false),RotationFilter=(bX=True,bY=True,bZ=True),ScaleFilter=(bX=True,bY=True,bZ=True))
    Parents=((Item=(Type=Bone,Name="thigh_r"),Weight=0.6),(Item=(Type=Bone,Name="pelvis"),Weight=0.4))
    Weight=1
   )
   > &RigUnit_ParentConstraint(
    AdvancedSettings=(RotationOrderForFilter=XZY)
    Child=(Type=Bone,Name="dyn_jacketBack_mstr_l")
    Filter=(TranslationFilter=(bX=false,bY=false,bZ=false),RotationFilter=(bX=True,bY=True,bZ=True),ScaleFilter=(bX=True,bY=True,bZ=True))
    Parents=((Item=(Type=Bone,Name="dyn_jacketSide_mstr_l"),Weight=0.2),(Item=(Type=Bone,Name="pelvis"),Weight=0.8))
    Weight=1
   )
   > &RigUnit_ParentConstraint(
    AdvancedSettings=(RotationOrderForFilter=XZY)
    Child=(Type=Bone,Name="dyn_jacketBack_mstr_r")
    Filter=(TranslationFilter=(bX=false,bY=false,bZ=false),RotationFilter=(bX=True,bY=True,bZ=True),ScaleFilter=(bX=True,bY=True,bZ=True))
    Parents=((Item=(Type=Bone,Name="dyn_jacketSide_mstr_r"),Weight=0.2),(Item=(Type=Bone,Name="pelvis"),Weight=0.8))
    Weight=1
   )
   n1 = &RigUnit_SphericalPoseReader(
    ActiveRegionScaleFactors=(PositiveWidth=3.5,NegativeWidth=0.5,PositiveHeight=3,NegativeHeight=3)
    ActiveRegionSize=0.1
    Debug=(bDrawDebug=false,DebugScale=25,DebugSegments=20,DebugThickness=0.25)
    DriverAxis=(X=-1,Y=0,Z=0)
    DriverItem=(Type=Bone,Name="thigh_l")
    FalloffRegionScaleFactors=(PositiveWidth=1,NegativeWidth=1,PositiveHeight=1,NegativeHeight=1)
    FalloffSize=0.45
    OptionalParentItem=(Type=Bone)
    RotationOffset=(X=90,Y=0,Z=0)
    ExecutePin=@SequenceAggregate.B
   )
   > &RigUnit_ModifyTransforms(
    ItemToModify=((Item=(Type=Bone,Name="dyn_jacketFront_mstr_l"),Transform=(Rotation=(X=0,Y=0,Z=0,W=1),Translation=(X=-3,Y=-5,Z=6),Scale3D=(X=1,Y=1,Z=1))))
    ItemToModify.0.Transform.Rotation=&RigVMFunction_MathQuaternionFromEuler(Euler=(X=30,Y=0,Z=-15)).Result
    Weight=@n1.OutputParam
   )
   > &RigUnit_ModifyTransforms(
    ItemToModify=((Item=(Type=Bone,Name="dyn_jacketSide_mstr_l"),Transform=(Rotation=(X=0,Y=0,Z=0,W=1),Translation=(X=-1,Y=0,Z=5),Scale3D=(X=1,Y=1,Z=1))))
    ItemToModify.0.Transform.Rotation=&RigVMFunction_MathQuaternionFromEuler(Euler=(X=0,Y=-10,Z=-15)).Result
    Weight=@n1.OutputParam
   )
   n2 = &RigVMFunction_MathQuaternionFromEuler(Euler=(X=0,Y=20,Z=-20))
   > &RigUnit_ModifyTransforms(
    ItemToModify=((Item=(Type=Bone,Name="dyn_jacketFront_02_l"),Transform=(Rotation=(X=0,Y=0,Z=0,W=1),Translation=(X=0,Y=0,Z=0),Scale3D=(X=1,Y=1,Z=1))))
    ItemToModify.0.Transform.Rotation=@n2.Result
    Weight=@n1.OutputParam
   )
   > &RigUnit_ModifyTransforms(
    ItemToModify=((Item=(Type=Bone,Name="dyn_jacketFront_03_l"),Transform=(Rotation=(X=0,Y=0,Z=0,W=1),Translation=(X=0,Y=0,Z=0),Scale3D=(X=1,Y=1,Z=1))))
    ItemToModify.0.Transform.Rotation=@n2.Result
    Weight=@n1.OutputParam
   )
   n3 = &RigVMFunction_MathQuaternionFromEuler(Euler=(X=0,Y=0,Z=-20))
   > &RigUnit_ModifyTransforms(
    ItemToModify=((Item=(Type=Bone,Name="dyn_jacketSide_02_l"),Transform=(Rotation=(X=0,Y=0,Z=0,W=1),Translation=(X=0,Y=0,Z=0),Scale3D=(X=1,Y=1,Z=1))))
    ItemToModify.0.Transform.Rotation=@n3.Result
    Weight=@n1.OutputParam
   )
   > &RigUnit_ModifyTransforms(
    ItemToModify=((Item=(Type=Bone,Name="dyn_jacketSide_03_l"),Transform=(Rotation=(X=0,Y=0,Z=0,W=1),Translation=(X=0,Y=0,Z=0),Scale3D=(X=1,Y=1,Z=1))))
    ItemToModify.0.Transform.Rotation=@n3.Result
    Weight=@n1.OutputParam
   )
   > &RigUnit_ModifyTransforms(
    ItemToModify=((Item=(Type=Bone,Name="dyn_Flask_MSTR"),Transform=(Rotation=(X=0,Y=0,Z=0,W=1),Translation=(X=-2,Y=-3,Z=-2),Scale3D=(X=1,Y=1,Z=1))))
    ItemToModify.0.Transform.Rotation=&RigVMFunction_MathQuaternionFromEuler(Euler=(X=0,Y=0,Z=-25)).Result
    Weight=@n1.OutputParam
   )
   > n4 = &RigUnit_SphericalPoseReader(
    ActiveRegionScaleFactors=(PositiveWidth=3.5,NegativeWidth=0.5,PositiveHeight=3,NegativeHeight=3)
    ActiveRegionSize=0.1
    Debug=(bDrawDebug=false,DebugScale=25,DebugSegments=20,DebugThickness=0.25)
    DriverAxis=(X=-1,Y=0,Z=0)
    DriverItem=(Type=Bone,Name="thigh_l")
    FalloffRegionScaleFactors=(PositiveWidth=1,NegativeWidth=1,PositiveHeight=1,NegativeHeight=1)
    FalloffSize=0.45
    FlipHeightScaling=false
    FlipWidthScaling=true
    OptionalParentItem=(Type=Bone)
    RotationOffset=(X=-90,Y=0,Z=0)
   )
   > &RigUnit_ModifyTransforms(
    ItemToModify=((Item=(Type=Bone,Name="dyn_jacketBack_mstr_l"),Transform=(Rotation=(X=0,Y=0,Z=0,W=1),Translation=(X=0,Y=0,Z=0),Scale3D=(X=1,Y=1,Z=1))))
    ItemToModify.0.Transform.Rotation=&RigVMFunction_MathQuaternionFromEuler(Euler=(X=0,Y=-30,Z=0)).Result
    Weight=@n4.OutputParam
   )
   n5 = &RigUnit_SphericalPoseReader(
    ActiveRegionScaleFactors=(PositiveWidth=3.5,NegativeWidth=0.5,PositiveHeight=3,NegativeHeight=3)
    ActiveRegionSize=0.1
    Debug=(bDrawDebug=false,DebugScale=25,DebugSegments=20,DebugThickness=0.25)
    DriverAxis=(X=1,Y=0,Z=0)
    DriverItem=(Type=Bone,Name="thigh_r")
    FalloffRegionScaleFactors=(PositiveWidth=1,NegativeWidth=1,PositiveHeight=1,NegativeHeight=1)
    FalloffSize=0.45
    FlipHeightScaling=false
    FlipWidthScaling=true
    OptionalParentItem=(Type=Bone)
    RotationOffset=(X=-90,Y=0,Z=0)
    ExecutePin=@SequenceAggregate.C
   )
   > &RigUnit_ModifyTransforms(
    ItemToModify=((Item=(Type=Bone,Name="dyn_jacketFront_mstr"),Transform=(Rotation=(X=0,Y=0,Z=0,W=1),Translation=(X=-3,Y=5,Z=6),Scale3D=(X=1,Y=1,Z=1))))
    ItemToModify.0.Transform.Rotation=&RigVMFunction_MathQuaternionFromEuler(Euler=(X=0,Y=-3,Z=5)).Result
    Weight=@n5.OutputParam
   )
   > &RigUnit_ModifyTransforms(
    ItemToModify=((Item=(Type=Bone,Name="dyn_jacketFront_mstr_r"),Transform=(Rotation=(X=0,Y=0,Z=0,W=1),Translation=(X=-3,Y=5,Z=6),Scale3D=(X=1,Y=1,Z=1))))
    ItemToModify.0.Transform.Rotation=&RigVMFunction_MathQuaternionFromEuler(Euler=(X=-5,Y=-5,Z=10)).Result
    Weight=@n5.OutputParam
   )
   > &RigUnit_ModifyTransforms(
    ItemToModify=((Item=(Type=Bone,Name="dyn_jacketFront_02"),Transform=(Rotation=(X=0,Y=0,Z=0,W=1),Translation=(X=0,Y=0,Z=0),Scale3D=(X=1,Y=1,Z=1))))
    ItemToModify.0.Transform.Rotation=&RigVMFunction_MathQuaternionFromEuler(Euler=(X=0,Y=0,Z=10)).Result
    Weight=@n5.OutputParam
   )
   > &RigUnit_ModifyTransforms(
    ItemToModify=((Item=(Type=Bone,Name="dyn_jacketSide_mstr_r"),Transform=(Rotation=(X=0,Y=0,Z=0,W=1),Translation=(X=-1,Y=0,Z=5),Scale3D=(X=1,Y=1,Z=1))))
    ItemToModify.0.Transform.Rotation=&RigVMFunction_MathQuaternionFromEuler(Euler=(X=0,Y=-10,Z=-10)).Result
    Weight=@n5.OutputParam
   )
   > &RigUnit_ModifyTransforms(
    ItemToModify=((Item=(Type=Bone,Name="dyn_bag_MSTR"),Transform=(Rotation=(X=0,Y=0,Z=0,W=1),Translation=(X=-5,Y=5,Z=0),Scale3D=(X=1,Y=1,Z=1))))
    ItemToModify.0.Transform.Rotation=&RigVMFunction_MathQuaternionFromEuler(Euler=(X=-35,Y=-25,Z=15)).Result
    Weight=@n5.OutputParam
   )
   > &RigUnit_ModifyTransforms(
    ItemToModify=((Item=(Type=Bone,Name="dyn_belt_rope_a_MSTR"),Transform=(Rotation=(X=0,Y=0,Z=0,W=1),Translation=(X=-6,Y=0,Z=0),Scale3D=(X=1,Y=1,Z=1))))
    ItemToModify.0.Transform.Rotation=&RigVMFunction_MathQuaternionFromEuler(Euler=(X=0,Y=25,Z=25)).Result
    Weight=@n5.OutputParam
   )
   > &RigUnit_ModifyTransforms(
    ItemToModify=((Item=(Type=Bone,Name="dyn_belt"),Transform=(Rotation=(X=0,Y=0,Z=0,W=1),Translation=(X=-8,Y=0,Z=0),Scale3D=(X=1,Y=1,Z=1))))
    Weight=@n5.OutputParam
   )
   > n6 = &RigUnit_SphericalPoseReader(
    ActiveRegionScaleFactors=(PositiveWidth=3.5,NegativeWidth=0.5,PositiveHeight=3,NegativeHeight=3)
    ActiveRegionSize=0.1
    Debug=(bDrawDebug=false,DebugScale=25,DebugSegments=20,DebugThickness=0.25)
    DriverAxis=(X=1,Y=0,Z=0)
    DriverItem=(Type=Bone,Name="thigh_r")
    FalloffRegionScaleFactors=(PositiveWidth=1,NegativeWidth=1,PositiveHeight=1,NegativeHeight=1)
    FalloffSize=0.45
    FlipHeightScaling=false
    FlipWidthScaling=false
    OptionalParentItem=(Type=Bone)
    RotationOffset=(X=90,Y=0,Z=0)
   )
   > &RigUnit_ModifyTransforms(
    ItemToModify=((Item=(Type=Bone,Name="dyn_jacketBack_mstr_r"),Transform=(Rotation=(X=0,Y=0,Z=0,W=1),Translation=(X=0,Y=0,Z=0),Scale3D=(X=1,Y=1,Z=1))))
    ItemToModify.0.Transform.Rotation=&RigVMFunction_MathQuaternionFromEuler(Euler=(X=0,Y=-30,Z=0)).Result
    Weight=@n6.OutputParam
   )
   n7 = &RigUnit_SphericalPoseReader(
    ActiveRegionScaleFactors=(PositiveWidth=1,NegativeWidth=4,PositiveHeight=0.5,NegativeHeight=4.5)
    ActiveRegionSize=0.08
    Debug=(bDrawDebug=false,bDraw2D=false,bDrawLocalAxes=true,DebugScale=25,DebugSegments=20,DebugThickness=0.25)
    DriverAxis=(X=-1,Y=0,Z=0)
    DriverItem=(Type=Bone,Name="upperarm_r")
    FalloffRegionScaleFactors=(PositiveWidth=0.5,NegativeWidth=0.5,PositiveHeight=0.5,NegativeHeight=4)
    FalloffSize=0.4
    FlipWidthScaling=true
    OptionalParentItem=(Type=Bone,Name="spine_04")
    RotationOffset=(X=-50,Y=35,Z=20)
    ExecutePin=@SequenceAggregate.D
   )
   > &RigUnit_ModifyTransforms(
    ItemToModify=((Item=(Type=Bone,Name="dyn_arm_chain_twist_under"),Transform=(Rotation=(X=0,Y=0,Z=0,W=1),Translation=(X=2,Y=6,Z=2),Scale3D=(X=1,Y=1,Z=1))))
    Weight=@n7.OutputParam
   )
   > &RigUnit_ModifyTransforms(
    ItemToModify=((Item=(Type=Bone,Name="dyn_arm_chain_twist_back"),Transform=(Rotation=(X=0,Y=0,Z=0,W=1),Translation=(X=4,Y=3,Z=0),Scale3D=(X=1,Y=1,Z=1))))
    Weight=@n7.OutputParam
   )
   > &RigUnit_ModifyTransforms(
    ItemToModify=((Item=(Type=Bone,Name="dyn_arm_chain_twist_front"),Transform=(Rotation=(X=0,Y=0,Z=0,W=1),Translation=(X=4,Y=2,Z=0),Scale3D=(X=1,Y=1,Z=1))))
    Weight=@n7.OutputParam
   )
   > n8 = &RigUnit_SphericalPoseReader(
    ActiveRegionScaleFactors=(PositiveWidth=1,NegativeWidth=1.5,PositiveHeight=1,NegativeHeight=1)
    ActiveRegionSize=0.1
    Debug=(bDrawDebug=false,bDraw2D=false,bDrawLocalAxes=true,DebugScale=25,DebugSegments=20,DebugThickness=0.25)
    DriverAxis=(X=-1,Y=0,Z=0)
    DriverItem=(Type=Bone,Name="upperarm_r")
    FalloffRegionScaleFactors=(PositiveWidth=0.3,NegativeWidth=0.5,PositiveHeight=0.6,NegativeHeight=1)
    FalloffSize=0.45
    FlipHeightScaling=false
    FlipWidthScaling=false
    OptionalParentItem=(Type=Bone,Name="spine_04")
    RotationOffset=(X=-95,Y=40,Z=25)
   )
   > &RigUnit_ModifyTransforms(
    ItemToModify=((Item=(Type=Bone,Name="dyn_arm_chain_twist_under"),Transform=(Rotation=(X=0,Y=0,Z=0,W=1),Translation=(X=6,Y=-2,Z=1),Scale3D=(X=1,Y=1,Z=1))))
    Weight=@n8.OutputParam
   )
   > &RigUnit_ModifyTransforms(
    ItemToModify=((Item=(Type=Bone,Name="dyn_arm_chain_twist_back"),Transform=(Rotation=(X=0,Y=0,Z=0,W=1),Translation=(X=4,Y=0.5,Z=0),Scale3D=(X=1,Y=1,Z=1))))
    Weight=@n8.OutputParam
   )
   > &RigUnit_ModifyTransforms(
    ItemToModify=((Item=(Type=Bone,Name="dyn_arm_chain_twist_front"),Transform=(Rotation=(X=0,Y=0,Z=0,W=1),Translation=(X=4,Y=2,Z=0),Scale3D=(X=1,Y=1,Z=1))))
    Weight=@n8.OutputParam
   )
   > n9 = &RigUnit_SphericalPoseReader(
    ActiveRegionScaleFactors=(PositiveWidth=0.6,NegativeWidth=0.6,PositiveHeight=0.5,NegativeHeight=1)
    ActiveRegionSize=0.15
    Debug=(bDrawDebug=false,bDraw2D=false,bDrawLocalAxes=true,DebugScale=25,DebugSegments=20,DebugThickness=0.25)
    DriverAxis=(X=-1,Y=0,Z=0)
    DriverItem=(Type=Bone,Name="upperarm_r")
    FalloffRegionScaleFactors=(PositiveWidth=0.3,NegativeWidth=0.3,PositiveHeight=0.5,NegativeHeight=1)
    FalloffSize=0.45
    FlipHeightScaling=false
    FlipWidthScaling=true
    OptionalParentItem=(Type=Bone,Name="spine_04")
    RotationOffset=(X=-110,Y=25,Z=40)
   )
   > &RigUnit_ModifyTransforms(
    ItemToModify=((Item=(Type=Bone,Name="dyn_arm_chain_twist_under"),Transform=(Rotation=(X=0,Y=0,Z=0,W=1),Translation=(X=7,Y=-1,Z=-3),Scale3D=(X=1,Y=1,Z=1))))
    ItemToModify.0.Transform.Rotation=&RigVMFunction_MathQuaternionFromEuler(Euler=(X=0,Y=15,Z=-10)).Result
    Weight=@n9.OutputParam
   )
   > &RigUnit_ModifyTransforms(
    ItemToModify=((Item=(Type=Bone,Name="dyn_arm_chain_twist_back"),Transform=(Rotation=(X=0,Y=0,Z=0,W=1),Translation=(X=5,Y=-2.5,Z=0),Scale3D=(X=1,Y=1,Z=1))))
    Weight=@n9.OutputParam
   )
   > &RigUnit_ModifyTransforms(
    ItemToModify=((Item=(Type=Bone,Name="dyn_arm_chain_twist_Parent"),Transform=(Rotation=(X=0,Y=0,Z=0,W=1),Translation=(X=6,Y=0,Z=0),Scale3D=(X=1,Y=1,Z=1))))
    ItemToModify.0.Transform.Rotation=&RigVMFunction_MathQuaternionFromEuler(Euler=(X=5,Y=0,Z=-25)).Result
    Weight=@n9.OutputParam
   )
   > &RigUnit_ModifyTransforms(
    ItemToModify=((Item=(Type=Bone,Name="dyn_arm_chain_twist_front"),Transform=(Rotation=(X=0,Y=0,Z=0,W=1),Translation=(X=4,Y=2,Z=0),Scale3D=(X=1,Y=1,Z=1))))
    Weight=@n9.OutputParam
   )
   > n10 = &RigUnit_SphericalPoseReader(
    ActiveRegionScaleFactors=(PositiveWidth=0.5,NegativeWidth=1.5,PositiveHeight=1.3,NegativeHeight=0.5)
    ActiveRegionSize=0.2
    Debug=(bDrawDebug=false,DebugScale=25,DebugSegments=20,DebugThickness=0.25)
    DriverAxis=(X=-1,Y=0,Z=0)
    DriverItem=(Type=Bone,Name="upperarm_r")
    FalloffRegionScaleFactors=(PositiveWidth=1,NegativeWidth=1.5,PositiveHeight=0.3,NegativeHeight=0.3)
    FalloffSize=0.45
    FlipHeightScaling=false
    FlipWidthScaling=false
    OptionalParentItem=(Type=Bone,Name="spine_04")
    RotationOffset=(X=155,Y=0,Z=0)
   )
   > &RigUnit_ModifyTransforms(
    ItemToModify=((Item=(Type=Bone,Name="dyn_arm_chain_twist_under"),Transform=(Rotation=(X=0,Y=0,Z=0,W=1),Translation=(X=8,Y=4,Z=1),Scale3D=(X=1,Y=1,Z=1))))
    Weight=@n10.OutputParam
   )
   > &RigUnit_ModifyTransforms(
    ItemToModify=((Item=(Type=Bone,Name="dyn_arm_chain_twist_back"),Transform=(Rotation=(X=0,Y=0,Z=0,W=1),Translation=(X=7,Y=1,Z=0),Scale3D=(X=1,Y=1,Z=1))))
    Weight=@n10.OutputParam
   )
   > &RigUnit_ModifyTransforms(
    ItemToModify=((Item=(Type=Bone,Name="dyn_arm_chain_twist_Parent"),Transform=(Rotation=(X=0,Y=0,Z=0,W=1),Translation=(X=4,Y=0,Z=0),Scale3D=(X=1,Y=1,Z=1))))
    Weight=@n10.OutputParam
   )
   > &RigUnit_ModifyTransforms(
    ItemToModify=((Item=(Type=Bone,Name="dyn_arm_chain_twist_front"),Transform=(Rotation=(X=0,Y=0,Z=0,W=1),Translation=(X=6,Y=2,Z=0),Scale3D=(X=1,Y=1,Z=1))))
    Weight=@n10.OutputParam
   )
   n11 = &RigUnit_SphericalPoseReader(
    ActiveRegionScaleFactors=(PositiveWidth=2,NegativeWidth=2,PositiveHeight=0.5,NegativeHeight=1)
    ActiveRegionSize=0.1
    Debug=(bDrawDebug=false,DebugScale=25,DebugSegments=20,DebugThickness=0.25)
    DriverAxis=(X=0,Y=1,Z=0)
    DriverItem=(Type=Bone,Name="spine_05")
    FalloffRegionScaleFactors=(PositiveWidth=1,NegativeWidth=1,PositiveHeight=1,NegativeHeight=1)
    FalloffSize=0.35
    OptionalParentItem=(Type=Bone,Name="pelvis")
    RotationOffset=(X=15,Y=0,Z=0)
    ExecutePin=@SequenceAggregate.E
   )
   > &RigUnit_ModifyTransforms(
    ItemToModify=((Item=(Type=Bone,Name="dyn_jacketFront_mstr_l"),Transform=(Rotation=(X=0,Y=0,Z=0,W=1),Translation=(X=0,Y=0,Z=0),Scale3D=(X=1,Y=1,Z=1))))
    ItemToModify.0.Transform.Rotation=&RigVMFunction_MathQuaternionFromEuler(Euler=(X=0,Y=-5,Z=-5)).Result
    Weight=@n11.OutputParam
   )
   > &RigUnit_ModifyTransforms(
    ItemToModify=((Item=(Type=Bone,Name="dyn_jacketSide_mstr_l"),Transform=(Rotation=(X=0,Y=0,Z=0,W=1),Translation=(X=0,Y=0,Z=0),Scale3D=(X=1,Y=1,Z=1))))
    ItemToModify.0.Transform.Rotation=&RigVMFunction_MathQuaternionFromEuler(Euler=(X=0,Y=-15,Z=0)).Result
    Weight=@n11.OutputParam
   )
   > n12 = &RigUnit_SphericalPoseReader(
    ActiveRegionScaleFactors=(PositiveWidth=2,NegativeWidth=2,PositiveHeight=1,NegativeHeight=1)
    ActiveRegionSize=0.1
    Debug=(bDrawDebug=false,DebugScale=25,DebugSegments=20,DebugThickness=0.25)
    DriverAxis=(X=0,Y=1,Z=0)
    DriverItem=(Type=Bone,Name="spine_05")
    FalloffRegionScaleFactors=(PositiveWidth=1,NegativeWidth=1,PositiveHeight=1,NegativeHeight=1)
    FalloffSize=0.4
    FlipHeightScaling=true
    FlipWidthScaling=false
    OptionalParentItem=(Type=Bone,Name="pelvis")
    RotationOffset=(X=120,Y=0,Z=0)
   )
   > &RigUnit_ModifyTransforms(
    ItemToModify=((Item=(Type=Bone,Name="dyn_jacketFront_mstr_r"),Transform=(Rotation=(X=0,Y=0,Z=0,W=1),Translation=(X=0,Y=0,Z=0),Scale3D=(X=1,Y=1,Z=1))))
    ItemToModify.0.Transform.Rotation=&RigVMFunction_MathQuaternionFromEuler(Euler=(X=0,Y=-5,Z=3)).Result
    Weight=@n12.OutputParam
   )
   > &RigUnit_ModifyTransforms(
    ItemToModify=((Item=(Type=Bone,Name="dyn_jacketFront_mstr"),Transform=(Rotation=(X=0,Y=0,Z=0,W=1),Translation=(X=0,Y=0,Z=0),Scale3D=(X=1,Y=1,Z=1))))
    ItemToModify.0.Transform.Rotation=&RigVMFunction_MathQuaternionFromEuler(Euler=(X=0,Y=-5,Z=5)).Result
    Weight=@n12.OutputParam
   )
   > &RigUnit_ModifyTransforms(
    ItemToModify=((Item=(Type=Bone,Name="dyn_jacketSide_mstr_r"),Transform=(Rotation=(X=0,Y=0,Z=0,W=1),Translation=(X=0,Y=0,Z=0),Scale3D=(X=1,Y=1,Z=1))))
    ItemToModify.0.Transform.Rotation=&RigVMFunction_MathQuaternionFromEuler(Euler=(X=0,Y=-5,Z=0)).Result
    Weight=@n12.OutputParam
   )
   &RigUnit_ModifyTransforms(
    ExecutePin=()
    ItemToModify=((Item=(Type=Bone,Name="dyn_belt_rope_b_MSTR"),Transform=(Rotation=(X=0,Y=0,Z=0,W=1),Translation=(X=-6,Y=0,Z=0),Scale3D=(X=1,Y=1,Z=1))))
    ItemToModify.0.Transform.Rotation=&RigVMFunction_MathQuaternionFromEuler(Euler=(X=0,Y=25,Z=25)).Result
    Weight=@n5.OutputParam
   )
   &RigUnit_ModifyTransforms(
    ExecutePin=()
    ItemToModify=((Item=(Type=Bone,Name="dyn_jacketFront_01"),Transform=(Rotation=(X=0,Y=0,Z=0,W=1),Translation=(X=0,Y=0,Z=0),Scale3D=(X=1,Y=1,Z=1))))
    Weight=1
    ItemToModify.0.Transform.Rotation=&RigVMFunction_MathQuaternionFromEuler(Euler=(X=0,Y=5,Z=5)).Result
   )
   &RigUnit_ModifyTransforms(
    ExecutePin=()
    ItemToModify=((Item=(Type=Bone,Name="dyn_arm_chain_twist_MSTR"),Transform=(Rotation=(X=0,Y=0,Z=0,W=1),Translation=(X=0,Y=0,Z=0),Scale3D=(X=1,Y=1,Z=1))))
    Weight=1
    ItemToModify.0.Transform.Rotation=&RigVMFunction_MathQuaternionFromEuler(Euler=(X=10,Y=5,Z=-15)).Result
   )
  ]}
]
hierarchy [
 root {loc=<0 0 0>, rot=<0 0 0 1>}
 pelvis {parent=root, loc=<0 -0.5301 98.0571>, rot=<0.0127 0.707 -0.0127 -0.707>}
 spine_01 {parent=pelvis, loc=<6.1208 -0.9328 0>, rot=<0 0 0.0576 -0.9983>}
 spine_02 {parent=spine_01, loc=<6.6022 -0.0153 0>, rot=<0 0 0.0022 -1>}
 spine_03 {parent=spine_02, loc=<6.4919 -0.0105 0>, rot=<0 0 0.0015 -1>}
 spine_04 {parent=spine_03, loc=<6.9596 -0.007 0>, rot=<0 0 0.0009 -1>}
 spine_05 {parent=spine_04, loc=<6.9981 -0.0017 0>, rot=<0 0 -0.0042 -1>}
 neck_01 {parent=spine_05, loc=<11.3138 0.1529 0>, rot=<0 0 -0.2003 -0.9797>}
 neck_02 {parent=neck_01, loc=<4.5627 0 0>, rot=<0 0 0.0504 -0.9987>}
 head {parent=neck_02, loc=<5.3914 0 0>, rot=<0 0 0.115 -0.9934>}
 dyn_neck_choker_MSTR {parent=neck_01, loc=<1.7419 5.3719 -0.1829>, rot=<-0.1682 -0.6867 -0.6866 -0.1697>}
 dyn_neck_choker {parent=dyn_neck_choker_MSTR, loc=<0 0 0>, rot=<0 0 0 1>}
 clavicle_l {parent=spine_05, loc=<5.4831 -0.1053 -3.2946>, rot=<0.0938 -0.7388 0.0307 -0.6666>}
 upperarm_l {parent=clavicle_l, loc=<9.9486 0 0>, rot=<0.0371 -0.2992 -0.0705 -0.9509>}
 lowerarm_l {parent=upperarm_l, loc=<26.4793 0 0>, rot=<0 0 -0.2382 -0.9712>}
 hand_l {parent=lowerarm_l, loc=<22.7458 0 0>, rot=<-0.6747 -0.0182 0.0605 -0.7354>}
 thumb_01_l {parent=hand_l, loc=<1.3118 -0.3191 -2.459>, rot=<0.6546 -0.4042 0.0664 -0.6354>}
 thumb_02_l {parent=thumb_01_l, loc=<5.9322 -0.0253 0.1416>, rot=<0.0121 0.0046 0.2964 -0.955>}
 thumb_03_l {parent=thumb_02_l, loc=<2.2223 0.3682 0>, rot=<0 0 0.0172 -0.9999>}
 index_metacarpal_l {parent=hand_l, loc=<1.6535 0.0183 -2.96>, rot=<0.0158 -0.0124 -0.0173 -0.9996>}
 index_01_l {parent=index_metacarpal_l, loc=<8.3371 0.1718 -0.2461>, rot=<0.0363 0.0499 0.142 -0.9879>}
 index_02_l {parent=index_01_l, loc=<3.4919 -0.3004 0>, rot=<0 0 0.2948 -0.9556>}
 index_03_l {parent=index_02_l, loc=<2.5578 -0.123 0>, rot=<0 0 0.1968 -0.9804>}
 middle_metacarpal_l {parent=hand_l, loc=<2.1856 -0.1725 -0.5984>, rot=<0.006 0.0168 -0.0329 -0.9993>}
 middle_01_l {parent=middle_metacarpal_l, loc=<7.6374 0.346 -0.1508>, rot=<-0.0085 0.0328 0.1813 -0.9829>}
 middle_02_l {parent=middle_01_l, loc=<3.9108 0 0>, rot=<0 0 0.2808 -0.9598>}
 middle_03_l {parent=middle_02_l, loc=<2.6103 0.0517 0>, rot=<0 0 0.1446 -0.9895>}
 ring_metacarpal_l {parent=hand_l, loc=<2.2074 -0.3625 1.3638>, rot=<-0.0284 0.0722 -0.0458 -0.9959>}
 ring_01_l {parent=ring_metacarpal_l, loc=<7.2483 0.0329 -0.1426>, rot=<-0.0456 -0.0277 0.2951 -0.954>}
 ring_02_l {parent=ring_01_l, loc=<3.7208 0.4647 0>, rot=<0 0 0.1583 -0.9874>}
 ring_03_l {parent=ring_02_l, loc=<2.4152 0.114 0>, rot=<0 0 0.1504 -0.9886>}
 pinky_metacarpal_l {parent=hand_l, loc=<1.7616 -0.9753 3.3814>, rot=<-0.0397 0.0814 -0.069 -0.9935>}
 pinky_01_l {parent=pinky_metacarpal_l, loc=<7.1933 -0.4205 -0.2601>, rot=<-0.0487 -0.0063 0.2556 -0.9655>}
 pinky_02_l {parent=pinky_01_l, loc=<2.8274 0.0893 0>, rot=<0 0 0.2474 -0.9689>}
 pinky_03_l {parent=pinky_02_l, loc=<2.2046 0.0307 0>, rot=<0 0 0.125 -0.9922>}
 deform_wrist_dn_base_l {parent=hand_l, loc=<0 -1.5326 0>, rot=<0 0 0 -1>}
 deform_wrist_dn_l {parent=deform_wrist_dn_base_l, loc=<0 0 0>, rot=<0 0 0 1>}
 deform_wrist_up_base_l {parent=hand_l, loc=<0 2.3009 0>, rot=<0 0 0 -1>}
 deform_wrist_up_l {parent=deform_wrist_up_base_l, loc=<0 0 0>, rot=<0 0 0 1>}
 lowerarm_twist_02_l {parent=lowerarm_l, loc=<9.6199 0 0>, rot=<0 0 0 -1>}
 lowerarm_twist_01_l {parent=lowerarm_l, loc=<17.8549 0 0>, rot=<0 0 0 -1>}
 dyn_wrist_shackle {parent=lowerarm_twist_01_l, loc=<1.6371 0 0>, rot=<0 0 0 1>}
 dyn_wrist_shackle_chain_MSTR {parent=dyn_wrist_shackle, loc=<5.6591 -6.9636 0.9384>, rot=<0.7215 -0.1214 0.0781 -0.6772>}
 dyn_wrist_shackle_chain_01 {parent=dyn_wrist_shackle_chain_MSTR, loc=<0 0 0>, rot=<0 0 0 1>}
 dyn_wrist_shackle_chain_02 {parent=dyn_wrist_shackle_chain_01, loc=<3.6536 -0.0281 0.257>, rot=<0 0.0242 0 -0.9997>}
 dyn_wrist_shackle_chain_03 {parent=dyn_wrist_shackle_chain_02, loc=<3.7397 -0.1977 0.0749>, rot=<0 0.0351 0 -0.9994>}
 dyn_wrist_shackle_chain_04 {parent=dyn_wrist_shackle_chain_03, loc=<3.4584 0 -0.0389>, rot=<0 0 0 1>}
 deform_elbow_in_base_l {parent=lowerarm_l, loc=<0.1385 3.6052 -0.2182>, rot=<0.0059 0.0117 -0.0071 -0.9999>}
 deform_elbow_in_l {parent=deform_elbow_in_base_l, loc=<0 0 0>, rot=<0 0 0 1>}
 deform_elbow_out_base_l {parent=lowerarm_l, loc=<0.0565 -3.9993 0.1259>, rot=<0.0059 0.0117 -0.0071 -0.9999>}
 deform_elbow_out_l {parent=deform_elbow_out_base_l, loc=<0 0 0>, rot=<0 0 0 1>}
 upperarm_twist_01_l {parent=upperarm_l, loc=<0.0041 0 0>, rot=<0 0 0 1>}
 deform_upperarm_bk_base_l {parent=upperarm_twist_01_l, loc=<0.9473 -4.46 0.3113>, rot=<0.0004 -0.0034 -0.0158 -0.9999>}
 deform_upperarm_bk_l {parent=deform_upperarm_bk_base_l, loc=<0 0 0>, rot=<0 0 0 1>}
 deform_upperarm_fr_base_l {parent=upperarm_twist_01_l, loc=<1.4228 5.0263 0.3212>, rot=<0.0004 -0.0034 -0.0158 -0.9999>}
 deform_upperarm_fr_l {parent=deform_upperarm_fr_base_l, loc=<0 0 0>, rot=<0 0 0 1>}
 deform_upperarm_up_base_l {parent=upperarm_twist_01_l, loc=<-1.6653 0.3195 3.9629>, rot=<0.0004 -0.0034 -0.0158 -0.9999>}
 deform_upperarm_up_l {parent=deform_upperarm_up_base_l, loc=<0 0 0>, rot=<0 0 0 1>}
 upperarm_twist_02_l {parent=upperarm_l, loc=<13.7941 0 0>, rot=<0 0 0 1>}
 clavicle_r {parent=spine_05, loc=<5.4831 -0.1053 3.2946>, rot=<0.7388 0.0938 -0.6666 -0.0307>}
 upperarm_r {parent=clavicle_r, loc=<-9.9486 0 0>, rot=<0.0371 -0.2992 -0.0705 -0.9509>}
 lowerarm_r {parent=upperarm_r, loc=<-26.4793 0 0>, rot=<0 0 -0.2382 -0.9712>}
 hand_r {parent=lowerarm_r, loc=<-22.7458 0 0>, rot=<-0.6747 -0.0182 0.0605 -0.7354>}
 thumb_01_r {parent=hand_r, loc=<-1.3118 0.3191 2.459>, rot=<0.6546 -0.4042 0.0664 -0.6354>}
 thumb_02_r {parent=thumb_01_r, loc=<-5.9322 0.0253 -0.1416>, rot=<0.0121 0.0046 0.2964 -0.955>}
 thumb_03_r {parent=thumb_02_r, loc=<-2.2223 -0.3682 0>, rot=<0 0 0.0172 -0.9999>}
 index_metacarpal_r {parent=hand_r, loc=<-1.6535 -0.0183 2.96>, rot=<0.0158 -0.0124 -0.0173 -0.9996>}
 index_01_r {parent=index_metacarpal_r, loc=<-8.3371 -0.1718 0.2461>, rot=<0.0363 0.0499 0.142 -0.9879>}
 index_02_r {parent=index_01_r, loc=<-3.4919 0.3004 0>, rot=<0 0 0.2948 -0.9556>}
 index_03_r {parent=index_02_r, loc=<-2.5578 0.123 0>, rot=<0 0 0.1968 -0.9804>}
 middle_metacarpal_r {parent=hand_r, loc=<-2.1856 0.1725 0.5984>, rot=<0.006 0.0168 -0.0329 -0.9993>}
 middle_01_r {parent=middle_metacarpal_r, loc=<-7.6374 -0.346 0.1508>, rot=<-0.0085 0.0328 0.1813 -0.9829>}
 middle_02_r {parent=middle_01_r, loc=<-3.9108 0 0>, rot=<0 0 0.2808 -0.9598>}
 middle_03_r {parent=middle_02_r, loc=<-2.6103 -0.0517 0>, rot=<0 0 0.1446 -0.9895>}
 ring_metacarpal_r {parent=hand_r, loc=<-2.2074 0.3625 -1.3638>, rot=<-0.0284 0.0722 -0.0458 -0.9959>}
 ring_01_r {parent=ring_metacarpal_r, loc=<-7.2483 -0.0329 0.1426>, rot=<-0.0456 -0.0277 0.2951 -0.954>}
 ring_02_r {parent=ring_01_r, loc=<-3.7208 -0.4647 0>, rot=<0 0 0.1583 -0.9874>}
 ring_03_r {parent=ring_02_r, loc=<-2.4152 -0.114 0>, rot=<0 0 0.1504 -0.9886>}
 pinky_metacarpal_r {parent=hand_r, loc=<-1.7616 0.9753 -3.3814>, rot=<-0.0397 0.0814 -0.069 -0.9935>}
 pinky_01_r {parent=pinky_metacarpal_r, loc=<-7.1933 0.4205 0.2601>, rot=<-0.0487 -0.0063 0.2556 -0.9655>}
 pinky_02_r {parent=pinky_01_r, loc=<-2.8274 -0.0893 0>, rot=<0 0 0.2474 -0.9689>}
 pinky_03_r {parent=pinky_02_r, loc=<-2.2046 -0.0307 0>, rot=<0 0 0.125 -0.9922>}
 deform_wrist_dn_base_r {parent=hand_r, loc=<-0.0001 1.5328 0>, rot=<0 0 0 -1>}
 deform_wrist_dn_r {parent=deform_wrist_dn_base_r, loc=<0 0 0>, rot=<0 0 0 1>}
 deform_wrist_up_base_r {parent=hand_r, loc=<-0.0002 -2.3007 -0.0001>, rot=<0 0 0 -1>}
 deform_wrist_up_r {parent=deform_wrist_up_base_r, loc=<0 0 0>, rot=<0 0 0 1>}
 lowerarm_twist_02_r {parent=lowerarm_r, loc=<-9.6199 0 0>, rot=<0 0 0 -1>}
 lowerarm_twist_01_r {parent=lowerarm_r, loc=<-17.8549 0 0>, rot=<0 0 0 -1>}
 deform_elbow_in_base_r {parent=lowerarm_r, loc=<-0.1385 -3.6052 0.2182>, rot=<0.0059 0.0117 -0.0071 -0.9999>}
 deform_elbow_in_r {parent=deform_elbow_in_base_r, loc=<0 0 0>, rot=<0 0 0 1>}
 deform_elbow_out_base_r {parent=lowerarm_r, loc=<-0.0565 3.9993 -0.126>, rot=<0.0059 0.0117 -0.0071 -0.9999>}
 deform_elbow_out_r {parent=deform_elbow_out_base_r, loc=<0 0 0>, rot=<0 0 0 1>}
 upperarm_twist_01_r {parent=upperarm_r, loc=<-0.0041 0 0>, rot=<0 0 0 -1>}
 deform_upperarm_bk_base_r {parent=upperarm_twist_01_r, loc=<-0.9473 4.46 -0.3113>, rot=<0.0004 -0.0034 -0.0158 -0.9999>}
 deform_upperarm_bk_r {parent=deform_upperarm_bk_base_r, loc=<0 0 0>, rot=<0 0 0 1>}
 deform_upperarm_fr_base_r {parent=upperarm_twist_01_r, loc=<-1.4228 -5.0263 -0.3212>, rot=<0.0004 -0.0034 -0.0158 -0.9999>}
 deform_upperarm_fr_r {parent=deform_upperarm_fr_base_r, loc=<0 0 0>, rot=<0 0 0 1>}
 deform_upperarm_up_base_r {parent=upperarm_twist_01_r, loc=<1.6653 -0.3195 -3.9628>, rot=<0.0004 -0.0034 -0.0158 -0.9999>}
 deform_upperarm_up_r {parent=deform_upperarm_up_base_r, loc=<0 0 0>, rot=<0 0 0 1>}
 upperarm_twist_02_r {parent=upperarm_r, loc=<-13.7941 0 0>, rot=<0 0 0 -1>}
 dyn_arm_chain_twist_MSTR {parent=upperarm_twist_02_r, loc=<0.6002 -0.2419 2.7234>, rot=<0 0 0 1>}
 dyn_arm_chain_twist_parent {parent=dyn_arm_chain_twist_MSTR, loc=<1.1245 0.8123 -6.5778>, rot=<0 0 0 1>}
 dyn_arm_chain_twist_lookat_01 {parent=dyn_arm_chain_twist_parent, loc=<0 0 0>, rot=<0.7507 0.0767 0.6545 0.0465>}
 dyn_arm_chain_twist_front {parent=dyn_arm_chain_twist_MSTR, loc=<1.2217 -6.316 -0.5067>, rot=<0 0 0 1>}
 dyn_arm_chain_twist_back {parent=dyn_arm_chain_twist_MSTR, loc=<-0.9513 6.4041 -0.9406>, rot=<0 0 0 1>}
 dyn_arm_chain_twist_under {parent=dyn_arm_chain_twist_MSTR, loc=<2.0267 0.959 5.7721>, rot=<0 0 0 1>}
 dyn_arm_hook_MSTR {parent=clavicle_r, loc=<-11.6021 8.6642 8.6393>, rot=<0.4248 0.5653 0.484 -0.5155>}
 dyn_arm_hook_01 {parent=dyn_arm_hook_MSTR, loc=<0 0 0>, rot=<0.1363 -0.0449 0.0543 -0.9882>}
 dyn_arm_hook_02 {parent=dyn_arm_hook_01, loc=<2.7642 0 0>, rot=<-0.0478 0.0542 -0.0652 -0.9952>}
 dyn_arm_hook_03 {parent=dyn_arm_hook_02, loc=<3.1161 0 0>, rot=<-0.0181 0.0046 0.0104 -0.9998>}
 dyn_arm_hook_04 {parent=dyn_arm_hook_03, loc=<2.9039 0 0>, rot=<0 -0.0082 0.0052 -1>}
 dyn_arm_hook_05 {parent=dyn_arm_hook_04, loc=<2.9558 0 0>, rot=<0 0 0 1>}
 dyn_arm_hook_06 {parent=dyn_arm_hook_05, loc=<2.8486 0.0018 -0.0124>, rot=<-0.0703 -0.0374 -0.0026 -0.9968>}
 dyn_arm_hook_07 {parent=dyn_arm_hook_06, loc=<2.9175 0 0>, rot=<0 0.0375 0 -0.9993>}
 dyn_arm_hook_08 {parent=dyn_arm_hook_07, loc=<2.3941 0 0>, rot=<0 0 0 1>}
 dyn_arm_chain_tgt_MSTR {parent=clavicle_r, loc=<-12.0092 8.5884 7.873>, rot=<0.0559 0.3318 0.9059 0.2572>}
 dyn_arm_chain_tgt {parent=dyn_arm_chain_tgt_MSTR, loc=<0 0 0>, rot=<-0.2718 -0.583 -0.1565 -0.7495>}
 dyn_necklace_MSTR {parent=spine_05, loc=<1.674 9.0441 0.0615>, rot=<-0.4795 -0.5196 -0.5196 -0.4795>}
 dyn_necklace_01 {parent=dyn_necklace_MSTR, loc=<0 0 0>, rot=<-0.0035 -0.3593 -0.0014 -0.9332>}
 dyn_necklace_02 {parent=dyn_necklace_01, loc=<3.8227 0 0>, rot=<0.0096 -0.057 0.0097 -0.9983>}
 dyn_necklace_03 {parent=dyn_necklace_02, loc=<4.8149 0 0>, rot=<-0.0008 -0.237 -0.0108 -0.9715>}
 deform_pec_aim_base_l {parent=spine_04, loc=<3.1942 9.3086 -1.2237>, rot=<0.0614 -0.4031 0.3295 -0.8516>}
 deform_pec_aim_l {parent=deform_pec_aim_base_l, loc=<0 0 0>, rot=<0 0 0 1>}
 deform_pec_base_l {parent=deform_pec_aim_l, loc=<9.3665 0 0>, rot=<0 0 0 1>}
 deform_pec_l {parent=deform_pec_base_l, loc=<0 0 0>, rot=<0 0 0 1>}
 deform_pec_aim_base_r {parent=spine_04, loc=<3.1938 9.3085 1.2237>, rot=<-0.4031 -0.0614 0.8516 0.3295>}
 deform_pec_aim_r {parent=deform_pec_aim_base_r, loc=<0 0 0>, rot=<0 0 0 1>}
 deform_pec_base_r {parent=deform_pec_aim_r, loc=<-9.3667 -0.0003 -0.0004>, rot=<0 0 0 1>}
 deform_pec_r {parent=deform_pec_base_r, loc=<0 0 0>, rot=<0 0 0 1>}
 thigh_l {parent=pelvis, loc=<-1.0855 -0.2109 -8.7053>, rot=<0.0009 0.0345 0.0155 -0.9993>}
 calf_l {parent=thigh_l, loc=<-42.5268 0 0>, rot=<0 0 -0.0657 -0.9978>}
 foot_l {parent=calf_l, loc=<-46.8037 0 0>, rot=<0.0422 -0.04 0.0613 -0.9964>}
 ball_l {parent=foot_l, loc=<-3.5658 10.8066 -0.1544>, rot=<0.0001 -0.0001 -0.7063 -0.7079>}
 calf_twist_02_l {parent=calf_l, loc=<-18.8064 -0.0967 0.0397>, rot=<0.0237 0.0095 -0.0035 -0.9997>}
 calf_twist_01_l {parent=calf_l, loc=<-35.813 -0.0967 0.0397>, rot=<0.0237 0.0095 -0.0035 -0.9997>}
 deform_knee_in_base_l {parent=calf_l, loc=<0.5405 -5.0592 -0.5944>, rot=<-0.0003 -0.0151 0.0267 -0.9995>}
 deform_knee_in_l {parent=deform_knee_in_base_l, loc=<0 0 0>, rot=<0 0 0 1>}
 deform_knee_out_base_l {parent=calf_l, loc=<0.2138 3.9943 -0.8167>, rot=<-0.0003 -0.0151 0.0267 -0.9995>}
 deform_knee_out_l {parent=deform_knee_out_base_l, loc=<0 0 0>, rot=<0 0 0 1>}
 thigh_twist_01_l {parent=thigh_l, loc=<-21.629 -0.0919 -0.0002>, rot=<0.0013 0 0 -1>}
 deform_groin_base_l {parent=thigh_l, loc=<0.4808 5.6776 -0.2472>, rot=<0 0.001 0 -1>}
 deform_groin_l {parent=deform_groin_base_l, loc=<0 0 0>, rot=<0 0 0 1>}
 thigh_r {parent=pelvis, loc=<-1.0855 -0.2109 8.7053>, rot=<0.0345 -0.0009 0.9993 0.0155>}
 calf_r {parent=thigh_r, loc=<42.5268 0 0>, rot=<0 0 -0.0657 -0.9978>}
 foot_r {parent=calf_r, loc=<46.8037 0 0>, rot=<0.0422 -0.04 0.0613 -0.9964>}
 ball_r {parent=foot_r, loc=<3.5658 -10.8066 0.1544>, rot=<0.0001 -0.0001 -0.7063 -0.7079>}
 calf_twist_02_r {parent=calf_r, loc=<18.8064 0.0967 -0.0397>, rot=<0.0237 0.0095 -0.0035 -0.9997>}
 calf_twist_01_r {parent=calf_r, loc=<35.813 0.0967 -0.0397>, rot=<0.0237 0.0095 -0.0035 -0.9997>}
 dyn_ankle_shackle {parent=calf_twist_01_r, loc=<6.5448 -0.1083 -0.6203>, rot=<-0.7225 0.0179 0.0822 -0.6862>}
 deform_knee_in_base_r {parent=calf_r, loc=<-0.5405 5.0592 0.5943>, rot=<-0.0003 -0.0151 0.0267 -0.9995>}
 deform_knee_in_r {parent=deform_knee_in_base_r, loc=<0 0 0>, rot=<0 0 0 1>}
 deform_knee_out_base_r {parent=calf_r, loc=<-0.2137 -3.9943 0.8168>, rot=<-0.0003 -0.0151 0.0267 -0.9995>}
 deform_knee_out_r {parent=deform_knee_out_base_r, loc=<0 0 0>, rot=<0 0 0 1>}
 thigh_twist_01_r {parent=thigh_r, loc=<21.629 0.0919 0.0002>, rot=<0.0013 0 0 -1>}
 deform_groin_base_r {parent=thigh_r, loc=<-0.4808 -5.6776 0.2472>, rot=<0 0.001 0 -1>}
 deform_groin_r {parent=deform_groin_base_r, loc=<0 0 0>, rot=<0 0 0 1>}
 deform_glute_base_l {parent=pelvis, loc=<-7.3466 -9.0213 -9.1245>, rot=<0.0172 0.031 -0.4477 -0.8935>}
 deform_glute_l {parent=deform_glute_base_l, loc=<-0.0006 -0.0014 0>, rot=<0 0 0 1>}
 deform_glute_base_r {parent=pelvis, loc=<-7.3464 -9.0213 9.1245>, rot=<0.031 -0.0172 0.8935 -0.4477>}
 deform_glute_r {parent=deform_glute_base_r, loc=<0.0006 0.0014 0>, rot=<0 0 0 1>}
 dyn_jacketFront_mstr_l {parent=pelvis, loc=<3.9301 5.6511 -8.6525>, rot=<0.0147 -0.792 -0.6074 -0.0595>}
 dyn_jacketFront_01_l {parent=dyn_jacketFront_mstr_l, loc=<0 0 0>, rot=<0 0 0 1>}
 dyn_jacketFront_02_l {parent=dyn_jacketFront_01_l, loc=<17 0 0>, rot=<0.0172 -0.0001 0.0157 -0.9997>}
 dyn_jacketFront_03_l {parent=dyn_jacketFront_02_l, loc=<17 0 0>, rot=<-0.0062 0.0009 0.0018 -1>}
 dyn_jacketFront_04_l {parent=dyn_jacketFront_03_l, loc=<17 0 0>, rot=<0.0093 -0.0031 -0.0033 -0.9999>}
 dyn_jacketFront_05_l {parent=dyn_jacketFront_04_l, loc=<17 0 0>, rot=<0 0 0 1>}
 dyn_jacketSide_mstr_l {parent=pelvis, loc=<2.9899 -3.0758 -12.9621>, rot=<-0.0044 -0.9937 -0.0463 -0.1018>}
 dyn_jacketSide_01_l {parent=dyn_jacketSide_mstr_l, loc=<0 0 0>, rot=<0 0 0 1>}
 dyn_jacketSide_02_l {parent=dyn_jacketSide_01_l, loc=<17 0 0>, rot=<0.0172 -0.0001 0.0157 -0.9997>}
 dyn_jacketSide_03_l {parent=dyn_jacketSide_02_l, loc=<17 0 0>, rot=<-0.0062 0.0009 0.0018 -1>}
 dyn_jacketSide_04_l {parent=dyn_jacketSide_03_l, loc=<17 0 0>, rot=<0.0093 -0.0031 -0.0033 -0.9999>}
 dyn_jacketSide_05_l {parent=dyn_jacketSide_04_l, loc=<17 0 0>, rot=<0 0 0 1>}
 dyn_jacketBack_mstr_l {parent=pelvis, loc=<2.0378 -11.0935 -5.991>, rot=<0.0055 -0.7971 0.6021 -0.0457>}
 dyn_jacketBack_01_l {parent=dyn_jacketBack_mstr_l, loc=<0 0 0>, rot=<0 0 0.0219 -0.9998>}
 dyn_jacketBack_02_l {parent=dyn_jacketBack_01_l, loc=<17 0 0>, rot=<0.0271 0.0057 0.0002 -0.9996>}
 dyn_jacketBack_03_l {parent=dyn_jacketBack_02_l, loc=<17 0 0>, rot=<-0.0184 0.0097 -0.0002 -0.9998>}
 dyn_jacketBack_04_l {parent=dyn_jacketBack_03_l, loc=<17 0 0>, rot=<-0.0016 -0.0003 0 -1>}
 dyn_jacketBack_05_l {parent=dyn_jacketBack_04_l, loc=<17 0 0>, rot=<0 0 0 1>}
 dyn_jacketBack_mstr_r {parent=pelvis, loc=<2.0845 -11.0823 5.9883>, rot=<0.0457 -0.6021 0.7971 -0.0055>}
 dyn_jacketBack_01_r {parent=dyn_jacketBack_mstr_r, loc=<0 0 0>, rot=<0 0 -0.022 -0.9998>}
 dyn_jacketBack_02_r {parent=dyn_jacketBack_01_r, loc=<17 0 0>, rot=<-0.0271 0.0057 -0.0002 -0.9996>}
 dyn_jacketBack_03_r {parent=dyn_jacketBack_02_r, loc=<17 0 0>, rot=<-0.001 -0.0005 0 -1>}
 dyn_jacketBack_04_r {parent=dyn_jacketBack_03_r, loc=<17 0 0>, rot=<0.0016 -0.0002 0 -1>}
 dyn_jacketSide_mstr_r {parent=pelvis, loc=<2.9417 -3.074 12.9725>, rot=<0.1012 0.0304 0.9944 0.0029>}
 dyn_jacketSide_01_r {parent=dyn_jacketSide_mstr_r, loc=<0 0 0>, rot=<0 0 0 1>}
 dyn_jacketSide_02_r {parent=dyn_jacketSide_01_r, loc=<16.1034 0 0>, rot=<0.0544 0.002 -0.0124 -0.9984>}
 dyn_jacketSide_03_r {parent=dyn_jacketSide_02_r, loc=<16.1034 0 0>, rot=<0.0063 0.0012 -0.0016 -1>}
 dyn_jacketSide_04_r {parent=dyn_jacketSide_03_r, loc=<16.1034 0 0>, rot=<-0.0093 -0.0036 0.0027 -0.9999>}
 dyn_jacketFront_mstr_r {parent=pelvis, loc=<3.7154 3.9746 10.8476>, rot=<0.0726 0.471 0.8791 0.005>}
 dyn_jacketFront_01_r {parent=dyn_jacketFront_mstr_r, loc=<0 0 0>, rot=<0 0 0 1>}
 dyn_jacketFront_02_r {parent=dyn_jacketFront_01_r, loc=<15.761 0 0>, rot=<0.0544 0.002 -0.0124 -0.9984>}
 dyn_jacketFront_03_r {parent=dyn_jacketFront_02_r, loc=<15.761 0 0>, rot=<0.0063 0.0012 -0.0016 -1>}
 dyn_jacketFront_04_r {parent=dyn_jacketFront_03_r, loc=<15.761 0 0>, rot=<-0.0093 -0.0036 0.0027 -0.9999>}
 dyn_jacketFront_mstr {parent=pelvis, loc=<3.6939 7.2933 2.6211>, rot=<-0.0359 -0.7182 -0.6944 0.025>}
 dyn_jacketFront_01 {parent=dyn_jacketFront_mstr, loc=<0 0 0>, rot=<0 0 0 1>}
 dyn_jacketFront_02 {parent=dyn_jacketFront_01, loc=<16.413 0 0>, rot=<0.0544 0.002 -0.0124 -0.9984>}
 dyn_jacketFront_03 {parent=dyn_jacketFront_02, loc=<15.8262 0.0144 0.0031>, rot=<0.0063 0.0012 -0.0016 -1>}
 dyn_jacketFront_04 {parent=dyn_jacketFront_03, loc=<15.2392 0.0181 0.0061>, rot=<-0.0093 -0.0036 0.0027 -0.9999>}
 dyn_Flask_MSTR {parent=pelvis, loc=<4.6932 4.2126 -12.8251>, rot=<-0.0168 -0.9387 -0.3443 -0.0062>}
 dyn_Flask_01 {parent=dyn_Flask_MSTR, loc=<0 0 0>, rot=<0 0 0 1>}
 dyn_bag_MSTR {parent=pelvis, loc=<3.0181 4.2244 13.9882>, rot=<0.0012 0.069 0.9975 0.0179>}
 dyn_bag_01 {parent=dyn_bag_MSTR, loc=<0 0 0>, rot=<0 0 0 1>}
 dyn_bag_02 {parent=dyn_bag_01, loc=<7.3197 0.7037 1.0256>, rot=<-0.0063 0 0 -1>}
 dyn_belt_rope_a_MSTR {parent=dyn_bag_MSTR, loc=<1.4428 -2.8784 -0.5623>, rot=<0.997 -0.0337 -0.0046 0.0692>}
 dyn_belt_rope_a_01 {parent=dyn_belt_rope_a_MSTR, loc=<0 0 0>, rot=<0 0 0 1>}
 dyn_belt_rope_a_02 {parent=dyn_belt_rope_a_01, loc=<7.0647 0 0>, rot=<-0.0001 -0.0029 0.0293 -0.9996>}
 dyn_belt_rope_b_MSTR {parent=dyn_bag_MSTR, loc=<0.1764 -4.3938 0.8886>, rot=<0.9949 -0.0636 0.0409 0.0664>}
 dyn_belt_rope_b_01 {parent=dyn_belt_rope_b_MSTR, loc=<0 0 0>, rot=<0 0 0 1>}
 dyn_belt_rope_b_02 {parent=dyn_belt_rope_b_01, loc=<6.0432 0 0>, rot=<0.0033 0.0399 0.051 -0.9979>}
 dyn_belt_rope_b_03 {parent=dyn_belt_rope_b_02, loc=<8.6779 0 0>, rot=<0 -0.0051 0.0291 -0.9996>}
 dyn_belt_MSTR {parent=pelvis, loc=<-0.9188 8.9877 3.8032>, rot=<0.0127 0.707 0.707 0.0127>}
 dyn_belt {parent=dyn_belt_MSTR, loc=<0 0 0>, rot=<0 0 0 1>}
 floorCOL {parent=root, loc=<0 0 0>, rot=<0 0 0 1>}
]
