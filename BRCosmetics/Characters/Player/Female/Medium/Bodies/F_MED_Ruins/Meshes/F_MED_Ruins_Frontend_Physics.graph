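# Physics asset for the F_MED_Ruins female body's Frontend variant, simulating a long jacket around the lower body.
# Capsules on pelvis, thighs, calves and feet form the body shapes beneath seven jacket bone chains (front, sides, back).
# Each chain starts at a tiny master body with free rotation on the pelvis, then uses angle-limited joints.
# Free-rotation cross-constraints tie neighboring chains at segments 03 and 04; left chains run one segment longer than right.
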PhysicsAsset /BRCosmetics/Characters/Player/Female/Medium/Bodies/F_MED_Ruins/Meshes/F_MED_Ruins_Frontend_Physics
bodies {
 calf_l [{shapeType=Capsule, length=38.7928, radius=8.1658, center=<-23.4523 -1.1162 0.1776>, rotation=<rot 0.1246 84.5345 -88.5757>}]
 calf_r [{shapeType=Capsule, length=39.9107, radius=8.7978, center=<26.3559 0.7114 -0.6693>, rotation=<rot -0.3385 89.5329 -88.4634>}]
 dyn_jacketBack_01_l [{shapeType=Capsule, length=7.125, radius=9.875, shapeName=dyn_jacketBack_01_l_capsule, center=<7.5 0 8>, rotation=<rot -0.0002 0.0001 80>}]
 dyn_jacketBack_01_r [{shapeType=Capsule, length=7, radius=10, shapeName=dyn_jacketBack_01_r_capsule, center=<7.5 0 8>, rotation=<rot -0.0001 0 100>}]
 dyn_jacketBack_02_l [{shapeType=Capsule, length=21.5, radius=11.5, shapeName=dyn_jacketBack_02_l_capsule, center=<7.5 0 8>, rotation=<rot -0.0001 0 80>}]
 dyn_jacketBack_02_r [{shapeType=Capsule, length=18.5, radius=12, shapeName=dyn_jacketBack_02_r_capsule, center=<7.5 0 8>, rotation=<rot -0.0001 0 100>}]
 dyn_jacketBack_03_l [{shapeType=Capsule, length=20, radius=13.5, shapeName=dyn_jacketBack_03_l_capsule, center=<7.5 0 8>, rotation=<rot -0.0002 0.0001 91.9578>}]
 dyn_jacketBack_03_r [{shapeType=Capsule, length=20, radius=13.5, shapeName=dyn_jacketBack_03_r_capsule, center=<7.5 -2 8>, rotation=<rot -0.0002 -0.0001 90>}]
 dyn_jacketBack_04_l [
  {
   shapeType=Capsule
   length=33.625
   radius=13.875
   shapeName=dyn_jacketBack_04_l_capsule
   center=<7.5 0 8>
   rotation=<rot -0.0004 0.0001 96.9578>
  }
  {
   shapeType=Capsule
   length=9
   radius=10.875
   shapeName=dyn_jacketBack_04_l_capsule
   center=<13.654 -17.256 1.4866>
   rotation=<rot -0.0004 0.0001 70>
  }
 ]
 dyn_jacketBack_04_r [{shapeType=Capsule, length=32.566, radius=14.184, shapeName=dyn_jacketBack_04_r_capsule, center=<8 -3 8>, rotation=<rot -0.0003 -0.0001 90>}]
 dyn_jacketBack_05_l [{shapeType=Capsule, length=32.125, radius=13, shapeName=dyn_jacketBack_05_l_capsule, center=<7.5 0 8>, rotation=<rot -0.0004 0.0001 83.65>}]
 dyn_jacketBack_mstr_l [{shapeType=Capsule, length=0.505, radius=0.505, shapeName=dyn_jacketBack_mstr_l_capsule, rotation=<rot 0 -90 0>}]
 dyn_jacketBack_mstr_r [{shapeType=Capsule, length=0.505, radius=0.505, shapeName=dyn_jacketBack_mstr_r_capsule, rotation=<rot 0 -90 0>}]
 dyn_jacketFront_01 [{shapeType=Capsule, length=9, radius=8, shapeName=dyn_jacketFront_01_capsule, center=<5 0 0>, rotation=<rot 0 0 90>}]
 dyn_jacketFront_01_l [{shapeType=Capsule, length=9, radius=8, shapeName=dyn_jacketFront_01_l_capsule, center=<5 0 0>, rotation=<rot 0 0 90>}]
 dyn_jacketFront_01_r [{shapeType=Capsule, length=9, radius=8, shapeName=dyn_jacketFront_01_r_capsule, center=<5 0 0>, rotation=<rot 0 0 90>}]
 dyn_jacketFront_02 [{shapeType=Capsule, length=9, radius=8, shapeName=dyn_jacketFront_02_capsule, center=<5 0 0>, rotation=<rot 0 0 90>}]
 dyn_jacketFront_02_l [{shapeType=Capsule, length=9, radius=8, shapeName=dyn_jacketFront_02_l_capsule, center=<5 0 0>, rotation=<rot 0 0 90>}]
 dyn_jacketFront_02_r [{shapeType=Capsule, length=15, radius=9, shapeName=dyn_jacketFront_02_r_capsule, center=<6 0 9.5>, rotation=<rot 0 0 90>}]
 dyn_jacketFront_03 [{shapeType=Capsule, length=6.875, radius=10.125, shapeName=dyn_jacketFront_03_capsule, center=<7.5 0 9.5>, rotation=<rot 0.0003 0 80>}]
 dyn_jacketFront_03_l [{
   shapeType=Capsule
   length=8.75
   radius=8.25
   shapeName=dyn_jacketFront_03_l_capsule
   center=<5 -3.5 7.5>
   rotation=<rot 0.0007 0.0002 74.9999>
  }]
 dyn_jacketFront_03_r [{
   shapeType=Capsule
   length=20.125
   radius=9.625
   shapeName=dyn_jacketFront_03_r_capsule
   center=<7.5 0 8>
   rotation=<rot 0.0002 -0.0002 90.1591>
  }]
 dyn_jacketFront_04 [{shapeType=Capsule, length=6.375, radius=10.625, shapeName=dyn_jacketFront_04_capsule, center=<8 0 9>, rotation=<rot 0.0001 0 85>}]
 dyn_jacketFront_04_l [{
   shapeType=Capsule
   length=8.75
   radius=8.25
   shapeName=dyn_jacketFront_03_l_capsule
   center=<5 -3.5 7.5>
   rotation=<rot 0.0007 0.0002 74.9999>
  }]
 dyn_jacketFront_04_r [{
   shapeType=Capsule
   length=18.625
   radius=11.125
   shapeName=dyn_jacketFront_04_r_capsule
   center=<7.5 0 8>
   rotation=<rot 0.0001 -0.0001 85.1591>
  }]
 dyn_jacketFront_05_l [{
   shapeType=Capsule
   length=9.864
   radius=8.9797
   shapeName=dyn_jacketFront_05_l_capsule
   center=<8.0598 -1 6.5>
   rotation=<rot 0 -97.9664 -86.2348>
  }]
 dyn_jacketFront_mstr [{shapeType=Capsule, length=0.505, radius=0.505, shapeName=dyn_jacketFront_mstr_capsule, rotation=<rot 0 -90 0>}]
 dyn_jacketFront_mstr_l [{shapeType=Capsule, length=0.505, radius=0.505, shapeName=dyn_jacketFront_mstr_l_capsule, rotation=<rot 0 -90 0>}]
 dyn_jacketFront_mstr_r [{shapeType=Capsule, length=0.505, radius=0.505, shapeName=dyn_jacketFront_mstr_r_capsule, rotation=<rot 0 -90 0>}]
 dyn_jacketSide_01_l [{shapeType=Capsule, length=9, radius=8, shapeName=dyn_jacketSide_01_l_capsule, center=<5 0 0>, rotation=<rot 0 0 90>}]
 dyn_jacketSide_01_r [{shapeType=Capsule, length=9, radius=8, shapeName=dyn_jacketSide_01_r_capsule, center=<5 0 0>, rotation=<rot 0 0 90>}]
 dyn_jacketSide_02_l [{shapeType=Capsule, length=15.75, radius=8.875, shapeName=dyn_jacketSide_02_l_capsule, center=<7.5 0 7.5>, rotation=<rot 0 0 82.0268>}]
 dyn_jacketSide_02_r [{shapeType=Capsule, length=20.375, radius=10.375, shapeName=dyn_jacketSide_02_r_capsule, center=<7.5 0 8>, rotation=<rot 0 -0.0002 95>}]
 dyn_jacketSide_03_l [{shapeType=Capsule, length=36.75, radius=10.125, shapeName=dyn_jacketSide_03_l_capsule, center=<7.5 -3 8>, rotation=<rot 0 0 82.0268>}]
 dyn_jacketSide_03_r [
  {shapeType=Capsule, length=14.25, radius=11.25, shapeName=dyn_jacketSide_03_r_capsule, center=<7.5 0 8>, rotation=<rot 0 -0.0002 95>}
  {
   shapeType=Capsule
   length=8.25
   radius=9.75
   shapeName=dyn_jacketSide_03_r_capsule
   center=<10.1594 11.7747 0.2179>
   rotation=<rot 0 -0.0001 -50>
  }
 ]
 dyn_jacketSide_04_l [{shapeType=Capsule, length=36.375, radius=10.5, shapeName=dyn_jacketSide_04_l_capsule, center=<7.5 -3 8>, rotation=<rot 0 0 82.0268>}]
 dyn_jacketSide_04_r [
  {shapeType=Capsule, length=19.125, radius=12, shapeName=dyn_jacketSide_04_r_capsule, center=<7.5 0 8>, rotation=<rot 0 -0.0002 95>}
  {
   shapeType=Capsule
   length=9.125
   radius=9.75
   shapeName=dyn_jacketSide_04_r_capsule
   center=<4.362 14.858 -2.4306>
   rotation=<rot 4.094 -2.8731 -30.1026>
  }
 ]
 dyn_jacketSide_05_l [{shapeType=Capsule, length=31.25, radius=11.25, shapeName=dyn_jacketSide_05_l_capsule, center=<7.5 -3 8>, rotation=<rot 0 0 82.0268>}]
 dyn_jacketSide_mstr_l [{shapeType=Capsule, length=0.505, radius=0.505, shapeName=dyn_jacketSide_mstr_l_capsule, rotation=<rot 0 -90 0>}]
 dyn_jacketSide_mstr_r [{shapeType=Capsule, length=0.505, radius=0.505, shapeName=dyn_jacketSide_mstr_r_capsule, rotation=<rot 0 -90 0>}]
 foot_l [{
   shapeType=Capsule
   length=8.2636
   radius=11.1655
   shapeName=foot_l_capsule
   center=<-5.043 4.5841 0.5164>
   rotation=<rot 2.2652 -176.811 -85.1339>
  }]
 foot_r [{
   shapeType=Capsule
   length=9.2412
   radius=10.1879
   shapeName=foot_r_capsule
   center=<3.6724 -3.9727 -0.3857>
   rotation=<rot 2.0111 4.5862 -84.4638>
  }]
 pelvis [
  {shapeType=Capsule, length=7.7125, radius=8.4301, center=<-0.9227 -1.2812 0>}
  {shapeType=Capsule, length=10.79, radius=7.4702, shapeName=pelvis_capsule, center=<-3.9978 -3.5061 0>}
 ]
 thigh_l [{shapeType=Capsule, length=27.8152, radius=8.1848, center=<-21.6167 -1.0761 1.3426>, rotation=<rot 85 179.8044 179.9999>}]
 thigh_r [{shapeType=Capsule, length=27.6902, radius=8.3098, center=<21.0533 0.5413 -1.1849>, rotation=<rot 85 -180 -180>}]
}
constraints [
 {bone1Name=thigh_l, bone2Name=pelvis, swing1LimitDegrees=45, swing2LimitDegrees=45, twistLimitDegrees=45}
 {bone1Name=calf_l, bone2Name=thigh_l, swing1LimitDegrees=45, swing2LimitDegrees=45, twistLimitDegrees=45}
 {bone1Name=thigh_r, bone2Name=pelvis, swing1LimitDegrees=45, swing2LimitDegrees=45, twistLimitDegrees=45}
 {bone1Name=calf_r, bone2Name=thigh_r, swing1LimitDegrees=45, swing2LimitDegrees=45, twistLimitDegrees=45}
 {
  bone1Name=foot_l
  bone2Name=calf_l
  swing1LimitDegrees=45
  swing1Motion=Free
  swing2LimitDegrees=45
  swing2Motion=Free
  twistLimitDegrees=45
  twistMotion=Free
 }
 {
  bone1Name=foot_r
  bone2Name=calf_r
  swing1LimitDegrees=45
  swing1Motion=Free
  swing2LimitDegrees=45
  swing2Motion=Free
  twistLimitDegrees=45
  twistMotion=Free
 }
 {
  bone1Name=dyn_jacketFront_mstr_l
  bone2Name=pelvis
  swing1LimitDegrees=45
  swing1Motion=Free
  swing2LimitDegrees=45
  swing2Motion=Free
  twistLimitDegrees=45
  twistMotion=Free
 }
 {bone1Name=dyn_jacketFront_01_l, bone2Name=dyn_jacketFront_mstr_l, swing1LimitDegrees=35, swing2LimitDegrees=35, twistLimitDegrees=15}
 {bone1Name=dyn_jacketFront_02_l, bone2Name=dyn_jacketFront_01_l, swing1LimitDegrees=35, swing2LimitDegrees=35, twistLimitDegrees=15}
 {bone1Name=dyn_jacketFront_03_l, bone2Name=dyn_jacketFront_02_l, swing1LimitDegrees=35, swing2LimitDegrees=35, twistLimitDegrees=15}
 {bone1Name=dyn_jacketFront_04_l, bone2Name=dyn_jacketFront_03_l, swing1LimitDegrees=35, swing2LimitDegrees=35, twistLimitDegrees=15}
 {bone1Name=dyn_jacketFront_05_l, bone2Name=dyn_jacketFront_04_l, swing1LimitDegrees=35, swing2LimitDegrees=35, twistLimitDegrees=15}
 {
  bone1Name=dyn_jacketSide_mstr_l
  bone2Name=pelvis
  swing1LimitDegrees=45
  swing1Motion=Free
  swing2LimitDegrees=45
  swing2Motion=Free
  twistLimitDegrees=45
  twistMotion=Free
 }
 {bone1Name=dyn_jacketSide_01_l, bone2Name=dyn_jacketSide_mstr_l, swing1LimitDegrees=35, swing2LimitDegrees=35, twistLimitDegrees=15}
 {bone1Name=dyn_jacketSide_02_l, bone2Name=dyn_jacketSide_01_l, swing1LimitDegrees=35, swing2LimitDegrees=35, twistLimitDegrees=15}
 {bone1Name=dyn_jacketSide_03_l, bone2Name=dyn_jacketSide_02_l, swing1LimitDegrees=35, swing2LimitDegrees=35, twistLimitDegrees=15}
 {bone1Name=dyn_jacketSide_04_l, bone2Name=dyn_jacketSide_03_l, swing1LimitDegrees=35, swing2LimitDegrees=35, twistLimitDegrees=15}
 {bone1Name=dyn_jacketSide_05_l, bone2Name=dyn_jacketSide_04_l, swing1LimitDegrees=35, swing2LimitDegrees=35, twistLimitDegrees=15}
 {
  bone1Name=dyn_jacketBack_mstr_l
  bone2Name=pelvis
  swing1LimitDegrees=45
  swing1Motion=Free
  swing2LimitDegrees=45
  swing2Motion=Free
  twistLimitDegrees=45
  twistMotion=Free
 }
 {bone1Name=dyn_jacketBack_01_l, bone2Name=dyn_jacketBack_mstr_l, swing1LimitDegrees=15, swing2LimitDegrees=45, twistLimitDegrees=15}
 {bone1Name=dyn_jacketBack_02_l, bone2Name=dyn_jacketBack_01_l, swing1LimitDegrees=15, swing2LimitDegrees=45, twistLimitDegrees=15}
 {bone1Name=dyn_jacketBack_03_l, bone2Name=dyn_jacketBack_02_l, swing1LimitDegrees=35, swing2LimitDegrees=45, twistLimitDegrees=20}
 {bone1Name=dyn_jacketBack_04_l, bone2Name=dyn_jacketBack_03_l, swing1LimitDegrees=35, swing2LimitDegrees=45, twistLimitDegrees=20}
 {bone1Name=dyn_jacketBack_05_l, bone2Name=dyn_jacketBack_04_l, swing1LimitDegrees=35, swing2LimitDegrees=45, twistLimitDegrees=20}
 {
  bone1Name=dyn_jacketBack_mstr_r
  bone2Name=pelvis
  swing1LimitDegrees=45
  swing1Motion=Free
  swing2LimitDegrees=45
  swing2Motion=Free
  twistLimitDegrees=45
  twistMotion=Free
 }
 {bone1Name=dyn_jacketBack_01_r, bone2Name=dyn_jacketBack_mstr_r, swing1LimitDegrees=15, swing2LimitDegrees=45, twistLimitDegrees=15}
 {bone1Name=dyn_jacketBack_02_r, bone2Name=dyn_jacketBack_01_r, swing1LimitDegrees=15, swing2LimitDegrees=45, twistLimitDegrees=15}
 {bone1Name=dyn_jacketBack_03_r, bone2Name=dyn_jacketBack_02_r, swing1LimitDegrees=35, swing2LimitDegrees=45, twistLimitDegrees=20}
 {bone1Name=dyn_jacketBack_04_r, bone2Name=dyn_jacketBack_03_r, swing1LimitDegrees=35, swing2LimitDegrees=45, twistLimitDegrees=20}
 {
  bone1Name=dyn_jacketSide_mstr_r
  bone2Name=pelvis
  swing1LimitDegrees=45
  swing1Motion=Free
  swing2LimitDegrees=45
  swing2Motion=Free
  twistLimitDegrees=45
  twistMotion=Free
 }
 {bone1Name=dyn_jacketSide_01_r, bone2Name=dyn_jacketSide_mstr_r, swing1LimitDegrees=35, swing2LimitDegrees=35, twistLimitDegrees=15}
 {bone1Name=dyn_jacketSide_02_r, bone2Name=dyn_jacketSide_01_r, swing1LimitDegrees=35, swing2LimitDegrees=35, twistLimitDegrees=15}
 {bone1Name=dyn_jacketSide_03_r, bone2Name=dyn_jacketSide_02_r, swing1LimitDegrees=35, swing2LimitDegrees=35, twistLimitDegrees=15}
 {bone1Name=dyn_jacketSide_04_r, bone2Name=dyn_jacketSide_03_r, swing1LimitDegrees=35, swing2LimitDegrees=35, twistLimitDegrees=15}
 {
  bone1Name=dyn_jacketFront_mstr_r
  bone2Name=pelvis
  swing1LimitDegrees=45
  swing1Motion=Free
  swing2LimitDegrees=45
  swing2Motion=Free
  twistLimitDegrees=45
  twistMotion=Free
 }
 {bone1Name=dyn_jacketFront_01_r, bone2Name=dyn_jacketFront_mstr_r, swing1LimitDegrees=35, swing2LimitDegrees=35, twistLimitDegrees=15}
 {bone1Name=dyn_jacketFront_02_r, bone2Name=dyn_jacketFront_01_r, swing1LimitDegrees=35, swing2LimitDegrees=35, twistLimitDegrees=15}
 {bone1Name=dyn_jacketFront_03_r, bone2Name=dyn_jacketFront_02_r, swing1LimitDegrees=35, swing2LimitDegrees=35, twistLimitDegrees=15}
 {bone1Name=dyn_jacketFront_04_r, bone2Name=dyn_jacketFront_03_r, swing1LimitDegrees=35, swing2LimitDegrees=35, twistLimitDegrees=15}
 {
  bone1Name=dyn_jacketFront_mstr
  bone2Name=pelvis
  swing1LimitDegrees=45
  swing1Motion=Free
  swing2LimitDegrees=45
  swing2Motion=Free
  twistLimitDegrees=45
  twistMotion=Free
 }
 {bone1Name=dyn_jacketFront_01, bone2Name=dyn_jacketFront_mstr, swing1LimitDegrees=35, swing2LimitDegrees=35, twistLimitDegrees=15}
 {bone1Name=dyn_jacketFront_02, bone2Name=dyn_jacketFront_01, swing1LimitDegrees=35, swing2LimitDegrees=35, twistLimitDegrees=15}
 {bone1Name=dyn_jacketFront_03, bone2Name=dyn_jacketFront_02, swing1LimitDegrees=15, swing2LimitDegrees=35, twistLimitDegrees=15}
 {bone1Name=dyn_jacketFront_04, bone2Name=dyn_jacketFront_03, swing1LimitDegrees=15, swing2LimitDegrees=35, twistLimitDegrees=15}
 {
  bone1Name=dyn_jacketBack_03_l
  bone2Name=dyn_jacketBack_03_r
  swing1LimitDegrees=45
  swing1Motion=Free
  swing2LimitDegrees=45
  swing2Motion=Free
  twistLimitDegrees=45
  twistMotion=Free
 }
 {
  bone1Name=dyn_jacketBack_04_l
  bone2Name=dyn_jacketBack_04_r
  swing1LimitDegrees=45
  swing1Motion=Free
  swing2LimitDegrees=45
  swing2Motion=Free
  twistLimitDegrees=45
  twistMotion=Free
 }
 {
  bone1Name=dyn_jacketSide_04_l
  bone2Name=dyn_jacketBack_04_l
  swing1LimitDegrees=45
  swing1Motion=Free
  swing2LimitDegrees=45
  swing2Motion=Free
  twistLimitDegrees=45
  twistMotion=Free
 }
 {
  bone1Name=dyn_jacketFront_04_l
  bone2Name=dyn_jacketSide_04_l
  swing1LimitDegrees=45
  swing1Motion=Free
  swing2LimitDegrees=45
  swing2Motion=Free
  twistLimitDegrees=45
  twistMotion=Free
 }
 {
  bone1Name=dyn_jacketFront_03_r
  bone2Name=dyn_jacketSide_03_r
  swing1LimitDegrees=45
  swing1Motion=Free
  swing2LimitDegrees=45
  swing2Motion=Free
  twistLimitDegrees=45
  twistMotion=Free
 }
 {
  bone1Name=dyn_jacketFront_04
  bone2Name=dyn_jacketFront_04_r
  swing1LimitDegrees=45
  swing1Motion=Free
  swing2LimitDegrees=45
  swing2Motion=Free
  twistLimitDegrees=45
  twistMotion=Free
 }
 {
  bone1Name=dyn_jacketFront_03
  bone2Name=dyn_jacketFront_03_r
  swing1LimitDegrees=45
  swing1Motion=Free
  swing2LimitDegrees=45
  swing2Motion=Free
  twistLimitDegrees=45
  twistMotion=Free
 }
 {
  bone1Name=dyn_jacketFront_04
  bone2Name=dyn_jacketFront_03
  swing1LimitDegrees=45
  swing1Motion=Free
  swing2LimitDegrees=45
  swing2Motion=Free
  twistLimitDegrees=45
  twistMotion=Free
 }
 {
  bone1Name=dyn_jacketBack_04_r
  bone2Name=dyn_jacketSide_04_r
  swing1LimitDegrees=45
  swing1Motion=Free
  swing2LimitDegrees=45
  swing2Motion=Free
  twistLimitDegrees=45
  twistMotion=Free
 }
]
