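# Physics asset for the chains mesh of the female medium Ruins player body, covering a dangling arm hook chain.
# A master bone constrained to spine_05 anchors eight serially linked capsules with tighter twist than swing limits, ending in a larger capsule.
# Capsules on pelvis, spine, right thigh and right upper arm approximate the nearby body.
# Spine, upper arm and master joints rotate freely, while the thigh joint keeps angular limits.
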PhysicsAsset /BRCosmetics/Characters/Player/Female/Medium/Bodies/F_MED_Ruins/Meshes/F_MED_Ruins_chains_Physics
bodies {
 dyn_arm_hook_01 [{shapeType=Capsule, length=1.5, radius=1.5, shapeName=dyn_arm_hook_01_capsule, center=<2 0 0>, rotation=<rot 90 0 0>}]
 dyn_arm_hook_02 [{shapeType=Capsule, length=1.5, radius=1.5, shapeName=dyn_arm_hook_02_capsule, center=<2 0 0>, rotation=<rot 90 0 0>}]
 dyn_arm_hook_03 [{shapeType=Capsule, length=1.5, radius=1.5, shapeName=dyn_arm_hook_03_capsule, center=<2 0 0>, rotation=<rot 90 0 0>}]
 dyn_arm_hook_04 [{shapeType=Capsule, length=1.5, radius=1.5, shapeName=dyn_arm_hook_04_capsule, center=<2 0 0>, rotation=<rot 90 0 0>}]
 dyn_arm_hook_05 [{shapeType=Capsule, length=1.5, radius=1.5, shapeName=dyn_arm_hook_05_capsule, center=<2 0 0>, rotation=<rot 90 0 0>}]
 dyn_arm_hook_06 [{shapeType=Capsule, length=1.5, radius=1.5, shapeName=dyn_arm_hook_06_capsule, center=<2 0 0>, rotation=<rot 90 0 0>}]
 dyn_arm_hook_07 [{shapeType=Capsule, length=1.5, radius=1.5, shapeName=dyn_arm_hook_07_capsule, center=<2 0 0>, rotation=<rot 90 0 0>}]
 dyn_arm_hook_08 [{shapeType=Capsule, length=6.3256, radius=2.8536, shapeName=dyn_arm_hook_08_capsule, center=<6.8361 0 -0.5>, rotation=<rot 90 0 0>}]
 dyn_arm_hook_MSTR [{shapeType=Capsule, length=0.505, radius=0.505, shapeName=dyn_arm_hook_MSTR_capsule, rotation=<rot 0 -90 0>}]
 pelvis [
  {shapeType=Capsule, length=5.4625, radius=9.8051, center=<-0.9227 -1.8845 0>}
  {shapeType=Capsule, length=5.79, radius=8.4702, shapeName=pelvis_capsule, center=<-3.9978 -4.0638 0>}
 ]
 spine_01 [{shapeType=Capsule, length=4.88, radius=9.505, shapeName=spine_01_capsule, center=<-2.891 0 0>}]
 spine_02 [{shapeType=Capsule, length=0.375, radius=7.4622, shapeName=spine_02_capsule}]
 spine_03 [{shapeType=Capsule, length=4.755, radius=9.38, shapeName=spine_03_capsule}]
 spine_04 [{shapeType=Capsule, length=1.625, radius=12.0095, shapeName=spine_04_capsule}]
 spine_05 [{shapeType=Capsule, length=5.4751, radius=11.3851, shapeName=spine_05_capsule}]
 thigh_r [{shapeType=Capsule, length=24.0594, radius=11.9406, center=<21.4108 0.5413 -3.2123>, rotation=<rot 80 180 -180>}]
 upperarm_r [{shapeType=Capsule, length=19.6307, radius=5.1387, shapeName=upperarm_r_capsule, center=<-14.1504 -0.5694 1.563>, rotation=<rot 90 0 0>}]
}
constraints [
 {bone1Name=thigh_r, bone2Name=pelvis, swing1LimitDegrees=45, swing2LimitDegrees=45, twistLimitDegrees=45}
 {
  bone1Name=spine_01
  bone2Name=pelvis
  swing1LimitDegrees=45
  swing1Motion=Free
  swing2LimitDegrees=45
  swing2Motion=Free
  twistLimitDegrees=45
  twistMotion=Free
 }
 {
  bone1Name=spine_02
  bone2Name=spine_01
  swing1LimitDegrees=45
  swing1Motion=Free
  swing2LimitDegrees=45
  swing2Motion=Free
  twistLimitDegrees=45
  twistMotion=Free
 }
 {
  bone1Name=spine_03
  bone2Name=spine_02
  swing1LimitDegrees=45
  swing1Motion=Free
  swing2LimitDegrees=45
  swing2Motion=Free
  twistLimitDegrees=45
  twistMotion=Free
 }
 {
  bone1Name=spine_04
  bone2Name=spine_03
  swing1LimitDegrees=45
  swing1Motion=Free
  swing2LimitDegrees=45
  swing2Motion=Free
  twistLimitDegrees=45
  twistMotion=Free
 }
 {
  bone1Name=spine_05
  bone2Name=spine_04
  swing1LimitDegrees=45
  swing1Motion=Free
  swing2LimitDegrees=45
  swing2Motion=Free
  twistLimitDegrees=45
  twistMotion=Free
 }
 {bone1Name=dyn_arm_hook_08, bone2Name=dyn_arm_hook_07, swing1LimitDegrees=45, swing2LimitDegrees=45, twistLimitDegrees=15}
 {bone1Name=dyn_arm_hook_01, bone2Name=dyn_arm_hook_MSTR, swing1LimitDegrees=45, swing2LimitDegrees=45, twistLimitDegrees=15}
 {bone1Name=dyn_arm_hook_02, bone2Name=dyn_arm_hook_01, swing1LimitDegrees=45, swing2LimitDegrees=45, twistLimitDegrees=15}
 {bone1Name=dyn_arm_hook_03, bone2Name=dyn_arm_hook_02, swing1LimitDegrees=45, swing2LimitDegrees=45, twistLimitDegrees=15}
 {bone1Name=dyn_arm_hook_04, bone2Name=dyn_arm_hook_03, swing1LimitDegrees=45, swing2LimitDegrees=45, twistLimitDegrees=15}
 {bone1Name=dyn_arm_hook_05, bone2Name=dyn_arm_hook_04, swing1LimitDegrees=45, swing2LimitDegrees=45, twistLimitDegrees=15}
 {bone1Name=dyn_arm_hook_06, bone2Name=dyn_arm_hook_05, swing1LimitDegrees=45, swing2LimitDegrees=45, twistLimitDegrees=15}
 {bone1Name=dyn_arm_hook_07, bone2Name=dyn_arm_hook_06, swing1LimitDegrees=45, swing2LimitDegrees=45, twistLimitDegrees=15}
 {
  bone1Name=dyn_arm_hook_MSTR
  bone2Name=spine_05
  swing1LimitDegrees=45
  swing1Motion=Free
  swing2LimitDegrees=45
  swing2Motion=Free
  twistLimitDegrees=45
  twistMotion=Free
 }
 {
  bone1Name=upperarm_r
  bone2Name=spine_05
  swing1LimitDegrees=45
  swing1Motion=Free
  swing2LimitDegrees=45
  swing2Motion=Free
  twistLimitDegrees=45
  twistMotion=Free
 }
]
