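# Physics asset for the skydive chains setup on the F_MED_Ruins female cosmetic body, covering dangling hook-chain, bag and jacket bones.
# An eight-link hook chain hangs from a master body on spine_05, with tighter twist limits between links.
# Bag and back/side jacket chains root at the pelvis with free angular motion and large uniform jacket capsules.
# Capsules on pelvis, spine, right thigh and right upper arm represent the body.
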PhysicsAsset /BRCosmetics/Characters/Player/Female/Medium/Bodies/F_MED_Ruins/Meshes/F_MED_Ruins_chains_skydive_Physics
bodies {
 dyn_arm_hook_01 [{shapeType=Capsule, length=1.5, radius=1.5, shapeName=dyn_arm_hook_01_capsule, center=<2 0 0>, rotation=<rot 90 0 0>}]
 dyn_arm_hook_02 [{shapeType=Capsule, length=1.5, radius=1.5, shapeName=dyn_arm_hook_02_capsule, center=<2 0 0>, rotation=<rot 90 0 0>}]
 dyn_arm_hook_03 [{shapeType=Capsule, length=1.5, radius=1.5, shapeName=dyn_arm_hook_03_capsule, center=<2 0 0>, rotation=<rot 90 0 0>}]
 dyn_arm_hook_04 [{shapeType=Capsule, length=1.5, radius=1.5, shapeName=dyn_arm_hook_04_capsule, center=<2 0 0>, rotation=<rot 90 0 0>}]
 dyn_arm_hook_05 [{shapeType=Capsule, length=1.5, radius=1.5, shapeName=dyn_arm_hook_05_capsule, center=<2 0 0>, rotation=<rot 90 0 0>}]
 dyn_arm_hook_06 [{shapeType=Capsule, length=1.5, radius=1.5, shapeName=dyn_arm_hook_06_capsule, center=<2 0 0>, rotation=<rot 90 0 0>}]
 dyn_arm_hook_07 [{shapeType=Capsule, length=1.5, radius=1.5, shapeName=dyn_arm_hook_07_capsule, center=<2 0 0>, rotation=<rot 90 0 0>}]
 dyn_arm_hook_08 [{shapeType=Capsule, length=6.3256, radius=2.8536, shapeName=dyn_arm_hook_08_capsule, center=<6.8361 0 -0.5>, rotation=<rot 90 0 0>}]
 dyn_arm_hook_MSTR [{shapeType=Capsule, length=0.505, radius=0.505, shapeName=dyn_arm_hook_MSTR_capsule, rotation=<rot 0 -90 0>}]
 dyn_bag_01 [{
   shapeType=Capsule
   length=3.6603
   radius=4.3397
   shapeName=dyn_bag_01_capsule
   center=<4.7997 0.3139 0.4006>
   rotation=<rot 65.138 -176.9247 167.8455>
  }]
 dyn_bag_02 [{shapeType=Capsule, length=2.4794, radius=5.5206, shapeName=dyn_bag_02_capsule, center=<4.583 2.6154 1.8373>, rotation=<rot 60 179.9998 165>}]
 dyn_jacketBack_01_l [{shapeType=Capsule, length=9, radius=9, shapeName=dyn_jacketBack_01_l_capsule, center=<5 0 -8>, rotation=<rot 90 0 0>}]
 dyn_jacketBack_01_r [{shapeType=Capsule, length=9, radius=9, shapeName=dyn_jacketBack_01_r_capsule, center=<5 0 -8>, rotation=<rot 90 0 0>}]
 dyn_jacketBack_02_l [{shapeType=Capsule, length=9, radius=9, shapeName=dyn_jacketBack_02_l_capsule, center=<5 0 -8>, rotation=<rot 90 0 0>}]
 dyn_jacketBack_02_r [{shapeType=Capsule, length=9, radius=9, shapeName=dyn_jacketBack_02_r_capsule, center=<5 0 -8>, rotation=<rot 90 0 0>}]
 dyn_jacketBack_03_l [{shapeType=Capsule, length=9, radius=9, shapeName=dyn_jacketBack_03_l_capsule, center=<5 0 -8>, rotation=<rot 90 0 0>}]
 dyn_jacketBack_03_r [{shapeType=Capsule, length=9, radius=9, shapeName=dyn_jacketBack_03_r_capsule, center=<5 0 -8>, rotation=<rot 90 0 0>}]
 dyn_jacketBack_04_l [{shapeType=Capsule, length=9, radius=9, shapeName=dyn_jacketBack_04_l_capsule, center=<5 0 -8>, rotation=<rot 90 0 0>}]
 dyn_jacketBack_04_r [{shapeType=Capsule, length=9, radius=9, shapeName=dyn_jacketBack_04_r_capsule, center=<5 0 -8>, rotation=<rot 90 0 0>}]
 dyn_jacketBack_05_l [{shapeType=Capsule, length=9, radius=9, shapeName=dyn_jacketBack_05_l_capsule, center=<5 0 -8>, rotation=<rot 90 0 0>}]
 dyn_jacketSide_01_l [{shapeType=Capsule, length=9, radius=9, shapeName=dyn_jacketSide_01_l_capsule, center=<5 0 -8>, rotation=<rot 90 0 0>}]
 dyn_jacketSide_01_r [{shapeType=Capsule, length=9, radius=9, shapeName=dyn_jacketSide_01_r_capsule, center=<5 0 -8>, rotation=<rot 90 0 0>}]
 dyn_jacketSide_02_l [{shapeType=Capsule, length=9, radius=9, shapeName=dyn_jacketSide_02_l_capsule, center=<5 0 -8>, rotation=<rot 90 0 0>}]
 dyn_jacketSide_02_r [{shapeType=Capsule, length=9, radius=9, shapeName=dyn_jacketSide_02_r_capsule, center=<5 0 -8>, rotation=<rot 90 0 0>}]
 dyn_jacketSide_03_l [{shapeType=Capsule, length=9, radius=9, shapeName=dyn_jacketSide_03_l_capsule, center=<5 0 -8>, rotation=<rot 90 0 0>}]
 dyn_jacketSide_03_r [{shapeType=Capsule, length=9, radius=9, shapeName=dyn_jacketSide_03_r_capsule, center=<5 0 -8>, rotation=<rot 90 0 0>}]
 dyn_jacketSide_04_l [{shapeType=Capsule, length=9, radius=9, shapeName=dyn_jacketSide_04_l_capsule, center=<5 0 -8>, rotation=<rot 90 0 0>}]
 dyn_jacketSide_04_r [{shapeType=Capsule, length=9, radius=9, shapeName=dyn_jacketSide_04_r_capsule, center=<5 0 -8>, rotation=<rot 90 0 0>}]
 dyn_jacketSide_05_l [{shapeType=Capsule, length=9, radius=9, shapeName=dyn_jacketSide_05_l_capsule, center=<5 0 -8>, rotation=<rot 90 0 0>}]
 pelvis [
  {shapeType=Capsule, length=5.4625, radius=9.8051, center=<-0.9227 -1.8845 0>}
  {shapeType=Capsule, length=5.79, radius=8.4702, shapeName=pelvis_capsule, center=<-3.9978 -4.0638 0>}
 ]
 spine_01 [{shapeType=Capsule, length=4.88, radius=9.505, shapeName=spine_01_capsule, center=<-2.891 0 0>}]
 spine_02 [{shapeType=Capsule, length=0.375, radius=7.4622, shapeName=spine_02_capsule}]
 spine_03 [{shapeType=Capsule, length=4.755, radius=9.38, shapeName=spine_03_capsule}]
 spine_04 [{shapeType=Capsule, length=1.625, radius=12.0095, shapeName=spine_04_capsule}]
 spine_05 [{shapeType=Capsule, length=5.4751, radius=11.3851, shapeName=spine_05_capsule}]
 thigh_r [{shapeType=Capsule, length=24.0594, radius=11.9406, center=<21.4108 0.5413 -3.2123>, rotation=<rot 80 180 -180>}]
 upperarm_r [{shapeType=Capsule, length=19.6307, radius=5.1387, shapeName=upperarm_r_capsule, center=<-14.1504 -0.5694 1.563>, rotation=<rot 90 0 0>}]
}
constraints [
 {bone1Name=thigh_r, bone2Name=pelvis, swing1LimitDegrees=45, swing2LimitDegrees=45, twistLimitDegrees=45}
 {
  bone1Name=spine_01
  bone2Name=pelvis
  swing1LimitDegrees=45
  swing1Motion=Free
  swing2LimitDegrees=45
  swing2Motion=Free
  twistLimitDegrees=45
  twistMotion=Free
 }
 {
  bone1Name=spine_02
  bone2Name=spine_01
  swing1LimitDegrees=45
  swing1Motion=Free
  swing2LimitDegrees=45
  swing2Motion=Free
  twistLimitDegrees=45
  twistMotion=Free
 }
 {
  bone1Name=spine_03
  bone2Name=spine_02
  swing1LimitDegrees=45
  swing1Motion=Free
  swing2LimitDegrees=45
  swing2Motion=Free
  twistLimitDegrees=45
  twistMotion=Free
 }
 {
  bone1Name=spine_04
  bone2Name=spine_03
  swing1LimitDegrees=45
  swing1Motion=Free
  swing2LimitDegrees=45
  swing2Motion=Free
  twistLimitDegrees=45
  twistMotion=Free
 }
 {
  bone1Name=spine_05
  bone2Name=spine_04
  swing1LimitDegrees=45
  swing1Motion=Free
  swing2LimitDegrees=45
  swing2Motion=Free
  twistLimitDegrees=45
  twistMotion=Free
 }
 {bone1Name=dyn_arm_hook_08, bone2Name=dyn_arm_hook_07, swing1LimitDegrees=45, swing2LimitDegrees=45, twistLimitDegrees=15}
 {bone1Name=dyn_arm_hook_01, bone2Name=dyn_arm_hook_MSTR, swing1LimitDegrees=45, swing2LimitDegrees=45, twistLimitDegrees=15}
 {bone1Name=dyn_arm_hook_02, bone2Name=dyn_arm_hook_01, swing1LimitDegrees=45, swing2LimitDegrees=45, twistLimitDegrees=15}
 {bone1Name=dyn_arm_hook_03, bone2Name=dyn_arm_hook_02, swing1LimitDegrees=45, swing2LimitDegrees=45, twistLimitDegrees=15}
 {bone1Name=dyn_arm_hook_04, bone2Name=dyn_arm_hook_03, swing1LimitDegrees=45, swing2LimitDegrees=45, twistLimitDegrees=15}
 {bone1Name=dyn_arm_hook_05, bone2Name=dyn_arm_hook_04, swing1LimitDegrees=45, swing2LimitDegrees=45, twistLimitDegrees=15}
 {bone1Name=dyn_arm_hook_06, bone2Name=dyn_arm_hook_05, swing1LimitDegrees=45, swing2LimitDegrees=45, twistLimitDegrees=15}
 {bone1Name=dyn_arm_hook_07, bone2Name=dyn_arm_hook_06, swing1LimitDegrees=45, swing2LimitDegrees=45, twistLimitDegrees=15}
 {
  bone1Name=dyn_arm_hook_MSTR
  bone2Name=spine_05
  swing1LimitDegrees=45
  swing1Motion=Free
  swing2LimitDegrees=45
  swing2Motion=Free
  twistLimitDegrees=45
  twistMotion=Free
 }
 {
  bone1Name=upperarm_r
  bone2Name=spine_05
  swing1LimitDegrees=45
  swing1Motion=Free
  swing2LimitDegrees=45
  swing2Motion=Free
  twistLimitDegrees=45
  twistMotion=Free
 }
 {
  bone1Name=dyn_jacketSide_04_r
  bone2Name=dyn_jacketSide_03_r
  swing1LimitDegrees=45
  swing1Motion=Free
  swing2LimitDegrees=45
  swing2Motion=Free
  twistLimitDegrees=45
  twistMotion=Free
 }
 {
  bone1Name=dyn_jacketSide_01_l
  bone2Name=pelvis
  swing1LimitDegrees=45
  swing1Motion=Free
  swing2LimitDegrees=45
  swing2Motion=Free
  twistLimitDegrees=45
  twistMotion=Free
 }
 {
  bone1Name=dyn_jacketSide_02_l
  bone2Name=dyn_jacketSide_01_l
  swing1LimitDegrees=45
  swing1Motion=Free
  swing2LimitDegrees=45
  swing2Motion=Free
  twistLimitDegrees=45
  twistMotion=Free
 }
 {
  bone1Name=dyn_jacketSide_03_l
  bone2Name=dyn_jacketSide_02_l
  swing1LimitDegrees=45
  swing1Motion=Free
  swing2LimitDegrees=45
  swing2Motion=Free
  twistLimitDegrees=45
  twistMotion=Free
 }
 {
  bone1Name=dyn_jacketSide_04_l
  bone2Name=dyn_jacketSide_03_l
  swing1LimitDegrees=45
  swing1Motion=Free
  swing2LimitDegrees=45
  swing2Motion=Free
  twistLimitDegrees=45
  twistMotion=Free
 }
 {
  bone1Name=dyn_jacketSide_05_l
  bone2Name=dyn_jacketSide_04_l
  swing1LimitDegrees=45
  swing1Motion=Free
  swing2LimitDegrees=45
  swing2Motion=Free
  twistLimitDegrees=45
  twistMotion=Free
 }
 {
  bone1Name=dyn_jacketSide_01_r
  bone2Name=pelvis
  swing1LimitDegrees=45
  swing1Motion=Free
  swing2LimitDegrees=45
  swing2Motion=Free
  twistLimitDegrees=45
  twistMotion=Free
 }
 {
  bone1Name=dyn_jacketSide_02_r
  bone2Name=dyn_jacketSide_01_r
  swing1LimitDegrees=45
  swing1Motion=Free
  swing2LimitDegrees=45
  swing2Motion=Free
  twistLimitDegrees=45
  twistMotion=Free
 }
 {
  bone1Name=dyn_jacketSide_03_r
  bone2Name=dyn_jacketSide_02_r
  swing1LimitDegrees=45
  swing1Motion=Free
  swing2LimitDegrees=45
  swing2Motion=Free
  twistLimitDegrees=45
  twistMotion=Free
 }
 {
  bone1Name=dyn_jacketBack_04_r
  bone2Name=dyn_jacketBack_03_r
  swing1LimitDegrees=45
  swing1Motion=Free
  swing2LimitDegrees=45
  swing2Motion=Free
  twistLimitDegrees=45
  twistMotion=Free
 }
 {
  bone1Name=dyn_jacketBack_01_l
  bone2Name=pelvis
  swing1LimitDegrees=45
  swing1Motion=Free
  swing2LimitDegrees=45
  swing2Motion=Free
  twistLimitDegrees=45
  twistMotion=Free
 }
 {
  bone1Name=dyn_jacketBack_02_l
  bone2Name=dyn_jacketBack_01_l
  swing1LimitDegrees=45
  swing1Motion=Free
  swing2LimitDegrees=45
  swing2Motion=Free
  twistLimitDegrees=45
  twistMotion=Free
 }
 {
  bone1Name=dyn_jacketBack_03_l
  bone2Name=dyn_jacketBack_02_l
  swing1LimitDegrees=45
  swing1Motion=Free
  swing2LimitDegrees=45
  swing2Motion=Free
  twistLimitDegrees=45
  twistMotion=Free
 }
 {
  bone1Name=dyn_jacketBack_04_l
  bone2Name=dyn_jacketBack_03_l
  swing1LimitDegrees=45
  swing1Motion=Free
  swing2LimitDegrees=45
  swing2Motion=Free
  twistLimitDegrees=45
  twistMotion=Free
 }
 {
  bone1Name=dyn_jacketBack_05_l
  bone2Name=dyn_jacketBack_04_l
  swing1LimitDegrees=45
  swing1Motion=Free
  swing2LimitDegrees=45
  swing2Motion=Free
  twistLimitDegrees=45
  twistMotion=Free
 }
 {
  bone1Name=dyn_jacketBack_01_r
  bone2Name=pelvis
  swing1LimitDegrees=45
  swing1Motion=Free
  swing2LimitDegrees=45
  swing2Motion=Free
  twistLimitDegrees=45
  twistMotion=Free
 }
 {
  bone1Name=dyn_jacketBack_02_r
  bone2Name=dyn_jacketBack_01_r
  swing1LimitDegrees=45
  swing1Motion=Free
  swing2LimitDegrees=45
  swing2Motion=Free
  twistLimitDegrees=45
  twistMotion=Free
 }
 {
  bone1Name=dyn_jacketBack_03_r
  bone2Name=dyn_jacketBack_02_r
  swing1LimitDegrees=45
  swing1Motion=Free
  swing2LimitDegrees=45
  swing2Motion=Free
  twistLimitDegrees=45
  twistMotion=Free
 }
 {
  bone1Name=dyn_bag_01
  bone2Name=pelvis
  swing1LimitDegrees=45
  swing1Motion=Free
  swing2LimitDegrees=45
  swing2Motion=Free
  twistLimitDegrees=45
  twistMotion=Free
 }
 {
  bone1Name=dyn_bag_02
  bone2Name=dyn_bag_01
  swing1LimitDegrees=45
  swing1Motion=Free
  swing2LimitDegrees=45
  swing2Motion=Free
  twistLimitDegrees=45
  twistMotion=Free
 }
]
