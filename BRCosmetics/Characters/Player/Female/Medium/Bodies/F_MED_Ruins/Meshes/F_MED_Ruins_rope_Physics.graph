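# Physics asset for the female medium Ruins cosmetic body, simulating its dangling jacket panels, bag and belt ropes.
# Capsules on pelvis, right thigh and calf form the body, while box master bodies freely jointed to the pelvis root capsule chains for front and right-side jacket flaps.
# The bag hangs from the right front jacket master with free joints, and its master carries two rope chains with limited swing and tighter twist.
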PhysicsAsset /BRCosmetics/Characters/Player/Female/Medium/Bodies/F_MED_Ruins/Meshes/F_MED_Ruins_rope_Physics
bodies {
 calf_r [{shapeType=Capsule, length=35.7852, radius=12.9234, center=<26.3559 0.7114 -0.6693>, rotation=<rot -0.3385 89.5329 -88.4634>}]
 dyn_bag_01 [{shapeType=Capsule, length=3.1812, radius=3.2265, shapeName=dyn_bag_01_capsule, center=<4.848 0.7523 1.8082>, rotation=<rot 75 179.9998 170>}]
 dyn_bag_02 [{
   shapeType=Capsule
   length=1.5355
   radius=5.0261
   shapeName=dyn_bag_02_capsule
   center=<4.0302 2.5914 3.3449>
   rotation=<rot 0 -12.0399 -81.8202>
  }]
 dyn_bag_MSTR [{shapeType=Capsule, length=0.505, radius=0.505, shapeName=dyn_bag_MSTR_capsule, rotation=<rot 0 -90 0>}]
 dyn_belt_rope_a_01 [{shapeType=Capsule, length=3.8737, radius=1.4416, shapeName=dyn_belt_rope_a_01_capsule, center=<4 0 0>, rotation=<rot 90 0 0>}]
 dyn_belt_rope_a_02 [{shapeType=Capsule, length=6.8748, radius=1.165, shapeName=dyn_belt_rope_a_02_capsule, center=<6 0 0>, rotation=<rot 90 0 0>}]
 dyn_belt_rope_a_MSTR [{shapeType=Capsule, length=0.505, radius=0.505, shapeName=dyn_belt_rope_a_MSTR_capsule, rotation=<rot 0 -90 0>}]
 dyn_belt_rope_b_01 [{shapeType=Capsule, length=4.0371, radius=1.1354, shapeName=dyn_belt_rope_b_01_capsule, center=<3 0 0>, rotation=<rot 90 0 0>}]
 dyn_belt_rope_b_02 [{shapeType=Capsule, length=2.7438, radius=2.8505, shapeName=dyn_belt_rope_b_02_capsule, center=<5 0 0>, rotation=<rot 90 0 0>}]
 dyn_belt_rope_b_03 [{shapeType=Capsule, length=6.1978, radius=2.7472, shapeName=dyn_belt_rope_b_03_capsule, center=<7.3457 0 0>, rotation=<rot 90 0 0>}]
 dyn_belt_rope_b_MSTR [{shapeType=Capsule, length=0.505, radius=0.505, shapeName=dyn_belt_rope_b_MSTR_capsule, rotation=<rot 0 -90 0>}]
 dyn_jacketFront_01 [{shapeType=Capsule, length=6.375, radius=8, shapeName=dyn_jacketFront_01_capsule, center=<5 0 -8>, rotation=<rot 90 0 0>}]
 dyn_jacketFront_01_r [{shapeType=Capsule, length=6.375, radius=8, shapeName=dyn_jacketFront_01_r_capsule, center=<5 0 -9>, rotation=<rot 90 0 0>}]
 dyn_jacketFront_02 [{shapeType=Capsule, length=6.125, radius=8.25, shapeName=dyn_jacketFront_02_capsule, center=<5 0 -7>, rotation=<rot 90 0 0>}]
 dyn_jacketFront_02_r [{shapeType=Capsule, length=6.125, radius=8.25, shapeName=dyn_jacketFront_02_r_capsule, center=<5 0 -8.5>, rotation=<rot 90 0 0>}]
 dyn_jacketFront_03 [{shapeType=Capsule, length=5.75, radius=8.625, shapeName=dyn_jacketFront_03_capsule, center=<5 0 -7>, rotation=<rot 90 0 0>}]
 dyn_jacketFront_03_r [{shapeType=Capsule, length=5.75, radius=8.625, shapeName=dyn_jacketFront_03_r_capsule, center=<5 0 -8>, rotation=<rot 90 0 0>}]
 dyn_jacketFront_04 [{shapeType=Capsule, length=6.375, radius=8, shapeName=dyn_jacketFront_04_capsule, center=<5 0 -7>, rotation=<rot 90 0 0>}]
 dyn_jacketFront_04_r [{shapeType=Capsule, length=6.375, radius=8, shapeName=dyn_jacketFront_04_r_capsule, center=<5 0 -8>, rotation=<rot 90 0 0>}]
 dyn_jacketFront_mstr [{shapeType=Box, extentX=15, extentY=15, extentZ=2, shapeName=dyn_jacketFront_mstr_box, center=<5 0 0>}]
 dyn_jacketFront_mstr_r [{
   shapeType=Box
   extentX=17.25
   extentY=14.25
   extentZ=2
   shapeName=dyn_jacketFront_mstr_r_box
   center=<5 0 0>
   rotation=<rot 0.0002 -0.0001 -10>
  }]
 dyn_jacketSide_01_r [{shapeType=Capsule, length=6.375, radius=8, shapeName=dyn_jacketSide_01_r_capsule, center=<5 0 -9>, rotation=<rot 90 0 0>}]
 dyn_jacketSide_02_r [{shapeType=Capsule, length=6.375, radius=8, shapeName=dyn_jacketSide_02_r_capsule, center=<5 -1.0138 -8>, rotation=<rot 90 0 0>}]
 dyn_jacketSide_03_r [{shapeType=Capsule, length=5.625, radius=8.75, shapeName=dyn_jacketSide_03_r_capsule, center=<5 -1.0138 -8>, rotation=<rot 90 0 0>}]
 dyn_jacketSide_04_r [{shapeType=Capsule, length=5.25, radius=9.125, shapeName=dyn_jacketSide_04_r_capsule, center=<5 -1.0138 -8>, rotation=<rot 90 0 0>}]
 dyn_jacketSide_mstr_r [{shapeType=Box, extentX=15, extentY=15, extentZ=2, shapeName=dyn_jacketSide_mstr_r_box, center=<5 0 0>}]
 pelvis [
  {shapeType=Capsule, length=7.7125, radius=8.4301, center=<-0.9227 -1.2812 0>}
  {shapeType=Capsule, length=10.79, radius=7.4702, shapeName=pelvis_capsule, center=<-3.9978 -3.5061 0>}
 ]
 thigh_r [{shapeType=Capsule, length=25.375, radius=9.625, center=<21.0533 0.5413 -3.2804>, rotation=<rot 90 89.3553 94.3553>}]
}
constraints [
 {bone1Name=thigh_r, bone2Name=pelvis, swing1LimitDegrees=45, swing2LimitDegrees=45, twistLimitDegrees=45}
 {bone1Name=calf_r, bone2Name=thigh_r, swing1LimitDegrees=45, swing2LimitDegrees=45, twistLimitDegrees=45}
 {
  bone1Name=dyn_jacketSide_mstr_r
  bone2Name=pelvis
  swing1LimitDegrees=45
  swing1Motion=Free
  swing2LimitDegrees=45
  swing2Motion=Free
  twistLimitDegrees=45
  twistMotion=Free
 }
 {bone1Name=dyn_jacketSide_01_r, bone2Name=dyn_jacketSide_mstr_r, swing1LimitDegrees=35, swing2LimitDegrees=35, twistLimitDegrees=15}
 {bone1Name=dyn_jacketSide_02_r, bone2Name=dyn_jacketSide_01_r, swing1LimitDegrees=35, swing2LimitDegrees=35, twistLimitDegrees=15}
 {bone1Name=dyn_jacketSide_03_r, bone2Name=dyn_jacketSide_02_r, swing1LimitDegrees=35, swing2LimitDegrees=35, twistLimitDegrees=15}
 {bone1Name=dyn_jacketSide_04_r, bone2Name=dyn_jacketSide_03_r, swing1LimitDegrees=35, swing2LimitDegrees=35, twistLimitDegrees=15}
 {
  bone1Name=dyn_jacketFront_mstr_r
  bone2Name=pelvis
  swing1LimitDegrees=45
  swing1Motion=Free
  swing2LimitDegrees=45
  swing2Motion=Free
  twistLimitDegrees=45
  twistMotion=Free
 }
 {bone1Name=dyn_jacketFront_01_r, bone2Name=dyn_jacketFront_mstr_r, swing1LimitDegrees=35, swing2LimitDegrees=35, twistLimitDegrees=15}
 {bone1Name=dyn_jacketFront_02_r, bone2Name=dyn_jacketFront_01_r, swing1LimitDegrees=35, swing2LimitDegrees=35, twistLimitDegrees=15}
 {bone1Name=dyn_jacketFront_03_r, bone2Name=dyn_jacketFront_02_r, swing1LimitDegrees=35, swing2LimitDegrees=35, twistLimitDegrees=15}
 {bone1Name=dyn_jacketFront_04_r, bone2Name=dyn_jacketFront_03_r, swing1LimitDegrees=35, swing2LimitDegrees=35, twistLimitDegrees=15}
 {
  bone1Name=dyn_jacketFront_mstr
  bone2Name=pelvis
  swing1LimitDegrees=45
  swing1Motion=Free
  swing2LimitDegrees=45
  swing2Motion=Free
  twistLimitDegrees=45
  twistMotion=Free
 }
 {bone1Name=dyn_jacketFront_01, bone2Name=dyn_jacketFront_mstr, swing1LimitDegrees=35, swing2LimitDegrees=35, twistLimitDegrees=15}
 {bone1Name=dyn_jacketFront_02, bone2Name=dyn_jacketFront_01, swing1LimitDegrees=35, swing2LimitDegrees=35, twistLimitDegrees=15}
 {bone1Name=dyn_jacketFront_03, bone2Name=dyn_jacketFront_02, swing1LimitDegrees=35, swing2LimitDegrees=35, twistLimitDegrees=15}
 {bone1Name=dyn_jacketFront_04, bone2Name=dyn_jacketFront_03, swing1LimitDegrees=35, swing2LimitDegrees=35, twistLimitDegrees=15}
 {
  bone1Name=dyn_bag_MSTR
  bone2Name=dyn_jacketFront_mstr_r
  swing1LimitDegrees=45
  swing1Motion=Free
  swing2LimitDegrees=45
  swing2Motion=Free
  twistLimitDegrees=45
  twistMotion=Free
 }
 {
  bone1Name=dyn_bag_01
  bone2Name=dyn_bag_MSTR
  swing1LimitDegrees=45
  swing1Motion=Free
  swing2LimitDegrees=45
  swing2Motion=Free
  twistLimitDegrees=45
  twistMotion=Free
 }
 {
  bone1Name=dyn_bag_02
  bone2Name=dyn_bag_01
  swing1LimitDegrees=45
  swing1Motion=Free
  swing2LimitDegrees=45
  swing2Motion=Free
  twistLimitDegrees=45
  twistMotion=Free
 }
 {bone1Name=dyn_belt_rope_a_01, bone2Name=dyn_belt_rope_a_MSTR, swing1LimitDegrees=45, swing2LimitDegrees=45, twistLimitDegrees=15}
 {bone1Name=dyn_belt_rope_a_02, bone2Name=dyn_belt_rope_a_01, swing1LimitDegrees=45, swing2LimitDegrees=45, twistLimitDegrees=15}
 {bone1Name=dyn_belt_rope_b_01, bone2Name=dyn_belt_rope_b_MSTR, swing1LimitDegrees=45, swing2LimitDegrees=45, twistLimitDegrees=15}
 {bone1Name=dyn_belt_rope_b_02, bone2Name=dyn_belt_rope_b_01, swing1LimitDegrees=45, swing2LimitDegrees=45, twistLimitDegrees=15}
 {bone1Name=dyn_belt_rope_b_03, bone2Name=dyn_belt_rope_b_02, swing1LimitDegrees=45, swing2LimitDegrees=45, twistLimitDegrees=15}
 {
  bone1Name=dyn_belt_rope_b_MSTR
  bone2Name=dyn_bag_MSTR
  swing1LimitDegrees=45
  swing1Motion=Free
  swing2LimitDegrees=45
  swing2Motion=Free
  twistLimitDegrees=45
  twistMotion=Free
 }
 {
  bone1Name=dyn_belt_rope_a_MSTR
  bone2Name=dyn_bag_MSTR
  swing1LimitDegrees=45
  swing1Motion=Free
  swing2LimitDegrees=45
  swing2Motion=Free
  twistLimitDegrees=45
  twistMotion=Free
 }
]
