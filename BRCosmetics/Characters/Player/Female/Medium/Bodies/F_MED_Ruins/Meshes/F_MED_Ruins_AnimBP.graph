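# Post-process style AnimBP for the F_MED_Ruins female body cosmetic, adding secondary motion to clothing, chains and accessories.
# Copies the attached parent mesh's pose, then applies bone constraints, two Control Rigs and additive bone offsets driven by movement states.
# RigidBody nodes switch physics assets for front end and skydiving, with a separate rope simulation.
# AnimDynamics swings shackles, necklace, bag, flask and choker, plus arm hooks when bRigidBodyEnabled is off.
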AnimBlueprint /BRCosmetics/Characters/Player/Female/Medium/Bodies/F_MED_Ruins/Meshes/F_MED_Ruins_AnimBP
target_skeleton=&/F_MED_Ruins_Skeleton
graphs [
 AnimGraph {nodes [
   "Constraint - dyn_bag_MSTR (P)"(Alpha=1, ConstraintWeights_0=1) {
    alphaBoolBlend {blendOption=Linear}
    boneToModify {boneName=dyn_bag_MSTR}
    constraintSetup [{targetBone {boneName=thigh_r}}]
    constraintWeights [1]
   }
   n1 = "Constraint - dyn_belt_rope_a_MSTR (P)"(Alpha=1, ConstraintWeights_0=1) {
    alphaBoolBlend {blendOption=Linear}
    boneToModify {boneName=dyn_belt_rope_a_MSTR}
    constraintSetup [{targetBone {boneName=pelvis}}]
    constraintWeights [1]
   }
   "Constraint - dyn_belt_rope_b_MSTR (P)"(ComponentPose=@n1, Alpha=1, ConstraintWeights_0=1) {
    alphaBoolBlend {blendOption=Linear}
    boneToModify {boneName=dyn_belt_rope_b_MSTR}
    constraintSetup [{targetBone {boneName=pelvis}}]
    constraintWeights [1]
   }
   "Transform (Modify) Bone - Bone: dyn_jacketSide_mstr_l"(Alpha=1, Translation="0,0,0", Rotation="15,0,0.000", Scale="1,1,1") {
    rotationMode=BMM_Additive
    rotationSpace=BCS_BoneSpace
    alphaBoolBlend {blendOption=Linear}
    boneToModify {boneName=dyn_jacketSide_mstr_l}
    rotation {pitch=15}
   }
   "Transform (Modify) Bone - Bone: dyn_jacketSide_mstr_l"(Alpha=1, Translation="0,0,0", Rotation="15,0,0.000", Scale="1,1,1") {
    rotationMode=BMM_Additive
    rotationSpace=BCS_BoneSpace
    alphaBoolBlend {blendOption=Linear}
    boneToModify {boneName=dyn_jacketSide_mstr_l}
    rotation {pitch=15}
   }
   n2 = Get bIsInFrontEnd()
   n3 = Get GravityOverride()
   n4 = RigidBody!(
    ComponentPose=Local To Component(LocalPose=Use cached pose 'rban'())
    bAlphaBoolEnabled:bool=NOT Boolean(A=@n2)
    OverrideWorldGravity:Vector=@n3
    ExternalForce
   ) -> (Pose) {
    alphaInputType=Bool
    bClampLinearTranslationLimitToRefPose=true
    bOverrideWorldGravity=true
    overridePhysicsAsset=&/F_MED_Ruins_Physics
    simulationSpace=BaseBoneSpace
    alphaBoolBlend {blendInTime=0.1, blendOption=Linear, blendOutTime=0.1}
    baseBoneRef {boneName=root}
    componentAppliedLinearAccClamp=<3000 3000 3000>
    componentLinearAccScale=<0.3 0.3 0.5>
    componentLinearVelScale=<0.3 0.3 0.5>
    simSpaceSettings {
     maxAngularAcceleration=1000
     maxAngularVelocity=3
     maxLinearAcceleration=700
     maxLinearVelocity=500
     velocityScaleZ=0
     externalLinearDragV=<0.2 0.2>
    }
   }
   n5 = RigidBody(ComponentPose=@n4, bAlphaBoolEnabled=@n2, OverrideWorldGravity=@n3) {
    alphaInputType=Bool
    bClampLinearTranslationLimitToRefPose=true
    bOverrideWorldGravity=true
    overridePhysicsAsset=&/F_MED_Ruins_Frontend_Physics
    simulationSpace=BaseBoneSpace
    alphaBoolBlend {blendInTime=0.1, blendOption=Linear, blendOutTime=0.1}
    baseBoneRef {boneName=root}
    componentAppliedLinearAccClamp=<3000 3000 3000>
    componentLinearAccScale=<0.3 0.3 0.5>
    componentLinearVelScale=<0.3 0.3 0.5>
    simSpaceSettings {
     maxAngularAcceleration=1000
     maxAngularVelocity=3
     maxLinearAcceleration=700
     maxLinearVelocity=500
     velocityScaleZ=0
     externalLinearDragV=<0.2 0.2>
    }
   }
   n6 = Get bIsSkyDiving()
   n7 = Get GravityOverride_chain() -> (GravityOverride_chain:Vector)
   n8 = RigidBody!(
    ComponentPose=@n5
    bAlphaBoolEnabled=NOT Boolean(A=@n6)
    OverrideWorldGravity=@n7
    ExternalForce
    ComponentLinearAccScale:Vector=Get AccScale() -> (AccScale:Vector)
    ComponentLinearVelScale:Vector=Get VelScale() -> (VelScale:Vector)
   ) -> (Pose) {
    alphaInputType=Bool
    bClampLinearTranslationLimitToRefPose=true
    bOverrideWorldGravity=true
    overridePhysicsAsset=&/F_MED_Ruins_chains_Physics
    simulationSpace=BaseBoneSpace
    alphaBoolBlend {blendInTime=0.1, blendOption=Linear, blendOutTime=0.1}
    baseBoneRef {boneName=spine_02}
    componentAppliedLinearAccClamp=<1500 1500 1500>
    componentLinearAccScale=<0.25 0.25 0.25>
    componentLinearVelScale=<0.2 0.2 0.2>
    simSpaceSettings {
     maxAngularAcceleration=1000
     maxAngularVelocity=3
     maxLinearAcceleration=700
     maxLinearVelocity=500
     velocityScaleZ=0
     worldAlpha=1
     externalLinearDragV=<0.2 0.2>
    }
   }
   n9 = RigidBody(
    ComponentPose=@n8
    bAlphaBoolEnabled=@n6
    OverrideWorldGravity=@n7
    ComponentLinearAccScale="0.1,0.1,0.1"
    ComponentLinearVelScale="0.1,0.1,0.1"
   ) {
    alphaInputType=Bool
    bClampLinearTranslationLimitToRefPose=true
    bOverrideWorldGravity=true
    overridePhysicsAsset=&/F_MED_Ruins_chains_skydive_Physics
    simulationSpace=BaseBoneSpace
    alphaBoolBlend {blendInTime=0.1, blendOption=Linear, blendOutTime=0.1}
    baseBoneRef {boneName=pelvis}
    componentAppliedLinearAccClamp=<1500 1500 1500>
    componentLinearAccScale=<0.25 0.25 0.25>
    componentLinearVelScale=<0.2 0.2 0.2>
    simSpaceSettings {
     maxAngularAcceleration=1000
     maxAngularVelocity=3
     maxLinearAcceleration=700
     maxLinearVelocity=500
     velocityScaleZ=0
     externalLinearDragV=<0.2 0.2>
    }
   }
   n10 = NOT Boolean(A=Get bRigidBodyEnabled())
   n11 = Anim Dynamics - Bone: dyn_arm_hook_01!(ComponentPose=@n9, bAlphaBoolEnabled:bool=@n10, GravityOverride:Vector="0,0,-980", ExternalForce="0,0,0") -> (Pose) {
    alphaInputType=Bool
    angularDampingOverride=0.99
    angularSpringConstant=100
    bAngularSpring=true
    bOverrideAngularDamping=true
    bUseSphericalLimits=true
    simulationSpace=RootRelative
    alphaBoolBlend {blendOption=Linear}
    boundBone {boneName=dyn_arm_hook_01}
    componentAppliedLinearAccClamp=<3500 3500 3500>
    componentLinearAccScale=<0.8 0.8 0.8>
    componentLinearVelScale=<0.8 0.8 0.8>
    gravityOverride {z=-980}
    physicsBodyDefinitions [{
      collisionType=CustomSphere
      sphereCollisionRadius=4.5
      boundBone {boneName=dyn_arm_hook_01}
      boxExtents=<25 15 15>
      constraintSetup {angularLimitsMax=<1 35 35>, angularLimitsMin=<-1 -35 -35>, angularTarget=<1 0 0>}
      localJointOffset=<2 0 0>
     }]
    relativeSpaceBone {boneName=pelvis}
    sphericalLimits [
     {limitRadius=10, drivingBone {boneName=spine_02}, sphereLocalOffset=<0 0 0>}
     {limitRadius=10, drivingBone {boneName=spine_03}, sphereLocalOffset=<0 0 0>}
     {limitRadius=10, drivingBone {boneName=spine_04}, sphereLocalOffset=<0 0 0>}
     {limitRadius=6, drivingBone {boneName=upperarm_twist_02_r}, sphereLocalOffset=<0 -2 0>}
     {limitRadius=5, drivingBone {boneName=lowerarm_r}, sphereLocalOffset=<0 -2 0>}
    ]
   }
   n12 = Anim Dynamics - Bone: dyn_arm_hook_02!(ComponentPose=@n11, bAlphaBoolEnabled:bool=@n10, GravityOverride:Vector="0,0,-980", ExternalForce="0,0,0") -> (Pose) {
    alphaInputType=Bool
    angularDampingOverride=0.99
    angularSpringConstant=100
    bAngularSpring=true
    bOverrideAngularDamping=true
    bUseSphericalLimits=true
    simulationSpace=RootRelative
    alphaBoolBlend {blendOption=Linear}
    boundBone {boneName=dyn_arm_hook_02}
    componentAppliedLinearAccClamp=<3500 3500 3500>
    componentLinearAccScale=<0.8 0.8 0.8>
    componentLinearVelScale=<0.8 0.8 0.8>
    gravityOverride {z=-980}
    physicsBodyDefinitions [{
      collisionType=CustomSphere
      sphereCollisionRadius=4.5
      boundBone {boneName=dyn_arm_hook_02}
      boxExtents=<25 15 15>
      constraintSetup {angularLimitsMax=<1 35 35>, angularLimitsMin=<-1 -35 -35>, angularTarget=<1 0 0>}
      localJointOffset=<2 0 0>
     }]
    relativeSpaceBone {boneName=pelvis}
    sphericalLimits [
     {limitRadius=10, drivingBone {boneName=spine_02}, sphereLocalOffset=<0 0 0>}
     {limitRadius=10, drivingBone {boneName=spine_03}, sphereLocalOffset=<0 0 0>}
     {limitRadius=10, drivingBone {boneName=spine_04}, sphereLocalOffset=<0 0 0>}
     {limitRadius=6, drivingBone {boneName=upperarm_twist_02_r}, sphereLocalOffset=<0 -2 0>}
     {limitRadius=5, drivingBone {boneName=lowerarm_r}, sphereLocalOffset=<0 -2 0>}
    ]
   }
   n13 = Anim Dynamics - Bone: dyn_arm_hook_03!(ComponentPose=@n12, bAlphaBoolEnabled:bool=@n10, GravityOverride:Vector="0,0,-980", ExternalForce="0,0,0") -> (Pose) {
    alphaInputType=Bool
    angularDampingOverride=0.99
    angularSpringConstant=100
    bAngularSpring=true
    bOverrideAngularDamping=true
    bUseSphericalLimits=true
    simulationSpace=RootRelative
    alphaBoolBlend {blendOption=Linear}
    boundBone {boneName=dyn_arm_hook_03}
    componentAppliedLinearAccClamp=<3500 3500 3500>
    componentLinearAccScale=<0.8 0.8 0.8>
    componentLinearVelScale=<0.8 0.8 0.8>
    gravityOverride {z=-980}
    physicsBodyDefinitions [{
      collisionType=CustomSphere
      sphereCollisionRadius=4.5
      boundBone {boneName=dyn_arm_hook_03}
      boxExtents=<25 15 15>
      constraintSetup {angularLimitsMax=<1 35 35>, angularLimitsMin=<-1 -35 -35>, angularTarget=<1 0 0>}
      localJointOffset=<2 0 0>
     }]
    relativeSpaceBone {boneName=pelvis}
    sphericalLimits [
     {limitRadius=10, drivingBone {boneName=spine_02}, sphereLocalOffset=<0 0 0>}
     {limitRadius=10, drivingBone {boneName=spine_03}, sphereLocalOffset=<0 0 0>}
     {limitRadius=10, drivingBone {boneName=spine_04}, sphereLocalOffset=<0 0 0>}
     {limitRadius=6, drivingBone {boneName=upperarm_twist_02_r}, sphereLocalOffset=<0 -2 0>}
     {limitRadius=5, drivingBone {boneName=lowerarm_r}, sphereLocalOffset=<0 -2 0>}
    ]
   }
   n14 = Anim Dynamics - Bone: dyn_arm_hook_04!(ComponentPose=@n13, bAlphaBoolEnabled:bool=@n10, GravityOverride:Vector="0,0,-980", ExternalForce="0,0,0") -> (Pose) {
    alphaInputType=Bool
    angularDampingOverride=0.99
    angularSpringConstant=100
    bAngularSpring=true
    bOverrideAngularDamping=true
    bUseSphericalLimits=true
    simulationSpace=RootRelative
    alphaBoolBlend {blendOption=Linear}
    boundBone {boneName=dyn_arm_hook_04}
    componentAppliedLinearAccClamp=<3500 3500 3500>
    componentLinearAccScale=<0.8 0.8 0.8>
    componentLinearVelScale=<0.8 0.8 0.8>
    gravityOverride {z=-980}
    physicsBodyDefinitions [{
      collisionType=CustomSphere
      sphereCollisionRadius=4.5
      boundBone {boneName=dyn_arm_hook_04}
      boxExtents=<25 15 15>
      constraintSetup {angularLimitsMax=<1 35 35>, angularLimitsMin=<-1 -35 -35>, angularTarget=<1 0 0>}
      localJointOffset=<2 0 0>
     }]
    relativeSpaceBone {boneName=pelvis}
    sphericalLimits [
     {limitRadius=10, drivingBone {boneName=spine_02}, sphereLocalOffset=<0 0 0>}
     {limitRadius=10, drivingBone {boneName=spine_03}, sphereLocalOffset=<0 0 0>}
     {limitRadius=10, drivingBone {boneName=spine_04}, sphereLocalOffset=<0 0 0>}
     {limitRadius=6, drivingBone {boneName=upperarm_twist_02_r}, sphereLocalOffset=<0 -2 0>}
     {limitRadius=5, drivingBone {boneName=lowerarm_r}, sphereLocalOffset=<0 -2 0>}
    ]
   }
   n15 = Anim Dynamics - Bone: dyn_arm_hook_05!(ComponentPose=@n14, bAlphaBoolEnabled:bool=@n10, GravityOverride:Vector="0,0,-980", ExternalForce="0,0,0") -> (Pose) {
    alphaInputType=Bool
    angularDampingOverride=0.99
    angularSpringConstant=100
    bAngularSpring=true
    bOverrideAngularDamping=true
    bUseSphericalLimits=true
    simulationSpace=RootRelative
    alphaBoolBlend {blendOption=Linear}
    boundBone {boneName=dyn_arm_hook_05}
    componentAppliedLinearAccClamp=<3500 3500 3500>
    componentLinearAccScale=<0.8 0.8 0.8>
    componentLinearVelScale=<0.8 0.8 0.8>
    gravityOverride {z=-980}
    physicsBodyDefinitions [{
      collisionType=CustomSphere
      sphereCollisionRadius=4.5
      boundBone {boneName=dyn_arm_hook_05}
      boxExtents=<25 15 15>
      constraintSetup {angularLimitsMax=<1 35 35>, angularLimitsMin=<-1 -35 -35>, angularTarget=<1 0 0>}
      localJointOffset=<2 0 0>
     }]
    relativeSpaceBone {boneName=pelvis}
    sphericalLimits [
     {limitRadius=10, drivingBone {boneName=spine_02}, sphereLocalOffset=<0 0 0>}
     {limitRadius=10, drivingBone {boneName=spine_03}, sphereLocalOffset=<0 0 0>}
     {limitRadius=10, drivingBone {boneName=spine_04}, sphereLocalOffset=<0 0 0>}
     {limitRadius=6, drivingBone {boneName=upperarm_twist_02_r}, sphereLocalOffset=<0 -2 0>}
     {limitRadius=5, drivingBone {boneName=lowerarm_r}, sphereLocalOffset=<0 -2 0>}
    ]
   }
   n16 = Anim Dynamics - Bone: dyn_arm_hook_06!(ComponentPose=@n15, bAlphaBoolEnabled:bool=@n10, GravityOverride:Vector="0,0,-980", ExternalForce="0,0,0") -> (Pose) {
    alphaInputType=Bool
    angularDampingOverride=0.99
    angularSpringConstant=100
    bAngularSpring=true
    bOverrideAngularDamping=true
    bUseSphericalLimits=true
    simulationSpace=RootRelative
    alphaBoolBlend {blendOption=Linear}
    boundBone {boneName=dyn_arm_hook_06}
    componentAppliedLinearAccClamp=<3500 3500 3500>
    componentLinearAccScale=<0.8 0.8 0.8>
    componentLinearVelScale=<0.8 0.8 0.8>
    gravityOverride {z=-980}
    physicsBodyDefinitions [{
      collisionType=CustomSphere
      sphereCollisionRadius=4.5
      boundBone {boneName=dyn_arm_hook_06}
      boxExtents=<25 15 15>
      constraintSetup {angularLimitsMax=<1 35 35>, angularLimitsMin=<-1 -35 -35>, angularTarget=<1 0 0>}
      localJointOffset=<2 0 0>
     }]
    relativeSpaceBone {boneName=pelvis}
    sphericalLimits [
     {limitRadius=10, drivingBone {boneName=spine_02}, sphereLocalOffset=<0 0 0>}
     {limitRadius=10, drivingBone {boneName=spine_03}, sphereLocalOffset=<0 0 0>}
     {limitRadius=10, drivingBone {boneName=spine_04}, sphereLocalOffset=<0 0 0>}
     {limitRadius=6, drivingBone {boneName=upperarm_twist_02_r}, sphereLocalOffset=<0 -2 0>}
     {limitRadius=5, drivingBone {boneName=lowerarm_r}, sphereLocalOffset=<0 -2 0>}
    ]
   }
   n17 = Anim Dynamics - Bone: dyn_arm_hook_07!(ComponentPose=@n16, bAlphaBoolEnabled:bool=@n10, GravityOverride:Vector="0,0,-980", ExternalForce="0,0,0") -> (Pose) {
    alphaInputType=Bool
    angularDampingOverride=0.99
    angularSpringConstant=100
    bAngularSpring=true
    bOverrideAngularDamping=true
    bUseSphericalLimits=true
    simulationSpace=RootRelative
    alphaBoolBlend {blendOption=Linear}
    boundBone {boneName=dyn_arm_hook_07}
    componentAppliedLinearAccClamp=<3500 3500 3500>
    componentLinearAccScale=<0.8 0.8 0.8>
    componentLinearVelScale=<0.8 0.8 0.8>
    gravityOverride {z=-980}
    physicsBodyDefinitions [{
      collisionType=CustomSphere
      sphereCollisionRadius=4.5
      boundBone {boneName=dyn_arm_hook_07}
      boxExtents=<25 15 15>
      constraintSetup {angularLimitsMax=<1 35 35>, angularLimitsMin=<-1 -35 -35>, angularTarget=<1 0 0>}
      localJointOffset=<2 0 0>
     }]
    relativeSpaceBone {boneName=pelvis}
    sphericalLimits [
     {limitRadius=10, drivingBone {boneName=spine_02}, sphereLocalOffset=<0 0 0>}
     {limitRadius=10, drivingBone {boneName=spine_03}, sphereLocalOffset=<0 0 0>}
     {limitRadius=10, drivingBone {boneName=spine_04}, sphereLocalOffset=<0 0 0>}
     {limitRadius=6, drivingBone {boneName=upperarm_twist_02_r}, sphereLocalOffset=<0 -2 0>}
     {limitRadius=5, drivingBone {boneName=lowerarm_r}, sphereLocalOffset=<0 -2 0>}
    ]
   }
   n18 = Anim Dynamics - Bone: dyn_arm_hook_08!(ComponentPose=@n17, bAlphaBoolEnabled:bool=@n10, GravityOverride:Vector="0,0,-980", ExternalForce="0,0,0") -> (Pose) {
    alphaInputType=Bool
    angularDampingOverride=0.99
    angularSpringConstant=100
    bAngularSpring=true
    bOverrideAngularDamping=true
    bUseSphericalLimits=true
    simulationSpace=RootRelative
    alphaBoolBlend {blendOption=Linear}
    boundBone {boneName=dyn_arm_hook_08}
    componentAppliedLinearAccClamp=<3500 3500 3500>
    componentLinearAccScale=<0.8 0.8 0.8>
    componentLinearVelScale=<0.8 0.8 0.8>
    gravityOverride {z=-980}
    physicsBodyDefinitions [{
      collisionType=CustomSphere
      sphereCollisionRadius=4.5
      boundBone {boneName=dyn_arm_hook_08}
      boxExtents=<25 15 15>
      constraintSetup {angularLimitsMax=<1 35 35>, angularLimitsMin=<-1 -35 -35>, angularTarget=<1 0 0>}
      localJointOffset=<2 0 0>
     }]
    relativeSpaceBone {boneName=pelvis}
    sphericalLimits [
     {limitRadius=10, drivingBone {boneName=spine_02}, sphereLocalOffset=<0 0 0>}
     {limitRadius=10, drivingBone {boneName=spine_03}, sphereLocalOffset=<0 0 0>}
     {limitRadius=10, drivingBone {boneName=spine_04}, sphereLocalOffset=<0 0 0>}
     {limitRadius=6, drivingBone {boneName=upperarm_twist_02_r}, sphereLocalOffset=<0 -2 0>}
     {limitRadius=5, drivingBone {boneName=lowerarm_r}, sphereLocalOffset=<0 -2 0>}
    ]
   }
   n19 = Anim Dynamics - Bone: dyn_wrist_shackle_chain_01(ComponentPose=@n18, Alpha=1, ExternalForce="0,0,0") {
    angularDampingOverride=0.98
    angularSpringConstant=75
    bAngularSpring=true
    bGravityOverrideInSimSpace=true
    bOverrideAngularDamping=true
    simulationSpace=RootRelative
    alphaBoolBlend {blendOption=Linear}
    boundBone {boneName=dyn_wrist_shackle_chain_01}
    componentAppliedLinearAccClamp=<3500 3500 3500>
    componentLinearAccScale=<0.9 0.9 0.9>
    componentLinearVelScale=<0.9 0.9 0.9>
    physicsBodyDefinitions [{
      sphereCollisionRadius=0
      boundBone {boneName=dyn_wrist_shackle_chain_01}
      boxExtents=<25 15 15>
      constraintSetup {angularLimitsMax=<1 50 60>, angularLimitsMin=<-1 -5 -60>, angularTarget=<1 0 0>}
      localJointOffset=<2 0 0>
     }]
    relativeSpaceBone {boneName=pelvis}
   }
   n20 = Anim Dynamics - Bone: dyn_wrist_shackle_chain_02(ComponentPose=@n19, Alpha=1, ExternalForce="0,0,0") {
    angularDampingOverride=0.98
    angularSpringConstant=75
    bAngularSpring=true
    bGravityOverrideInSimSpace=true
    bOverrideAngularDamping=true
    bUseSphericalLimits=true
    simulationSpace=RootRelative
    alphaBoolBlend {blendOption=Linear}
    boundBone {boneName=dyn_wrist_shackle_chain_02}
    componentAppliedLinearAccClamp=<3500 3500 3500>
    componentLinearAccScale=<0.9 0.9 0.9>
    componentLinearVelScale=<0.9 0.9 0.9>
    physicsBodyDefinitions [{
      collisionType=CustomSphere
      sphereCollisionRadius=4
      boundBone {boneName=dyn_wrist_shackle_chain_02}
      boxExtents=<25 15 15>
      constraintSetup {angularLimitsMax=<1 50 60>, angularLimitsMin=<-1 -5 -60>, angularTarget=<1 0 0>}
      localJointOffset=<2 0 0>
     }]
    relativeSpaceBone {boneName=pelvis}
    sphericalLimits [{limitRadius=6, drivingBone {boneName=upperarm_twist_02_r}, sphereLocalOffset=<0 0 3>}]
   }
   n21 = Anim Dynamics - Bone: dyn_wrist_shackle_chain_03(ComponentPose=@n20, Alpha=1, ExternalForce="0,0,0") {
    angularDampingOverride=0.98
    angularSpringConstant=75
    bAngularSpring=true
    bGravityOverrideInSimSpace=true
    bOverrideAngularDamping=true
    bUseSphericalLimits=true
    simulationSpace=RootRelative
    alphaBoolBlend {blendOption=Linear}
    boundBone {boneName=dyn_wrist_shackle_chain_03}
    componentAppliedLinearAccClamp=<3500 3500 3500>
    componentLinearAccScale=<0.9 0.9 0.9>
    componentLinearVelScale=<0.9 0.9 0.9>
    physicsBodyDefinitions [{
      collisionType=CustomSphere
      sphereCollisionRadius=4
      boundBone {boneName=dyn_wrist_shackle_chain_03}
      boxExtents=<25 15 15>
      constraintSetup {angularLimitsMax=<1 50 60>, angularLimitsMin=<-1 -5 -60>, angularTarget=<1 0 0>}
      localJointOffset=<2 0 0>
     }]
    relativeSpaceBone {boneName=pelvis}
    sphericalLimits [{limitRadius=6, drivingBone {boneName=upperarm_twist_02_r}, sphereLocalOffset=<0 0 3>}]
   }
   n22 = Anim Dynamics - Bone: dyn_wrist_shackle_chain_04(ComponentPose=@n21, Alpha=1, ExternalForce="0,0,0") {
    angularDampingOverride=0.98
    angularSpringConstant=75
    bAngularSpring=true
    bGravityOverrideInSimSpace=true
    bOverrideAngularDamping=true
    bUseSphericalLimits=true
    simulationSpace=RootRelative
    alphaBoolBlend {blendOption=Linear}
    boundBone {boneName=dyn_wrist_shackle_chain_04}
    componentAppliedLinearAccClamp=<3500 3500 3500>
    componentLinearAccScale=<0.9 0.9 0.9>
    componentLinearVelScale=<0.9 0.9 0.9>
    physicsBodyDefinitions [{
      collisionType=CustomSphere
      sphereCollisionRadius=4
      boundBone {boneName=dyn_wrist_shackle_chain_04}
      boxExtents=<25 15 15>
      constraintSetup {angularLimitsMax=<1 50 60>, angularLimitsMin=<-1 -5 -60>, angularTarget=<1 0 0>}
      localJointOffset=<2 0 0>
     }]
    relativeSpaceBone {boneName=pelvis}
    sphericalLimits [{limitRadius=6, drivingBone {boneName=upperarm_twist_02_r}, sphereLocalOffset=<0 0 3>}]
   }
   n23 = Anim Dynamics - Bone: dyn_ankle_shackle(ComponentPose=@n22, Alpha=1, ExternalForce="0,0,0") {
    angularDampingOverride=0.95
    angularSpringConstant=1500
    bAngularSpring=true
    bGravityOverrideInSimSpace=true
    bOverrideAngularDamping=true
    gravityScale=0.75
    simulationSpace=RootRelative
    alphaBoolBlend {blendOption=Linear}
    boundBone {boneName=dyn_ankle_shackle}
    componentAppliedLinearAccClamp=<500 500 500>
    componentLinearAccScale=<1 1 1>
    componentLinearVelScale=<1 1 1>
    physicsBodyDefinitions [{
      sphereCollisionRadius=0
      boundBone {boneName=dyn_ankle_shackle}
      boxExtents=<20 20 20>
      constraintSetup {angularLimitsMax=<30 15 15>, angularLimitsMin=<-30 -15 -15>, angularTarget=<1 0 0>, linearAxesMax=<0.5 0 0>}
      localJointOffset=<0.5 0 0.5>
     }]
    relativeSpaceBone {boneName=pelvis}
   }
   n24 = Anim Dynamics - Bone: dyn_necklace_01(ComponentPose=@n23, Alpha=1, ExternalForce="0,0,0") {
    angularDampingOverride=0.998
    angularSpringConstant=150
    bAngularSpring=true
    bGravityOverrideInSimSpace=true
    bOverrideAngularDamping=true
    bUseGravityOverride=true
    bUsePlanarLimit=false
    simulationSpace=RootRelative
    alphaBoolBlend {blendOption=Linear}
    boundBone {boneName=dyn_necklace_01}
    componentAppliedLinearAccClamp=<3000 3000 3000>
    componentLinearAccScale=<0.6 0.6 0.6>
    componentLinearVelScale=<0.6 0.6 0.6>
    gravityOverride {z=-980}
    physicsBodyDefinitions [{
      collisionType=InnerSphere
      sphereCollisionRadius=1
      boundBone {boneName=dyn_necklace_01}
      boxExtents=<7 12 12>
      constraintSetup {twistAxis=AxisZ, angularLimitsMax=<0 0 10>, angularLimitsMin=<0 -35 -10>, angularTarget=<0 0 0>}
      localJointOffset=<2 0 0>
     }]
    relativeSpaceBone {boneName=root}
   }
   n25 = Anim Dynamics - Bone: dyn_necklace_02(ComponentPose=@n24, Alpha=1, ExternalForce="0,0,0") {
    angularDampingOverride=0.998
    angularSpringConstant=150
    bAngularSpring=true
    bGravityOverrideInSimSpace=true
    bOverrideAngularDamping=true
    bUseGravityOverride=true
    bUsePlanarLimit=false
    simulationSpace=RootRelative
    alphaBoolBlend {blendOption=Linear}
    boundBone {boneName=dyn_necklace_02}
    componentAppliedLinearAccClamp=<3000 3000 3000>
    componentLinearAccScale=<0.6 0.6 0.6>
    componentLinearVelScale=<0.6 0.6 0.6>
    gravityOverride {z=-980}
    physicsBodyDefinitions [{
      collisionType=InnerSphere
      sphereCollisionRadius=1
      boundBone {boneName=dyn_necklace_02}
      boxExtents=<7 12 12>
      constraintSetup {twistAxis=AxisZ, angularLimitsMax=<0 0 10>, angularLimitsMin=<0 -35 -10>, angularTarget=<0 0 0>}
      localJointOffset=<2 0 0>
     }]
    relativeSpaceBone {boneName=root}
    sphericalLimits [
     {limitRadius=7, drivingBone {boneName=spine_04}, sphereLocalOffset=<0 3.1 4>}
     {limitRadius=7, drivingBone {boneName=spine_04}, sphereLocalOffset=<0 3.1 -4>}
    ]
   }
   n26 = Anim Dynamics - Bone: dyn_necklace_03(ComponentPose=@n25, Alpha=1, ExternalForce="0,0,0") {
    angularDampingOverride=0.998
    angularSpringConstant=150
    bAngularSpring=true
    bGravityOverrideInSimSpace=true
    bOverrideAngularDamping=true
    bUseGravityOverride=true
    bUsePlanarLimit=false
    simulationSpace=RootRelative
    alphaBoolBlend {blendOption=Linear}
    boundBone {boneName=dyn_necklace_03}
    componentAppliedLinearAccClamp=<3000 3000 3000>
    componentLinearAccScale=<0.6 0.6 0.6>
    componentLinearVelScale=<0.6 0.6 0.6>
    gravityOverride {z=-980}
    physicsBodyDefinitions [{
      collisionType=InnerSphere
      sphereCollisionRadius=1
      boundBone {boneName=dyn_necklace_03}
      boxExtents=<7 12 12>
      constraintSetup {twistAxis=AxisZ, angularLimitsMax=<0 0 15>, angularLimitsMin=<0 -35 -15>, angularTarget=<0 0 0>}
      localJointOffset=<2 0 0>
     }]
    relativeSpaceBone {boneName=root}
    sphericalLimits [
     {limitRadius=7, drivingBone {boneName=spine_04}, sphereLocalOffset=<0 3 4>}
     {limitRadius=7, drivingBone {boneName=spine_04}, sphereLocalOffset=<0 3 -4>}
    ]
   }
   n27 = Anim Dynamics - Bone: dyn_bag_01(ComponentPose=@n26, Alpha=1, ExternalForce="0,0,0") {
    angularDampingOverride=0.98
    angularSpringConstant=75
    bAngularSpring=true
    bGravityOverrideInSimSpace=true
    bOverrideAngularDamping=true
    bUseSphericalLimits=true
    simulationSpace=RootRelative
    alphaBoolBlend {blendOption=Linear}
    boundBone {boneName=dyn_bag_01}
    componentAppliedLinearAccClamp=<3500 3500 3500>
    componentLinearAccScale=<0.9 0.9 0.9>
    componentLinearVelScale=<0.9 0.9 0.9>
    physicsBodyDefinitions [{
      collisionType=CustomSphere
      sphereCollisionRadius=4
      boundBone {boneName=dyn_bag_01}
      boxExtents=<25 15 15>
      constraintSetup {angularLimitsMax=<1 5 10>, angularLimitsMin=<-1 -45 -25>, angularTarget=<1 0 0>}
      localJointOffset=<2 0 0>
     }]
    relativeSpaceBone {boneName=pelvis}
    sphericalLimits [{limitRadius=6, drivingBone {boneName=upperarm_twist_02_r}, sphereLocalOffset=<0 0 3>}]
   }
   n28 = Anim Dynamics - Bone: dyn_bag_02(ComponentPose=@n27, Alpha=1, ExternalForce="0,0,0") {
    angularDampingOverride=0.98
    angularSpringConstant=75
    bAngularSpring=true
    bGravityOverrideInSimSpace=true
    bOverrideAngularDamping=true
    bUseSphericalLimits=true
    simulationSpace=RootRelative
    alphaBoolBlend {blendOption=Linear}
    boundBone {boneName=dyn_bag_02}
    componentAppliedLinearAccClamp=<3500 3500 3500>
    componentLinearAccScale=<0.9 0.9 0.9>
    componentLinearVelScale=<0.9 0.9 0.9>
    physicsBodyDefinitions [{
      collisionType=CustomSphere
      sphereCollisionRadius=4
      boundBone {boneName=dyn_bag_02}
      boxExtents=<25 15 15>
      constraintSetup {angularLimitsMax=<1 5 10>, angularLimitsMin=<-1 -45 -25>, angularTarget=<1 0 0>}
      localJointOffset=<2 0 0>
     }]
    relativeSpaceBone {boneName=pelvis}
    sphericalLimits [{limitRadius=6, drivingBone {boneName=upperarm_twist_02_r}, sphereLocalOffset=<0 0 3>}]
   }
   n29 = Anim Dynamics - Bone: dyn_Flask_01(ComponentPose=@n28, Alpha=1, ExternalForce="0,0,0") {
    angularDampingOverride=0.98
    angularSpringConstant=75
    bAngularSpring=true
    bGravityOverrideInSimSpace=true
    bOverrideAngularDamping=true
    bUseSphericalLimits=true
    simulationSpace=RootRelative
    alphaBoolBlend {blendOption=Linear}
    boundBone {boneName=dyn_Flask_01}
    componentAppliedLinearAccClamp=<3500 3500 3500>
    componentLinearAccScale=<0.9 0.9 0.9>
    componentLinearVelScale=<0.9 0.9 0.9>
    physicsBodyDefinitions [{
      collisionType=CustomSphere
      sphereCollisionRadius=4
      boundBone {boneName=dyn_Flask_01}
      boxExtents=<25 15 15>
      constraintSetup {angularLimitsMax=<1 5 10>, angularLimitsMin=<-1 -45 -25>, angularTarget=<1 0 0>}
      localJointOffset=<2 0 0>
     }]
    relativeSpaceBone {boneName=pelvis}
    sphericalLimits [{limitRadius=6, drivingBone {boneName=upperarm_twist_02_r}, sphereLocalOffset=<0 0 3>}]
   }
   n30 = Anim Dynamics - Bone: dyn_neck_choker(ComponentPose=@n29, Alpha=1, ExternalForce="0,0,0") {
    angularDampingOverride=0.998
    angularSpringConstant=150
    bAngularSpring=true
    bGravityOverrideInSimSpace=true
    bOverrideAngularDamping=true
    bUseGravityOverride=true
    bUsePlanarLimit=false
    alphaBoolBlend {blendOption=Linear}
    boundBone {boneName=dyn_neck_choker}
    componentAppliedLinearAccClamp=<3000 3000 3000>
    componentLinearAccScale=<0.5 0.5 0.5>
    componentLinearVelScale=<0.5 0.5 0.5>
    gravityOverride {z=-980}
    physicsBodyDefinitions [{
      collisionType=InnerSphere
      sphereCollisionRadius=1
      boundBone {boneName=dyn_neck_choker}
      boxExtents=<7 12 12>
      constraintSetup {twistAxis=AxisZ, angularLimitsMax=<0 10 15>, angularLimitsMin=<0 -35 -15>, angularTarget=<0 0 0>}
      localJointOffset=<2 0 0>
     }]
    relativeSpaceBone {boneName=root}
    sphericalLimits [
     {limitRadius=7, drivingBone {boneName=spine_04}, sphereLocalOffset=<0 3 4>}
     {limitRadius=7, drivingBone {boneName=spine_04}, sphereLocalOffset=<0 3 -4>}
    ]
   }
   n31 = RigidBody!(ComponentPose=@n30, Alpha:float, OverrideWorldGravity=Get GravityOverride(), ExternalForce) -> (Pose) {
    bClampLinearTranslationLimitToRefPose=true
    bOverrideWorldGravity=true
    overridePhysicsAsset=&/F_MED_Ruins_rope_Physics
    simulationSpace=BaseBoneSpace
    alphaBoolBlend {blendOption=Linear}
    baseBoneRef {boneName=pelvis}
    componentAppliedLinearAccClamp=<3000 3000 3000>
    componentLinearAccScale=<0.3 0.3 0.3>
    componentLinearVelScale=<0.3 0.3 0.3>
    simSpaceSettings {
     maxAngularAcceleration=1000
     maxAngularVelocity=3
     maxLinearAcceleration=700
     maxLinearVelocity=500
     velocityScaleZ=0
     worldAlpha=1
     externalLinearDragV=<0.2 0.2>
    }
   }
   Output Pose(Result=Component To Local(ComponentPose=@n31)) {layerGroup=None}
   n32 = "Constraint - dyn_arm_hook_MSTR (PPPPPPPP)"!(
    ComponentPose=Local To Component(LocalPose=Copy Pose From Mesh() {bUseAttachedParent=true})
    Alpha=1
    ConstraintWeights_0=0.1547
    ConstraintWeights_1:float=0.01
    ConstraintWeights_2:float=0.0175
    ConstraintWeights_3:float=0.1912
    ConstraintWeights_4:float=0.2042
    ConstraintWeights_5:float=0.0155
    ConstraintWeights_6:float=0.3828
    ConstraintWeights_7:float=0.0242
   ) -> (Pose) {
    alphaBoolBlend {blendOption=Linear}
    boneToModify {boneName=dyn_arm_hook_MSTR}
    constraintSetup [
     {targetBone {boneName=spine_04}}
     {targetBone {boneName=spine_03}}
     {targetBone {boneName=spine_02}}
     {targetBone {boneName=upperarm_twist_01_r}}
     {targetBone {boneName=spine_05}}
     {targetBone {boneName=clavicle_l}}
     {targetBone {boneName=clavicle_r}}
     {targetBone {boneName=deform_upperarm_up_r}}
    ]
    constraintWeights [1,1,1,1,1,1,1,1]
   }
   n33 = "Constraint - dyn_arm_chain_tgt_MSTR (PPPPPPPP)"!(
    ComponentPose=@n32
    Alpha=1
    ConstraintWeights_0=0.1547
    ConstraintWeights_1:float=0.01
    ConstraintWeights_2:float=0.0175
    ConstraintWeights_3:float=0.1912
    ConstraintWeights_4:float=0.2042
    ConstraintWeights_5:float=0.0155
    ConstraintWeights_6:float=0.3828
    ConstraintWeights_7:float=0.0242
   ) -> (Pose) {
    alphaBoolBlend {blendOption=Linear}
    boneToModify {boneName=dyn_arm_chain_tgt_MSTR}
    constraintSetup [
     {targetBone {boneName=spine_04}}
     {targetBone {boneName=spine_03}}
     {targetBone {boneName=spine_02}}
     {targetBone {boneName=upperarm_twist_01_r}}
     {targetBone {boneName=spine_05}}
     {targetBone {boneName=clavicle_l}}
     {targetBone {boneName=clavicle_r}}
     {targetBone {boneName=deform_upperarm_up_r}}
    ]
    constraintWeights [1,1,1,1,1,1,1,1]
   }
   n34 = "Constraint - dyn_neck_choker_MSTR (PPP)"!(ComponentPose=@n33, Alpha=1, ConstraintWeights_0=0.532, ConstraintWeights_1:float=0.2642, ConstraintWeights_2:float=0.2038) -> (Pose) {
    alphaBoolBlend {blendOption=Linear}
    boneToModify {boneName=dyn_neck_choker_MSTR}
    constraintSetup [{targetBone {boneName=spine_05}}, {targetBone {boneName=neck_02}}, {targetBone {boneName=neck_01}}]
    constraintWeights [1,1,1]
   }
   n35 = "Transform (Modify) Bone - Bone: floorCOL"!(ComponentPose=@n34, bAlphaBoolEnabled:bool=Get FloorCOL() -> (FloorCOL:bool), Translation="0,0,-100", Rotation="0,0,0", Scale="1,1,1") -> (Pose) {
    alphaInputType=Bool
    rotationMode=BMM_Additive
    rotationSpace=BCS_BoneSpace
    translationMode=BMM_Additive
    translationSpace=BCS_BoneSpace
    alphaBoolBlend {blendInTime=0.1, blendOption=Linear, blendOutTime=0.5}
    alphaScaleBiasClamp {bInterpResult=true}
    boneToModify {boneName=floorCOL}
    translation {z=-100}
   }
   n36 = Get bIsCrouching()
   n37 = "Transform (Modify) Bone - Bone: dyn_bag_MSTR"!(ComponentPose=@n35, bAlphaBoolEnabled:bool=@n36, Translation="0,0,0", Rotation="25,-0.000,-0.000", Scale="1,1,1") -> (Pose) {
    alphaInputType=Bool
    rotationMode=BMM_Additive
    rotationSpace=BCS_BoneSpace
    scaleMode=BMM_Additive
    scaleSpace=BCS_BoneSpace
    translationMode=BMM_Additive
    translationSpace=BCS_BoneSpace
    alphaBoolBlend {blendInTime=0.1, blendOption=Linear, blendOutTime=0.1}
    boneToModify {boneName=dyn_bag_MSTR}
    rotation {pitch=25}
   }
   n38 = "Transform (Modify) Bone - Bone: dyn_belt_rope_a_MSTR"!(ComponentPose=@n37, bAlphaBoolEnabled:bool=@n36, Translation="0,3,-3", Rotation="0,0,0", Scale="1,1,1") -> (Pose) {
    alphaInputType=Bool
    rotationMode=BMM_Additive
    rotationSpace=BCS_BoneSpace
    scaleMode=BMM_Additive
    scaleSpace=BCS_BoneSpace
    translationMode=BMM_Additive
    translationSpace=BCS_BoneSpace
    alphaBoolBlend {blendInTime=0.1, blendOption=Linear, blendOutTime=0.1}
    boneToModify {boneName=dyn_belt_rope_a_MSTR}
    translation {y=3, z=-3}
   }
   n39 = "Transform (Modify) Bone - Bone: dyn_belt_rope_b_MSTR"!(ComponentPose=@n38, bAlphaBoolEnabled:bool=@n36, Translation="0,5,0", Rotation="0,15,0", Scale="1,1,1") -> (Pose) {
    alphaInputType=Bool
    rotationMode=BMM_Additive
    rotationSpace=BCS_BoneSpace
    scaleMode=BMM_Additive
    scaleSpace=BCS_BoneSpace
    translationMode=BMM_Additive
    translationSpace=BCS_BoneSpace
    alphaBoolBlend {blendInTime=0.1, blendOption=Linear, blendOutTime=0.1}
    boneToModify {boneName=dyn_belt_rope_b_MSTR}
    rotation {yaw=15}
    translation {y=5}
   }
   n40 = "Transform (Modify) Bone - Bone: dyn_bag_MSTR"!(ComponentPose=@n39, Alpha:float=Get strafe() -> (strafe:double), Translation="0,0,0", Rotation="25,-0.000,-0.000", Scale="1,1,1") -> (Pose) {
    rotationMode=BMM_Additive
    rotationSpace=BCS_BoneSpace
    scaleMode=BMM_Additive
    scaleSpace=BCS_BoneSpace
    translationMode=BMM_Additive
    translationSpace=BCS_BoneSpace
    alphaBoolBlend {blendInTime=0.1, blendOption=Linear, blendOutTime=0.1}
    alphaScaleBiasClamp {bInterpResult=true}
    boneToModify {boneName=dyn_bag_MSTR}
    rotation {pitch=25}
   }
   n41 = "Transform (Modify) Bone - Bone: dyn_belt_rope_b_MSTR"!(ComponentPose=@n40, Alpha:float=Get ThighRotR() -> (ThighRotR:double), Translation="-3,0,-4", Rotation="-35,0,0", Scale="1,1,1") -> (Pose) {
    rotationMode=BMM_Additive
    rotationSpace=BCS_BoneSpace
    scaleMode=BMM_Additive
    scaleSpace=BCS_BoneSpace
    translationMode=BMM_Additive
    translationSpace=BCS_BoneSpace
    alphaBoolBlend {blendInTime=0.1, blendOption=Linear, blendOutTime=0.1}
    boneToModify {boneName=dyn_belt_rope_b_MSTR}
    rotation {pitch=-35}
    translation {x=-3, z=-4}
   }
   n42 = Control Rig(Source=Component To Local(ComponentPose=@n41)) {
    alphaBoolBlend {blendOption=Linear}
    controlRigAssetReference {blueprintRigClass=&/F_MED_DeformRig_Ruins_ControlRig.F_MED_DeformRig_Ruins_ControlRig_C}
   }
   n43 = Control Rig(Source=@n42) {
    alphaBoolBlend {blendOption=Linear}
    controlRigAssetReference {blueprintRigClass=&/F_MED_Ruins_CtrlRig_New.F_MED_Ruins_CtrlRig_New_C}
   }
   n44 = Look At - Bone: dyn_jacketBack_mstr_r!(ComponentPose=Local To Component(LocalPose=@n43), bAlphaBoolEnabled:bool=Get back skirt standing() -> (back skirt standing:bool)) -> (Pose) {
    alphaInputType=Bool
    alphaBoolBlend {blendInTime=0.5, blendOption=Linear, blendOutTime=1}
    boneToModify {boneName=dyn_jacketBack_mstr_r}
    lookAtLocation=<0 13 0>
    lookAtTarget {boneReference {boneName=foot_r}}
    lookAt_Axis {axis=<1 0>}
   }
   n45 = Get bIsMoving2D()
   n46 = "Transform (Modify) Bone - Bone: dyn_jacketBack_mstr_l"!(ComponentPose=@n44, bAlphaBoolEnabled:bool=@n45, Translation="0,0,0", Rotation="25,0,0", Scale="1,1,1") -> (Pose) {
    alphaInputType=Bool
    rotationMode=BMM_Additive
    rotationSpace=BCS_BoneSpace
    alphaBoolBlend {blendInTime=0.1, blendOption=Linear, blendOutTime=1}
    alphaScaleBiasClamp {bInterpResult=true}
    boneToModify {boneName=dyn_jacketBack_mstr_l}
    rotation {pitch=25}
   }
   n47 = "Transform (Modify) Bone - Bone: dyn_jacketBack_mstr_r"!(ComponentPose=@n46, bAlphaBoolEnabled:bool=@n45, Translation="0,0,0", Rotation="25,0,0.000", Scale="1,1,1") -> (Pose) {
    alphaInputType=Bool
    rotationMode=BMM_Additive
    rotationSpace=BCS_BoneSpace
    alphaBoolBlend {blendInTime=0.1, blendOption=Linear, blendOutTime=1}
    alphaScaleBiasClamp {bInterpResult=true}
    boneToModify {boneName=dyn_jacketBack_mstr_r}
    rotation {pitch=25}
   }
   n48 = "Transform (Modify) Bone - Bone: dyn_jacketSide_mstr_l"!(ComponentPose=@n47, bAlphaBoolEnabled:bool=@n45, Translation="0,0,0", Rotation="20,0,0", Scale="1,1,1") -> (Pose) {
    alphaInputType=Bool
    rotationMode=BMM_Additive
    rotationSpace=BCS_BoneSpace
    alphaBoolBlend {blendInTime=0.1, blendOption=Linear, blendOutTime=1}
    alphaScaleBiasClamp {bInterpResult=true}
    boneToModify {boneName=dyn_jacketSide_mstr_l}
    rotation {pitch=20}
   }
   n49 = "Transform (Modify) Bone - Bone: dyn_jacketSide_mstr_r"!(ComponentPose=@n48, bAlphaBoolEnabled:bool=@n45, Translation="0,0,0", Rotation="20,15,0", Scale="1,1,1") -> (Pose) {
    alphaInputType=Bool
    rotationMode=BMM_Additive
    rotationSpace=BCS_BoneSpace
    alphaBoolBlend {blendInTime=0.1, blendOption=Linear, blendOutTime=1}
    alphaScaleBiasClamp {bInterpResult=true}
    boneToModify {boneName=dyn_jacketSide_mstr_r}
    rotation {pitch=20, yaw=15}
   }
   n50 = "Transform (Modify) Bone - Bone: dyn_jacketFront_mstr_l"!(ComponentPose=@n49, bAlphaBoolEnabled:bool=@n45, Translation="0,0,0", Rotation="5,0,0", Scale="1,1,1") -> (Pose) {
    alphaInputType=Bool
    rotationMode=BMM_Additive
    rotationSpace=BCS_BoneSpace
    alphaBoolBlend {blendInTime=0.1, blendOption=Linear, blendOutTime=1}
    alphaScaleBiasClamp {bInterpResult=true}
    boneToModify {boneName=dyn_jacketFront_mstr_l}
    rotation {pitch=5}
   }
   n51 = "Transform (Modify) Bone - Bone: dyn_jacketFront_mstr_r"!(ComponentPose=@n50, bAlphaBoolEnabled:bool=@n45, Translation="0,0,0", Rotation="5,0,0", Scale="1,1,1") -> (Pose) {
    alphaInputType=Bool
    rotationMode=BMM_Additive
    rotationSpace=BCS_BoneSpace
    alphaBoolBlend {blendInTime=0.1, blendOption=Linear, blendOutTime=1}
    alphaScaleBiasClamp {bInterpResult=true}
    boneToModify {boneName=dyn_jacketFront_mstr_r}
    rotation {pitch=5}
   }
   n52 = "Transform (Modify) Bone - Bone: dyn_jacketFront_mstr"!(ComponentPose=@n51, bAlphaBoolEnabled:bool=@n45, Translation="0,0,0", Rotation="5,10,0", Scale="1,1,1") -> (Pose) {
    alphaInputType=Bool
    rotationMode=BMM_Additive
    rotationSpace=BCS_BoneSpace
    alphaBoolBlend {blendInTime=0.1, blendOption=Linear, blendOutTime=1}
    alphaScaleBiasClamp {bInterpResult=true}
    boneToModify {boneName=dyn_jacketFront_mstr}
    rotation {pitch=5, yaw=10}
   }
   n53 = Get jumping() -> (jumping:double)
   n54 = "Transform (Modify) Bone - Bone: dyn_jacketBack_mstr_l"!(ComponentPose=@n52, Alpha:float=@n53, Translation="0,0,0", Rotation="35,10,0.000", Scale="1,1,1") -> (Pose) {
    rotationMode=BMM_Additive
    rotationSpace=BCS_BoneSpace
    alphaBoolBlend {blendOption=Linear}
    alphaScaleBiasClamp {bInterpResult=true, interpSpeedIncreasing=5}
    boneToModify {boneName=dyn_jacketBack_mstr_l}
    rotation {pitch=35, yaw=10}
   }
   n55 = "Transform (Modify) Bone - Bone: dyn_jacketBack_mstr_r"!(ComponentPose=@n54, Alpha:float=@n53, Translation="0,0,0", Rotation="35,-10,0.000", Scale="1,1,1") -> (Pose) {
    rotationMode=BMM_Additive
    rotationSpace=BCS_BoneSpace
    alphaBoolBlend {blendOption=Linear}
    alphaScaleBiasClamp {bInterpResult=true, interpSpeedIncreasing=5}
    boneToModify {boneName=dyn_jacketBack_mstr_r}
    rotation {pitch=35, yaw=-10}
   }
   n56 = "Transform (Modify) Bone - Bone: dyn_jacketSide_mstr_l"!(ComponentPose=@n55, Alpha:float=@n53, Translation="0,0,0", Rotation="15,-15,0.000", Scale="1,1,1") -> (Pose) {
    rotationMode=BMM_Additive
    rotationSpace=BCS_BoneSpace
    alphaBoolBlend {blendOption=Linear}
    alphaScaleBiasClamp {bInterpResult=true, interpSpeedIncreasing=5}
    boneToModify {boneName=dyn_jacketSide_mstr_l}
    rotation {pitch=15, yaw=-15}
   }
   n57 = "Transform (Modify) Bone - Bone: dyn_jacketSide_mstr_r"!(ComponentPose=@n56, Alpha:float=@n53, Translation="0,0,0", Rotation="15,15,0.000", Scale="1,1,1") -> (Pose) {
    rotationMode=BMM_Additive
    rotationSpace=BCS_BoneSpace
    alphaBoolBlend {blendOption=Linear}
    alphaScaleBiasClamp {bInterpResult=true, interpSpeedIncreasing=5}
    boneToModify {boneName=dyn_jacketSide_mstr_r}
    rotation {pitch=15, yaw=15}
   }
   n58 = Get thigh_rot_inr_l() -> (thigh_rot_inr_l:double)
   n59 = "Transform (Modify) Bone - Bone: dyn_jacketSide_mstr_l"(ComponentPose=@n57, Alpha=@n58, Translation="0,0,0", Rotation="15,0,0.000", Scale="1,1,1") {
    rotationMode=BMM_Additive
    rotationSpace=BCS_BoneSpace
    alphaBoolBlend {blendOption=Linear}
    boneToModify {boneName=dyn_jacketSide_mstr_l}
    rotation {pitch=15}
   }
   n60 = Get thigh_rot_inr_r() -> (thigh_rot_inr_r:double)
   n61 = "Transform (Modify) Bone - Bone: dyn_jacketSide_mstr_r"(ComponentPose=@n59, Alpha=@n60, Translation="0,0,0", Rotation="10,0,0.000", Scale="1,1,1") {
    rotationMode=BMM_Additive
    rotationSpace=BCS_BoneSpace
    alphaBoolBlend {blendOption=Linear}
    alphaScaleBiasClamp {interpSpeedIncreasing=5}
    boneToModify {boneName=dyn_jacketSide_mstr_r}
    rotation {pitch=10}
   }
   n62 = "Transform (Modify) Bone - Bone: dyn_jacketFront_mstr_l"!(ComponentPose=@n61, Alpha:float=@n58, Translation="0,0,0", Rotation="15,0,0.000", Scale="1,1,1") -> (Pose) {
    rotationMode=BMM_Additive
    rotationSpace=BCS_BoneSpace
    alphaBoolBlend {blendOption=Linear}
    boneToModify {boneName=dyn_jacketFront_mstr_l}
    rotation {pitch=15}
   }
   n63 = "Transform (Modify) Bone - Bone: dyn_jacketFront_mstr_r"!(ComponentPose=@n62, Alpha:float=@n60, Translation="0,0,0", Rotation="5,0,0.000", Scale="1,1,1") -> (Pose) {
    rotationMode=BMM_Additive
    rotationSpace=BCS_BoneSpace
    alphaBoolBlend {blendOption=Linear}
    alphaScaleBiasClamp {interpSpeedIncreasing=5}
    boneToModify {boneName=dyn_jacketFront_mstr_r}
    rotation {pitch=5}
   }
   n64 = Get thigh_rot_otr_l() -> (thigh_rot_otr_l:double)
   n65 = "Transform (Modify) Bone - Bone: dyn_jacketSide_mstr_l"(ComponentPose=@n63, Alpha=@n64, Translation="0,0,0", Rotation="15,0,0.000", Scale="1,1,1") {
    rotationMode=BMM_Additive
    rotationSpace=BCS_BoneSpace
    alphaBoolBlend {blendOption=Linear}
    boneToModify {boneName=dyn_jacketSide_mstr_l}
    rotation {pitch=15}
   }
   n66 = Get thigh_rot_otr_r() -> (thigh_rot_otr_r:double)
   n67 = "Transform (Modify) Bone - Bone: dyn_jacketSide_mstr_r"(ComponentPose=@n65, Alpha=@n66, Translation="0,0,0", Rotation="15,0,0.000", Scale="1,1,1") {
    rotationMode=BMM_Additive
    rotationSpace=BCS_BoneSpace
    alphaBoolBlend {blendOption=Linear}
    alphaScaleBiasClamp {interpSpeedIncreasing=5}
    boneToModify {boneName=dyn_jacketSide_mstr_r}
    rotation {pitch=15}
   }
   n68 = "Transform (Modify) Bone - Bone: dyn_jacketFront_mstr_l"(ComponentPose=@n67, Alpha=@n64, Translation="0,0,0", Rotation="15,0,0.000", Scale="1,1,1") {
    rotationMode=BMM_Additive
    rotationSpace=BCS_BoneSpace
    alphaBoolBlend {blendOption=Linear}
    boneToModify {boneName=dyn_jacketFront_mstr_l}
    rotation {pitch=15}
   }
   n69 = "Transform (Modify) Bone - Bone: dyn_jacketFront_mstr_r"(ComponentPose=@n68, Alpha=@n66, Translation="0,0,0", Rotation="15,0,0.000", Scale="1,1,1") {
    rotationMode=BMM_Additive
    rotationSpace=BCS_BoneSpace
    alphaBoolBlend {blendOption=Linear}
    alphaScaleBiasClamp {interpSpeedIncreasing=5}
    boneToModify {boneName=dyn_jacketFront_mstr_r}
    rotation {pitch=15}
   }
   n70 = Get still_crouch() -> (still_crouch:bool)
   n71 = "Transform (Modify) Bone - Bone: dyn_jacketBack_mstr_l"!(ComponentPose=@n69, bAlphaBoolEnabled:bool=@n70, Translation="0,0,0", Rotation="15,-10,-15", Scale="1,1,1") -> (Pose) {
    alphaInputType=Bool
    rotationMode=BMM_Additive
    rotationSpace=BCS_BoneSpace
    scaleMode=BMM_Additive
    scaleSpace=BCS_BoneSpace
    translationMode=BMM_Additive
    translationSpace=BCS_BoneSpace
    alphaBoolBlend {blendInTime=0.1, blendOption=Linear, blendOutTime=0.5}
    boneToModify {boneName=dyn_jacketBack_mstr_l}
    rotation=<rot 15 -10 -15>
   }
   n72 = "Transform (Modify) Bone - Bone: dyn_jacketBack_mstr_r"!(ComponentPose=@n71, bAlphaBoolEnabled:bool=@n70, Translation="0,0,0", Rotation="15,0,-45", Scale="1,1,1") -> (Pose) {
    alphaInputType=Bool
    rotationMode=BMM_Additive
    rotationSpace=BCS_BoneSpace
    scaleMode=BMM_Additive
    scaleSpace=BCS_BoneSpace
    translationMode=BMM_Additive
    translationSpace=BCS_BoneSpace
    alphaBoolBlend {blendInTime=0.1, blendOption=Linear, blendOutTime=0.5}
    boneToModify {boneName=dyn_jacketBack_mstr_r}
    rotation {pitch=15, roll=-45}
   }
   n73 = "Transform (Modify) Bone - Bone: dyn_jacketBack_01_l"!(ComponentPose=@n72, bAlphaBoolEnabled:bool=@n70, Translation="0,0,0", Rotation="15,0,0", Scale="1,1,1") -> (Pose) {
    alphaInputType=Bool
    rotationMode=BMM_Additive
    rotationSpace=BCS_BoneSpace
    scaleMode=BMM_Additive
    scaleSpace=BCS_BoneSpace
    translationMode=BMM_Additive
    translationSpace=BCS_BoneSpace
    alphaBoolBlend {blendInTime=0.1, blendOption=Linear, blendOutTime=0.5}
    boneToModify {boneName=dyn_jacketBack_01_l}
    rotation {pitch=15}
   }
   n74 = "Transform (Modify) Bone - Bone: dyn_jacketBack_01_r"!(ComponentPose=@n73, bAlphaBoolEnabled:bool=@n70, Translation="0,0,0", Rotation="30,0,0", Scale="1,1,1") -> (Pose) {
    alphaInputType=Bool
    rotationMode=BMM_Additive
    rotationSpace=BCS_BoneSpace
    scaleMode=BMM_Additive
    scaleSpace=BCS_BoneSpace
    translationMode=BMM_Additive
    translationSpace=BCS_BoneSpace
    alphaBoolBlend {blendInTime=0.1, blendOption=Linear, blendOutTime=0.5}
    boneToModify {boneName=dyn_jacketBack_01_r}
    rotation {pitch=30}
   }
   n75 = "Transform (Modify) Bone - Bone: dyn_jacketBack_02_r"!(ComponentPose=@n74, bAlphaBoolEnabled:bool=@n70, Translation="0,0,0", Rotation="30,0,0", Scale="1,1,1") -> (Pose) {
    alphaInputType=Bool
    rotationMode=BMM_Additive
    rotationSpace=BCS_BoneSpace
    scaleMode=BMM_Additive
    scaleSpace=BCS_BoneSpace
    translationMode=BMM_Additive
    translationSpace=BCS_BoneSpace
    alphaBoolBlend {blendInTime=0.1, blendOption=Linear, blendOutTime=0.5}
    boneToModify {boneName=dyn_jacketBack_02_r}
    rotation {pitch=30}
   }
   n76 = "Transform (Modify) Bone - Bone: dyn_jacketSide_01_r"!(ComponentPose=@n75, bAlphaBoolEnabled:bool=@n70, Translation="0,0,0", Rotation="15,35,0", Scale="1,1,1") -> (Pose) {
    alphaInputType=Bool
    rotationMode=BMM_Additive
    rotationSpace=BCS_BoneSpace
    scaleMode=BMM_Additive
    scaleSpace=BCS_BoneSpace
    translationMode=BMM_Additive
    translationSpace=BCS_BoneSpace
    alphaBoolBlend {blendInTime=0.1, blendOption=Linear, blendOutTime=0.5}
    boneToModify {boneName=dyn_jacketSide_01_r}
    rotation {pitch=15, yaw=35}
   }
   n77 = "Transform (Modify) Bone - Bone: dyn_jacketBack_03_l"!(ComponentPose=@n76, bAlphaBoolEnabled:bool=@n70, Translation="0,0,0", Rotation="35,0,0", Scale="1,1,1") -> (Pose) {
    alphaInputType=Bool
    rotationMode=BMM_Additive
    rotationSpace=BCS_BoneSpace
    scaleMode=BMM_Additive
    scaleSpace=BCS_BoneSpace
    translationMode=BMM_Additive
    translationSpace=BCS_BoneSpace
    alphaBoolBlend {blendInTime=0.1, blendOption=Linear, blendOutTime=0.5}
    boneToModify {boneName=dyn_jacketBack_03_l}
    rotation {pitch=35}
   }
   n78 = "Transform (Modify) Bone - Bone: dyn_jacketBack_04_l"!(ComponentPose=@n77, bAlphaBoolEnabled:bool=@n70, Translation="0,0,0", Rotation="30,0,0", Scale="1,1,1") -> (Pose) {
    alphaInputType=Bool
    rotationMode=BMM_Additive
    rotationSpace=BCS_BoneSpace
    scaleMode=BMM_Additive
    scaleSpace=BCS_BoneSpace
    translationMode=BMM_Additive
    translationSpace=BCS_BoneSpace
    alphaBoolBlend {blendInTime=0.1, blendOption=Linear, blendOutTime=0.5}
    boneToModify {boneName=dyn_jacketBack_04_l}
    rotation {pitch=30}
   }
   n79 = "Transform (Modify) Bone - Bone: dyn_jacketSide_03_l"!(ComponentPose=@n78, bAlphaBoolEnabled:bool=@n70, Translation="0,0,0", Rotation="15,0,0", Scale="1,1,1") -> (Pose) {
    alphaInputType=Bool
    rotationMode=BMM_Additive
    rotationSpace=BCS_BoneSpace
    scaleMode=BMM_Additive
    scaleSpace=BCS_BoneSpace
    translationMode=BMM_Additive
    translationSpace=BCS_BoneSpace
    alphaBoolBlend {blendInTime=0.1, blendOption=Linear, blendOutTime=0.5}
    boneToModify {boneName=dyn_jacketSide_03_l}
    rotation {pitch=15}
   }
   n80 = "Transform (Modify) Bone - Bone: dyn_jacketSide_03_l"(ComponentPose=@n79, bAlphaBoolEnabled=@n70, Translation="0,0,0", Rotation="20,0,0", Scale="1,1,1") {
    alphaInputType=Bool
    rotationMode=BMM_Additive
    rotationSpace=BCS_BoneSpace
    scaleMode=BMM_Additive
    scaleSpace=BCS_BoneSpace
    translationMode=BMM_Additive
    translationSpace=BCS_BoneSpace
    alphaBoolBlend {blendInTime=0.1, blendOption=Linear, blendOutTime=0.5}
    boneToModify {boneName=dyn_jacketSide_03_l}
    rotation {pitch=20}
   }
   n81 = Get bIsSliding()
   n82 = "Transform (Modify) Bone - Bone: dyn_jacketBack_mstr_l"(ComponentPose=@n80, bAlphaBoolEnabled=@n81, Translation="-5,0,8", Rotation="75,0,0", Scale="1,1,1") {
    alphaInputType=Bool
    rotationMode=BMM_Additive
    rotationSpace=BCS_BoneSpace
    translationMode=BMM_Additive
    translationSpace=BCS_BoneSpace
    alphaBoolBlend {blendInTime=0.2, blendOption=Linear, blendOutTime=0.3}
    alphaScaleBiasClamp {bInterpResult=true}
    boneToModify {boneName=dyn_jacketBack_mstr_l}
    rotation {pitch=75}
    translation {x=-5, z=8}
   }
   n83 = "Transform (Modify) Bone - Bone: dyn_jacketBack_mstr_r"(ComponentPose=@n82, bAlphaBoolEnabled=@n81, Translation="-5,0,8", Rotation="75,0,0", Scale="1,1,1") {
    alphaInputType=Bool
    rotationMode=BMM_Additive
    rotationSpace=BCS_BoneSpace
    translationMode=BMM_Additive
    translationSpace=BCS_BoneSpace
    alphaBoolBlend {blendInTime=0.2, blendOption=Linear, blendOutTime=0.3}
    alphaScaleBiasClamp {bInterpResult=true}
    boneToModify {boneName=dyn_jacketBack_mstr_r}
    rotation {pitch=75}
    translation {x=-5, z=8}
   }
   n84 = "Transform (Modify) Bone - Bone: dyn_jacketFront_mstr_r"!(ComponentPose=@n83, bAlphaBoolEnabled:bool=@n81, Translation="0,0,2", Rotation="5,0,0", Scale="1,1,1") -> (Pose) {
    alphaInputType=Bool
    rotationMode=BMM_Additive
    rotationSpace=BCS_BoneSpace
    translationMode=BMM_Additive
    translationSpace=BCS_BoneSpace
    alphaBoolBlend {blendInTime=0.2, blendOption=Linear, blendOutTime=0.3}
    alphaScaleBiasClamp {bInterpResult=true}
    boneToModify {boneName=dyn_jacketFront_mstr_r}
    rotation {pitch=5}
    translation {z=2}
   }
   n85 = "Transform (Modify) Bone - Bone: dyn_jacketFront_mstr"(ComponentPose=@n84, bAlphaBoolEnabled=@n81, Translation="0,0,3", Rotation="5,0,0.000", Scale="1,1,1") {
    alphaInputType=Bool
    rotationMode=BMM_Additive
    rotationSpace=BCS_BoneSpace
    translationMode=BMM_Additive
    translationSpace=BCS_BoneSpace
    alphaBoolBlend {blendInTime=0.2, blendOption=Linear, blendOutTime=0.3}
    alphaScaleBiasClamp {bInterpResult=true}
    boneToModify {boneName=dyn_jacketFront_mstr}
    rotation {pitch=5}
    translation {z=3}
   }
   n86 = "Transform (Modify) Bone - Bone: dyn_jacketSide_mstr_l"!(ComponentPose=@n85, bAlphaBoolEnabled:bool=@n81, Translation="0,0,5", Rotation="25,0.000,-0.000", Scale="1,1,1") -> (Pose) {
    alphaInputType=Bool
    rotationMode=BMM_Additive
    rotationSpace=BCS_BoneSpace
    translationMode=BMM_Additive
    translationSpace=BCS_BoneSpace
    alphaBoolBlend {blendInTime=0.2, blendOption=Linear, blendOutTime=0.3}
    alphaScaleBiasClamp {bInterpResult=true}
    boneToModify {boneName=dyn_jacketSide_mstr_l}
    rotation {pitch=25}
    translation {z=5}
   }
   n87 = "Transform (Modify) Bone - Bone: dyn_jacketSide_mstr_r"!(ComponentPose=@n86, bAlphaBoolEnabled:bool=@n81, Translation="0,0,5", Rotation="25,0,0.000", Scale="1,1,1") -> (Pose) {
    alphaInputType=Bool
    rotationMode=BMM_Additive
    rotationSpace=BCS_BoneSpace
    translationMode=BMM_Additive
    translationSpace=BCS_BoneSpace
    alphaBoolBlend {blendInTime=0.2, blendOption=Linear, blendOutTime=0.3}
    alphaScaleBiasClamp {bInterpResult=true}
    boneToModify {boneName=dyn_jacketSide_mstr_r}
    rotation {pitch=25}
    translation {z=5}
   }
   n88 = "Transform (Modify) Bone - Bone: dyn_bag_MSTR"!(ComponentPose=@n87, bAlphaBoolEnabled:bool=Get bIsSkydiveDiveMode(), Translation="0,0,0", Rotation="25,-0.000,-0.000", Scale="1,1,1") -> (Pose) {
    alphaInputType=Bool
    rotationMode=BMM_Additive
    rotationSpace=BCS_BoneSpace
    scaleMode=BMM_Additive
    scaleSpace=BCS_BoneSpace
    translationMode=BMM_Additive
    translationSpace=BCS_BoneSpace
    alphaBoolBlend {blendInTime=0.1, blendOption=Linear, blendOutTime=0.1}
    boneToModify {boneName=dyn_bag_MSTR}
    rotation {pitch=25}
   }
   n89 = Get bIsInFrontEnd()
   n90 = "Transform (Modify) Bone - Bone: dyn_shlr_chain_l"!(ComponentPose=@n88, bAlphaBoolEnabled:bool=@n89, Translation="0,3,0", Rotation="0,0,0", Scale="1,1,1") -> (Pose) {
    alphaInputType=Bool
    rotationMode=BMM_Additive
    rotationSpace=BCS_BoneSpace
    scaleMode=BMM_Additive
    scaleSpace=BCS_BoneSpace
    translationMode=BMM_Additive
    translationSpace=BCS_BoneSpace
    alphaBoolBlend {blendInTime=0.1, blendOption=Linear, blendOutTime=0.1}
    boneToModify {boneName=dyn_shlr_chain_l}
    translation {y=3}
   }
   n91 = "Transform (Modify) Bone - Bone: dyn_shlr_chain_r"!(ComponentPose=@n90, bAlphaBoolEnabled:bool=@n89, Translation="0,-5,0", Rotation="0,0,0", Scale="1,1,1") -> (Pose) {
    alphaInputType=Bool
    rotationMode=BMM_Additive
    rotationSpace=BCS_BoneSpace
    scaleMode=BMM_Additive
    scaleSpace=BCS_BoneSpace
    translationMode=BMM_Additive
    translationSpace=BCS_BoneSpace
    alphaBoolBlend {blendInTime=0.1, blendOption=Linear, blendOutTime=0.1}
    boneToModify {boneName=dyn_shlr_chain_r}
    translation {y=-5}
   }
   n92 = "Transform (Modify) Bone - Bone: dyn_arm_chain_twist_front"!(ComponentPose=@n91, bAlphaBoolEnabled:bool=@n89, Translation="0,-1,-3", Rotation="10,0,0", Scale="1,1,1") -> (Pose) {
    alphaInputType=Bool
    rotationMode=BMM_Additive
    rotationSpace=BCS_BoneSpace
    scaleMode=BMM_Additive
    scaleSpace=BCS_BoneSpace
    translationMode=BMM_Additive
    translationSpace=BCS_BoneSpace
    alphaBoolBlend {blendInTime=0.1, blendOption=Linear, blendOutTime=0.1}
    boneToModify {boneName=dyn_arm_chain_twist_front}
    rotation {pitch=10}
    translation {y=-1, z=-3}
   }
   Save cached pose 'presim'(Pose=Component To Local(ComponentPose=@n92)) {cacheName=presim}
   n93 = Get bIsMoving2D()
   n94 = "Transform (Modify) Bone - Bone: dyn_jacketSide_mstr_l"(
    ComponentPose=Local To Component(LocalPose=Use cached pose 'presim'())
    bAlphaBoolEnabled=@n93
    Translation="0,0,0"
    Rotation="-20,0,0"
    Scale="1,1,1"
   ) {
    alphaInputType=Bool
    rotationMode=BMM_Additive
    rotationSpace=BCS_BoneSpace
    alphaBoolBlend {blendInTime=0.1, blendOption=Linear, blendOutTime=1}
    alphaScaleBiasClamp {bInterpResult=true}
    boneToModify {boneName=dyn_jacketSide_mstr_l}
    rotation {pitch=-20}
   }
   n95 = "Transform (Modify) Bone - Bone: dyn_jacketSide_mstr_r"(ComponentPose=@n94, bAlphaBoolEnabled=@n93, Translation="0,0,0", Rotation="-20,-15,0", Scale="1,1,1") {
    alphaInputType=Bool
    rotationMode=BMM_Additive
    rotationSpace=BCS_BoneSpace
    alphaBoolBlend {blendInTime=0.1, blendOption=Linear, blendOutTime=1}
    alphaScaleBiasClamp {bInterpResult=true}
    boneToModify {boneName=dyn_jacketSide_mstr_r}
    rotation {pitch=-20, yaw=-15}
   }
   n96 = "Transform (Modify) Bone - Bone: dyn_jacketFront_mstr_l"!(ComponentPose=@n95, bAlphaBoolEnabled:bool=@n93, Translation="0,0,0", Rotation="-5,0,0", Scale="1,1,1") -> (Pose) {
    alphaInputType=Bool
    rotationMode=BMM_Additive
    rotationSpace=BCS_BoneSpace
    alphaBoolBlend {blendInTime=0.1, blendOption=Linear, blendOutTime=1}
    alphaScaleBiasClamp {bInterpResult=true}
    boneToModify {boneName=dyn_jacketFront_mstr_l}
    rotation {pitch=-5}
   }
   n97 = "Transform (Modify) Bone - Bone: dyn_jacketFront_mstr_r"(ComponentPose=@n96, bAlphaBoolEnabled=@n93, Translation="0,0,0", Rotation="-5,0,0", Scale="1,1,1") {
    alphaInputType=Bool
    rotationMode=BMM_Additive
    rotationSpace=BCS_BoneSpace
    alphaBoolBlend {blendInTime=0.1, blendOption=Linear, blendOutTime=1}
    alphaScaleBiasClamp {bInterpResult=true}
    boneToModify {boneName=dyn_jacketFront_mstr_r}
    rotation {pitch=-5}
   }
   n98 = "Transform (Modify) Bone - Bone: dyn_jacketFront_mstr"(ComponentPose=@n97, bAlphaBoolEnabled=@n93, Translation="0,0,0", Rotation="-5,-10,0", Scale="1,1,1") {
    alphaInputType=Bool
    rotationMode=BMM_Additive
    rotationSpace=BCS_BoneSpace
    alphaBoolBlend {blendInTime=0.1, blendOption=Linear, blendOutTime=1}
    alphaScaleBiasClamp {bInterpResult=true}
    boneToModify {boneName=dyn_jacketFront_mstr}
    rotation {pitch=-5, yaw=-10}
   }
   n99 = Get jumping()
   n100 = "Transform (Modify) Bone - Bone: dyn_jacketBack_mstr_l"!(ComponentPose=@n98, Alpha:float=@n99, Translation="0,0,0", Rotation="-10,-10,0", Scale="1,1,1") -> (Pose) {
    rotationMode=BMM_Additive
    rotationSpace=BCS_BoneSpace
    alphaBoolBlend {blendOption=Linear}
    alphaScaleBiasClamp {bInterpResult=true, interpSpeedIncreasing=5}
    boneToModify {boneName=dyn_jacketBack_mstr_l}
    rotation {pitch=-10, yaw=-10}
   }
   n101 = "Transform (Modify) Bone - Bone: dyn_jacketBack_mstr_r"!(ComponentPose=@n100, Alpha:float=@n99, Translation="0,0,0", Rotation="-25,10,0", Scale="1,1,1") -> (Pose) {
    rotationMode=BMM_Additive
    rotationSpace=BCS_BoneSpace
    alphaBoolBlend {blendOption=Linear}
    alphaScaleBiasClamp {bInterpResult=true, interpSpeedIncreasing=5}
    boneToModify {boneName=dyn_jacketBack_mstr_r}
    rotation {pitch=-25, yaw=10}
   }
   n102 = "Transform (Modify) Bone - Bone: dyn_jacketSide_mstr_l"!(ComponentPose=@n101, Alpha:float=@n99, Translation="0,0,0", Rotation="-10,10,0", Scale="1,1,1") -> (Pose) {
    rotationMode=BMM_Additive
    rotationSpace=BCS_BoneSpace
    alphaBoolBlend {blendOption=Linear}
    alphaScaleBiasClamp {bInterpResult=true, interpSpeedIncreasing=5}
    boneToModify {boneName=dyn_jacketSide_mstr_l}
    rotation {pitch=-10, yaw=10}
   }
   n103 = "Transform (Modify) Bone - Bone: dyn_jacketSide_mstr_r"!(ComponentPose=@n102, Alpha:float=@n99, Translation="0,0,0", Rotation="-10,-10,0", Scale="1,1,1") -> (Pose) {
    rotationMode=BMM_Additive
    rotationSpace=BCS_BoneSpace
    alphaBoolBlend {blendOption=Linear}
    alphaScaleBiasClamp {bInterpResult=true, interpSpeedIncreasing=5}
    boneToModify {boneName=dyn_jacketSide_mstr_r}
    rotation {pitch=-10, yaw=-10}
   }
   n104 = Get bIsSkydiveFloating()
   n105 = "Transform (Modify) Bone - Bone: dyn_jacketBack_mstr_l"!(ComponentPose=@n103, bAlphaBoolEnabled:bool=@n104, Translation="0,0,0", Rotation="-15,0,0", Scale="1,1,1") -> (Pose) {
    alphaInputType=Bool
    rotationMode=BMM_Additive
    rotationSpace=BCS_BoneSpace
    alphaBoolBlend {blendInTime=0.1, blendOption=Linear, blendOutTime=0.1}
    alphaScaleBiasClamp {bInterpResult=true, interpSpeedIncreasing=5}
    boneToModify {boneName=dyn_jacketBack_mstr_l}
    rotation {pitch=-15}
   }
   n106 = "Transform (Modify) Bone - Bone: dyn_jacketBack_mstr_r"!(ComponentPose=@n105, bAlphaBoolEnabled:bool=@n104, Translation="0,0,0", Rotation="-15,0,0", Scale="1,1,1") -> (Pose) {
    alphaInputType=Bool
    rotationMode=BMM_Additive
    rotationSpace=BCS_BoneSpace
    alphaBoolBlend {blendInTime=0.1, blendOption=Linear, blendOutTime=0.1}
    alphaScaleBiasClamp {bInterpResult=true, interpSpeedIncreasing=5}
    boneToModify {boneName=dyn_jacketBack_mstr_r}
    rotation {pitch=-15}
   }
   n107 = Get bIsSkydiveDiveMode()
   n108 = "Transform (Modify) Bone - Bone: dyn_jacketBack_mstr_l"(ComponentPose=@n106, bAlphaBoolEnabled=@n107, Translation="0,0,0", Rotation="-35,0,0", Scale="1,1,1") {
    alphaInputType=Bool
    rotationMode=BMM_Additive
    rotationSpace=BCS_BoneSpace
    alphaBoolBlend {blendInTime=0.3, blendOption=Linear, blendOutTime=0.3}
    alphaScaleBiasClamp {bInterpResult=true, interpSpeedIncreasing=5}
    boneToModify {boneName=dyn_jacketBack_mstr_l}
    rotation {pitch=-35}
   }
   n109 = "Transform (Modify) Bone - Bone: dyn_jacketBack_mstr_r"(ComponentPose=@n108, bAlphaBoolEnabled=@n107, Translation="0,0,0", Rotation="-30,0,0", Scale="1,1,1") {
    alphaInputType=Bool
    rotationMode=BMM_Additive
    rotationSpace=BCS_BoneSpace
    alphaBoolBlend {blendInTime=0.3, blendOption=Linear, blendOutTime=0.3}
    alphaScaleBiasClamp {bInterpResult=true, interpSpeedIncreasing=5}
    boneToModify {boneName=dyn_jacketBack_mstr_r}
    rotation {pitch=-30}
   }
   n110 = "Transform (Modify) Bone - Bone: dyn_jacketSide_mstr_l"!(ComponentPose=@n109, bAlphaBoolEnabled:bool=@n107, Translation="0,0,0", Rotation="-10,10,0", Scale="1,1,1") -> (Pose) {
    alphaInputType=Bool
    rotationMode=BMM_Additive
    rotationSpace=BCS_BoneSpace
    alphaBoolBlend {blendInTime=0.3, blendOption=Linear, blendOutTime=0.3}
    alphaScaleBiasClamp {bInterpResult=true, interpSpeedIncreasing=5}
    boneToModify {boneName=dyn_jacketSide_mstr_l}
    rotation {pitch=-10, yaw=10}
   }
   n111 = "Transform (Modify) Bone - Bone: dyn_jacketSide_mstr_r"!(ComponentPose=@n110, bAlphaBoolEnabled:bool=@n107, Translation="0,0,0", Rotation="-10,-10,0", Scale="1,1,1") -> (Pose) {
    alphaInputType=Bool
    rotationMode=BMM_Additive
    rotationSpace=BCS_BoneSpace
    alphaBoolBlend {blendInTime=0.3, blendOption=Linear, blendOutTime=0.3}
    alphaScaleBiasClamp {bInterpResult=true, interpSpeedIncreasing=5}
    boneToModify {boneName=dyn_jacketSide_mstr_r}
    rotation {pitch=-10, yaw=-10}
   }
   n112 = Get bIsTacticalSprinting()
   n113 = "Transform (Modify) Bone - Bone: dyn_jacketBack_mstr_l"(ComponentPose=@n111, bAlphaBoolEnabled=@n112, Translation="0,0,0", Rotation="15,0,0", Scale="1,1,1") {
    alphaInputType=Bool
    rotationMode=BMM_Additive
    rotationSpace=BCS_BoneSpace
    alphaBoolBlend {blendInTime=0.1, blendOption=Linear, blendOutTime=0.3}
    alphaScaleBiasClamp {bInterpResult=true, interpSpeedIncreasing=5}
    boneToModify {boneName=dyn_jacketBack_mstr_l}
    rotation {pitch=15}
   }
   n114 = "Transform (Modify) Bone - Bone: dyn_jacketBack_mstr_r"(ComponentPose=@n113, bAlphaBoolEnabled=@n112, Translation="0,0,0", Rotation="15,0,0", Scale="1,1,1") {
    alphaInputType=Bool
    rotationMode=BMM_Additive
    rotationSpace=BCS_BoneSpace
    alphaBoolBlend {blendInTime=0.1, blendOption=Linear, blendOutTime=0.3}
    alphaScaleBiasClamp {bInterpResult=true, interpSpeedIncreasing=5}
    boneToModify {boneName=dyn_jacketBack_mstr_r}
    rotation {pitch=15}
   }
   n115 = Get bIsSliding()
   n116 = "Transform (Modify) Bone - Bone: dyn_jacketSide_mstr_l"(ComponentPose=@n114, bAlphaBoolEnabled=@n115, Translation="0,0,0", Rotation="-35,10,0", Scale="1,1,1") {
    alphaInputType=Bool
    rotationMode=BMM_Additive
    rotationSpace=BCS_BoneSpace
    alphaBoolBlend {blendInTime=0.3, blendOption=Linear, blendOutTime=0.3}
    alphaScaleBiasClamp {bInterpResult=true, interpSpeedIncreasing=5}
    boneToModify {boneName=dyn_jacketSide_mstr_l}
    rotation {pitch=-35, yaw=10}
   }
   n117 = "Transform (Modify) Bone - Bone: dyn_jacketSide_mstr_r"(ComponentPose=@n116, bAlphaBoolEnabled=@n115, Translation="0,0,0", Rotation="-25,0,0", Scale="1,1,1") {
    alphaInputType=Bool
    rotationMode=BMM_Additive
    rotationSpace=BCS_BoneSpace
    alphaBoolBlend {blendInTime=0.3, blendOption=Linear, blendOutTime=0.3}
    alphaScaleBiasClamp {bInterpResult=true, interpSpeedIncreasing=5}
    boneToModify {boneName=dyn_jacketSide_mstr_r}
    rotation {pitch=-25}
   }
   n118 = "Transform (Modify) Bone - Bone: dyn_jacketFront_mstr_l"(ComponentPose=@n117, bAlphaBoolEnabled=@n115, Translation="0,0,0", Rotation="-15,10,0", Scale="1,1,1") {
    alphaInputType=Bool
    rotationMode=BMM_Additive
    rotationSpace=BCS_BoneSpace
    alphaBoolBlend {blendInTime=0.3, blendOption=Linear, blendOutTime=0.3}
    alphaScaleBiasClamp {bInterpResult=true, interpSpeedIncreasing=5}
    boneToModify {boneName=dyn_jacketFront_mstr_l}
    rotation {pitch=-15, yaw=10}
   }
   n119 = "Transform (Modify) Bone - Bone: dyn_jacketFront_mstr_r"(ComponentPose=@n118, bAlphaBoolEnabled=@n115, Translation="0,0,0", Rotation="-20,10,0", Scale="1,1,1") {
    alphaInputType=Bool
    rotationMode=BMM_Additive
    rotationSpace=BCS_BoneSpace
    alphaBoolBlend {blendInTime=0.3, blendOption=Linear, blendOutTime=0.3}
    alphaScaleBiasClamp {bInterpResult=true, interpSpeedIncreasing=5}
    boneToModify {boneName=dyn_jacketFront_mstr_r}
    rotation {pitch=-20, yaw=10}
   }
   n120 = Get still_crouch()
   n121 = "Transform (Modify) Bone - Bone: dyn_jacketBack_mstr_r"(ComponentPose=@n119, bAlphaBoolEnabled=@n120, Translation="0,0,0", Rotation="-15,-15,35", Scale="1,1,1") {
    alphaInputType=Bool
    rotationMode=BMM_Additive
    rotationSpace=BCS_BoneSpace
    scaleMode=BMM_Additive
    scaleSpace=BCS_BoneSpace
    translationMode=BMM_Additive
    translationSpace=BCS_BoneSpace
    alphaBoolBlend {blendInTime=0.1, blendOption=Linear, blendOutTime=0.5}
    boneToModify {boneName=dyn_jacketBack_mstr_r}
    rotation=<rot -15 -15 35>
   }
   n122 = "Transform (Modify) Bone - Bone: dyn_jacketBack_01_r"(ComponentPose=@n121, bAlphaBoolEnabled=@n120, Translation="0,0,0", Rotation="-0.000,5.000,-0.002", Scale="1,1,1") {
    alphaInputType=Bool
    rotationMode=BMM_Additive
    rotationSpace=BCS_BoneSpace
    scaleMode=BMM_Additive
    scaleSpace=BCS_BoneSpace
    translationMode=BMM_Additive
    translationSpace=BCS_BoneSpace
    alphaBoolBlend {blendInTime=0.1, blendOption=Linear, blendOutTime=0.5}
    boneToModify {boneName=dyn_jacketBack_01_r}
    rotation=<rot -0.0001 5 -0.002>
   }
   n123 = "Transform (Modify) Bone - Bone: dyn_jacketBack_02_r"(ComponentPose=@n122, bAlphaBoolEnabled=@n120, Translation="0,0,0", Rotation="10.011,0.003,0.047", Scale="1,1,1") {
    alphaInputType=Bool
    rotationMode=BMM_Additive
    rotationSpace=BCS_BoneSpace
    scaleMode=BMM_Additive
    scaleSpace=BCS_BoneSpace
    translationMode=BMM_Additive
    translationSpace=BCS_BoneSpace
    alphaBoolBlend {blendInTime=0.1, blendOption=Linear, blendOutTime=0.5}
    boneToModify {boneName=dyn_jacketBack_02_r}
    rotation=<rot 10.011 0.003 0.047>
   }
   n124 = "Transform (Modify) Bone - Bone: dyn_jacketSide_01_r"(ComponentPose=@n123, bAlphaBoolEnabled=@n120, Translation="0,0,0", Rotation="-20.000,0.001,0.005", Scale="1,1,1") {
    alphaInputType=Bool
    rotationMode=BMM_Additive
    rotationSpace=BCS_BoneSpace
    scaleMode=BMM_Additive
    scaleSpace=BCS_BoneSpace
    translationMode=BMM_Additive
    translationSpace=BCS_BoneSpace
    alphaBoolBlend {blendInTime=0.1, blendOption=Linear, blendOutTime=0.5}
    boneToModify {boneName=dyn_jacketSide_01_r}
    rotation=<rot -20 0.001 0.005>
   }
   n125 = "Transform (Modify) Bone - Bone: dyn_jacketBack_04_r"!(ComponentPose=@n124, bAlphaBoolEnabled:bool=@n120, Translation="0,0,0", Rotation="-45,0,0", Scale="1,1,1") -> (Pose) {
    alphaInputType=Bool
    rotationMode=BMM_Additive
    rotationSpace=BCS_BoneSpace
    scaleMode=BMM_Additive
    scaleSpace=BCS_BoneSpace
    translationMode=BMM_Additive
    translationSpace=BCS_BoneSpace
    alphaBoolBlend {blendInTime=0.1, blendOption=Linear, blendOutTime=0.5}
    boneToModify {boneName=dyn_jacketBack_04_r}
    rotation {pitch=-45}
   }
   n126 = "Transform (Modify) Bone - Bone: dyn_jacketSide_02_r"!(ComponentPose=@n125, bAlphaBoolEnabled:bool=@n120, Translation="0,0,0", Rotation="-0.000,-0.035,0.053", Scale="1,1,1") -> (Pose) {
    alphaInputType=Bool
    rotationMode=BMM_Additive
    rotationSpace=BCS_BoneSpace
    scaleMode=BMM_Additive
    scaleSpace=BCS_BoneSpace
    translationMode=BMM_Additive
    translationSpace=BCS_BoneSpace
    alphaBoolBlend {blendInTime=0.1, blendOption=Linear, blendOutTime=0.5}
    boneToModify {boneName=dyn_jacketSide_02_r}
    rotation=<rot 0 -0.035 0.053>
   }
   n127 = "Transform (Modify) Bone - Bone: dyn_jacketFront_01_r"!(ComponentPose=@n126, bAlphaBoolEnabled:bool=@n120, Translation="0,0,0", Rotation="-10.000,-0.013,-0.024", Scale="1,1,1") -> (Pose) {
    alphaInputType=Bool
    rotationMode=BMM_Additive
    rotationSpace=BCS_BoneSpace
    scaleMode=BMM_Additive
    scaleSpace=BCS_BoneSpace
    translationMode=BMM_Additive
    translationSpace=BCS_BoneSpace
    alphaBoolBlend {blendInTime=0.1, blendOption=Linear, blendOutTime=0.5}
    boneToModify {boneName=dyn_jacketFront_01_r}
    rotation=<rot -10 -0.013 -0.024>
   }
   n128 = "Transform (Modify) Bone - Bone: dyn_jacketSide_04_r"!(ComponentPose=@n127, bAlphaBoolEnabled:bool=@n120, Translation="0,0,0", Rotation="55.000,-0.043,0.077", Scale="1,1,1") -> (Pose) {
    alphaInputType=Bool
    rotationMode=BMM_Additive
    rotationSpace=BCS_BoneSpace
    scaleMode=BMM_Additive
    scaleSpace=BCS_BoneSpace
    translationMode=BMM_Additive
    translationSpace=BCS_BoneSpace
    alphaBoolBlend {blendInTime=0.1, blendOption=Linear, blendOutTime=0.5}
    boneToModify {boneName=dyn_jacketSide_04_r}
    rotation=<rot 55 -0.043 0.077>
   }
   n129 = Get BwdMovement() -> (BwdMovement:float)
   n130 = Trail controller - Bone: dyn_jacketBack_04_r!(ComponentPose=@n128, Alpha=@n129, FakeVelocity:Vector="0,0,0") -> (Pose) {
    bActorSpaceFakeVel=true
    bShowTrailLocation=true
    chainLength=3
    debugLifeTime=0.1
    maxDeltaTime=0.0333
    relaxationSpeedScale=3
    alphaScaleBiasClamp {bInterpResult=true}
    baseJoint {boneName=root}
    planarLimits [
     {drivingBone {boneName=root}, planeTransform {location=<0 0 10>, rotation=<rot 0 0 0>, scale=<1 1 1>}}
     {drivingBone {boneName=spine_03}, planeTransform {location=<0 -20 0>, rotation=<rot 0 15 -89.9999>, scale=<1 1 1>}}
    ]
    rotationLimits [{limitMax=<180 180 180>}, {limitMax=<180 180 180>}, {limitMax=<180 180 180>}]
    rotationOffsets [<0 0 0>, <0 0 0>, <0 0 0>]
    trailBone {boneName=dyn_jacketBack_04_r}
    trailRelaxationSpeed {editorCurveData {keys [{arriveTangent=-72.0739, interpMode=RCIM_Cubic, leaveTangent=-72.0739, tangentMode=RCTM_User, time=0, value=35}, <at 1 5 RCIM_Cubic>]}}
   }
   n131 = Trail controller - Bone: dyn_jacketBack_05_l!(ComponentPose=@n130, Alpha=@n129, FakeVelocity:Vector="0,0,0") -> (Pose) {
    bActorSpaceFakeVel=true
    bShowTrailLocation=true
    chainLength=4
    debugLifeTime=0.1
    maxDeltaTime=0.0333
    relaxationSpeedScale=3
    alphaScaleBiasClamp {bInterpResult=true}
    baseJoint {boneName=root}
    planarLimits [
     {drivingBone {boneName=root}, planeTransform {location=<0 0 10>, rotation=<rot 0 0 0>, scale=<1 1 1>}}
     {drivingBone {boneName=spine_03}, planeTransform {location=<0 -20 0>, rotation=<rot 0 15 -89.9999>, scale=<1 1 1>}}
    ]
    rotationLimits [{limitMax=<180 180 180>}, {limitMax=<180 180 180>}, {limitMax=<180 180 180>}, {limitMax=<180 180 180>}]
    rotationOffsets [<0 0 0>, <0 0 0>, <0 0 0>, <0 0 0>]
    trailBone {boneName=dyn_jacketBack_05_l}
    trailRelaxationSpeed {editorCurveData {keys [{arriveTangent=-72.0739, interpMode=RCIM_Cubic, leaveTangent=-72.0739, tangentMode=RCTM_User, time=0, value=35}, <at 1 5 RCIM_Cubic>]}}
   }
   n132 = Trail controller - Bone: dyn_jacketSide_04_r!(ComponentPose=@n131, Alpha=Get LeftMovement() -> (LeftMovement:float), FakeVelocity:Vector="0,0,0") -> (Pose) {
    bActorSpaceFakeVel=true
    bShowTrailLocation=true
    chainLength=3
    debugLifeTime=0.1
    maxDeltaTime=0.0333
    relaxationSpeedScale=3
    alphaScaleBiasClamp {bInterpResult=true}
    baseJoint {boneName=root}
    planarLimits [
     {drivingBone {boneName=root}, planeTransform {location=<0 0 10>, rotation=<rot 0 0 0>, scale=<1 1 1>}}
     {drivingBone {boneName=spine_03}, planeTransform {location=<0 -20 0>, rotation=<rot 0 15 -89.9999>, scale=<1 1 1>}}
    ]
    rotationLimits [{limitMax=<180 180 180>}, {limitMax=<180 180 180>}, {limitMax=<180 180 180>}]
    rotationOffsets [<0 0 0>, <0 0 0>, <0 0 0>]
    trailBone {boneName=dyn_jacketSide_04_r}
    trailRelaxationSpeed {editorCurveData {keys [{arriveTangent=-72.0739, interpMode=RCIM_Cubic, leaveTangent=-72.0739, tangentMode=RCTM_User, time=0, value=35}, <at 1 5 RCIM_Cubic>]}}
   }
   n133 = Trail controller - Bone: dyn_jacketSide_05_l!(ComponentPose=@n132, Alpha=Get RightMovement() -> (RightMovement:float), FakeVelocity:Vector="0,0,0") -> (Pose) {
    bActorSpaceFakeVel=true
    bShowTrailLocation=true
    chainLength=4
    debugLifeTime=0.1
    maxDeltaTime=0.0333
    relaxationSpeedScale=3
    alphaScaleBiasClamp {bInterpResult=true}
    baseJoint {boneName=root}
    planarLimits [
     {drivingBone {boneName=root}, planeTransform {location=<0 0 10>, rotation=<rot 0 0 0>, scale=<1 1 1>}}
     {drivingBone {boneName=spine_03}, planeTransform {location=<0 -20 0>, rotation=<rot 0 15 -89.9999>, scale=<1 1 1>}}
    ]
    rotationLimits [{limitMax=<180 180 180>}, {limitMax=<180 180 180>}, {limitMax=<180 180 180>}, {limitMax=<180 180 180>}]
    rotationOffsets [<0 0 0>, <0 0 0>, <0 0 0>, <0 0 0>]
    trailBone {boneName=dyn_jacketSide_05_l}
    trailRelaxationSpeed {editorCurveData {keys [{arriveTangent=-72.0739, interpMode=RCIM_Cubic, leaveTangent=-72.0739, tangentMode=RCTM_User, time=0, value=35}, <at 1 5 RCIM_Cubic>]}}
   }
   Save cached pose 'rban'(Pose=Blend Poses by bool(BlendPose_0=Use cached pose 'presim'(), BlendPose_1=Component To Local(ComponentPose=@n133), bActiveValue=Get bRigidBodyEnabled()) {blendType=Linear}) {cacheName=rban}
   Use cached pose 'presim'()
   Get bIsJumping()
  ]}
 EventGraph {nodes [
   Event BlueprintUpdateAnimation()
   > n1 = Sequence!(execute) -> (then_0, then_1, then_2:Exec, then_3:Exec, then_4:Exec, then_5:Exec, then_6:Exec, then_7:Exec)
   n2 = Get bIsMoving2D()
   > n3 = Branch(Condition=AND Boolean(A=Get bIsCrouching(), B=NOT Boolean(A=@n2)))
   > Set still_crouch(still_crouch:bool=true) -> (Output_Get:bool)
   Set still_crouch(execute=@n3.else, still_crouch=false)
   n4 = Get bIsCrouching()
   n5 = Get bIsMoving2D()
   n6 = Branch(execute=@n1.then_1, Condition=AND Boolean(A=NOT Boolean(A=@n5), B=NOT Boolean(A=@n4)))
   > Set back skirt standing(back skirt standing:bool=true) -> (Output_Get:bool)
   Set back skirt standing(execute=@n6.else, back skirt standing=false)
   n7 = Gravity Override State Function DEPRECATED(
    execute=@n1.then_2
    CharacterPartAnimInstance=Self-Reference()
    GameTimeInSeconds=GetGameTimeInSeconds()
    GravityOverrideParams=Get GravityOverrideParams() -> (GravityOverrideParams:Gravity Override Params Struct Structure)
   )
   > Set GravityOverride(GravityOverride=@n7)
   > n8 = Gravity Override State Function DEPRECATED(
    CharacterPartAnimInstance=Self-Reference()
    GameTimeInSeconds=GetGameTimeInSeconds()
    GravityOverrideParams=Get Gravity Override Params Chain() -> (Gravity Override Params Chain:Gravity Override Params Struct Structure)
   )
   > Set GravityOverride_chain(GravityOverride_chain:Vector=@n8) -> (Output_Get:Vector)
   n9 = Get Skel_Mesh()
   > Set thigh_rot_otr_l(thigh_rot_otr_l:double=GetAlphaFromReferenceTransformDelta(SkeletalMeshComponent=@n9, BoneName=thigh_l, BaseName=pelvis, DeltaComponentType=EulerY, InputRange=(X=20,Y=50), TwistAxis="1,0,0")) -> (Output_Get:double)
   > Set thigh_rot_otr_r(thigh_rot_otr_r:double=GetAlphaFromReferenceTransformDelta(SkeletalMeshComponent=@n9, BoneName=thigh_r, BaseName=pelvis, DeltaComponentType=EulerY, InputRange=(X=20,Y=50), TwistAxis="1,0,0")) -> (Output_Get:double)
   n10 = Get Skel_Mesh()
   > Set thigh_rot_inr_l(thigh_rot_inr_l:double=GetAlphaFromReferenceTransformDelta(SkeletalMeshComponent=@n10, BoneName=thigh_l, BaseName=pelvis, DeltaComponentType=EulerY, InputRange=(X=10,Y=-6), TwistAxis="1,0,0")) -> (Output_Get:double)
   > Set thigh_rot_inr_r(thigh_rot_inr_r:double=GetAlphaFromReferenceTransformDelta(SkeletalMeshComponent=@n10, BoneName=thigh_r, BaseName=pelvis, DeltaComponentType=EulerY, InputRange=(X=0,Y=-10), TwistAxis="1,0,0")) -> (Output_Get:double)
   > Set ThighRotR(ThighRotR:double=GetAlphaFromReferenceTransformDelta(SkeletalMeshComponent=Get Skel_Mesh(), BoneName=thigh_r, BaseName=pelvis, InputRange=(X=-20,Y=-100), TwistAxis="1,0,0")) -> (Output_Get:double)
   n11 = Branch(
    execute=@n1.then_3
    Condition=OR Boolean!(A=Get bIsJumping(), B=Get bIsFalling(), C:bool=Get bIsDBNO(), D:bool=Get bIsSliding()) -> (ReturnValue)
   )
   > Set FloorCOL(FloorCOL:bool=true) -> (Output_Get:bool)
   Set FloorCOL(execute=@n11.else, FloorCOL=false)
   n12 = CheckForBackbling()
   n13 = Branch(execute=@n1.then_4, Condition=@n12)
   > Set AccScale(AccScale:Vector="0, 0, 0") -> (Output_Get:Vector)
   > Set VelScale(VelScale:Vector="0, 0, 0") -> (Output_Get:Vector)
   Set AccScale(execute=@n13.else, AccScale="0.3,0.25,0.25")
   > Set VelScale(VelScale="0.25,0.25,0.25")
   n14 = Branch(execute=@n1.then_5, Condition=Get bIsCrouching())
   > Set strafe(strafe:double=0) -> (Output_Get:double)
   n15 = PlayerMovement(execute=@n14.else, AnimInstance=Self-Reference(), XDotProductCoefficient=1, YDotProductCoefficient=1, ZDotProductCoefficient=1)
   > Set strafe(strafe="Absolute (Float)"(A=@n15))
   n16 = PlayerMovement(execute=@n1.then_6, AnimInstance=Self-Reference(), XDotProductCoefficient=1, YDotProductCoefficient=1, ZDotProductCoefficient=1)
   n17 = MapRangeClamped(Value=@n16.YMovement, InRangeA=0, InRangeB=-1)
   n18 = float * float(A="Absolute (Float)"(A=@n16), B=0.5)
   > Set BwdMovement(BwdMovement:float=MapRangeClamped(
     Value=float + float(A=@n18, B=@n17) {is=title K2Node_CommutativeAssociativeBinaryOperator}
     InRangeA=0
     InRangeB=1.1
     OutRangeA=0.4
     OutRangeB=0
    )) -> (Output_Get:float)
   n19 = MapRangeClamped(Value=@n16, InRangeA=-0.2, InRangeB=1)
   n20 = float * float(A="Absolute (Float)"(A=@n16.YMovement), B=0.2)
   > Set RightMovement(RightMovement:float=MapRangeClamped(Value=float + float(A=@n20, B=@n19), InRangeA=0, InRangeB=1.2, OutRangeB=0.5)) -> (Output_Get:float)
   n21 = MapRangeClamped(Value=@n16, InRangeA=0.2, InRangeB=-1, OutRangeB=0.3)
   n22 = float * float(A="Absolute (Float)"(A=@n16.YMovement), B=0.2)
   > Set LeftMovement(LeftMovement:float=MapRangeClamped(Value=float + float(A=@n22, B=@n21), InRangeA=0, InRangeB=1.2, OutRangeB=0.5)) -> (Output_Get:float)
   n23 = Branch(execute=@n1.then_7, Condition=Get bIsSkyDiving())
   > n24 = Branch(Condition="float > float"(A=Get PawnVelocityZ(), B=50) {is=title K2Node_CallFunction})
   > Set BwdMovement(BwdMovement=float * float(A=Get BwdMovement(), B=0))
   Set BwdMovement(execute=@n24.else, BwdMovement=0.2)
   Set BwdMovement(execute=@n23.else, BwdMovement=float * float(A=Get BwdMovement(), B=1))
   n25 = Event OnBeginFalling()
   Set jumping(execute=[@n25.then, @n26.then], jumping:double=1) -> (Output_Get:double)
   n26 = Event OnBeginJumping()
   Event BlueprintInitializeAnimation()
   > Set Skel_Mesh(Skel_Mesh=GetOwnerSkeletalMeshComponent())
   n27 = Event OnEndFalling()
   Set jumping(execute=[@n27.then, @n28.then], jumping=0)
   n28 = Event OnEndJumping()
   OR Boolean!(A=@n12, B=Get bIsSkyDiving()) -> (ReturnValue)
  ]}
]
variables [
 AccScale {type=Vector}
 back skirt standing {type=boolean}
 BwdMovement {type=number}
 FloorCOL {type=boolean}
 Gravity Override Params Chain {
  type=GravityOverrideParamsStruct
  default {
   skydive_FloatingDown {gravityOverride=<-600 -700 -480>, windAmplitude=<50 -150 100>, windFrequency=<0.4 1.5 2>, windOffset=<0 -150 100>}
   skydive_GoingUp {gravityOverride=<-300 -300 -120>, windAmplitude=<100 -250 200>, windFrequency=<0.75 3.5 4>}
   skydive_Parachuting {gravityOverride=<-350 -250 -980>, windAmplitude=<25 -50 50>, windFrequency=<0.2 0.75 1>}
  }
 }
 GravityOverride {type=Vector}
 GravityOverride_chain {type=Vector}
 GravityOverrideParams {type=GravityOverrideParamsStruct}
 jumping {type=number}
 LeftMovement {type=number}
 RightMovement {type=number}
 Skel_Mesh {type=SkeletalMeshComponent}
 still_crouch {type=boolean}
 strafe {type=number}
 thigh_rot_inr_l {type=number}
 thigh_rot_inr_r {type=number}
 thigh_rot_otr_l {type=number}
 thigh_rot_otr_r {type=number}
 ThighRotR {type=number}
 VelScale {type=Vector}
]
