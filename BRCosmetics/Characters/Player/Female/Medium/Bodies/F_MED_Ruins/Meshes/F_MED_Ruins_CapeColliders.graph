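# Physics asset for the female medium Ruins body cosmetic, named as cape colliders.
# Capsules cover pelvis, spine, clavicles, head and feet, while arm, leg and neck bodies have no shapes.
# Jacket back and side chains hang from the pelvis with identical capsules, one link longer on the left; an arm-hook chain hangs from the right clavicle.
# Constraints link the bodies along these chains with free swing and twist.
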PhysicsAsset /BRCosmetics/Characters/Player/Female/Medium/Bodies/F_MED_Ruins/Meshes/F_MED_Ruins_CapeColliders
bodies {
 calf_l []
 calf_r []
 clavicle_l [{shapeType=Capsule, length=7.6964, radius=6.3853, center=<4.0147 -0.8354 -1.7419>, rotation=<rot -77.5566 177.3915 7.7314>}]
 clavicle_r [{shapeType=Capsule, length=7.6964, radius=8.2957, center=<-4.0147 0.8354 1.7419>, rotation=<rot -77.5565 177.3915 -172.2685>}]
 dyn_arm_hook_01 [{
   shapeType=Capsule
   length=1.6823
   radius=2.3918
   shapeName=dyn_arm_hook_01_capsule
   center=<1.8508 -0.1456 -0.004>
   rotation=<rot 0 80.9419 -82.2322>
  }]
 dyn_arm_hook_02 [{
   shapeType=Capsule
   length=1.7428
   radius=2.4205
   shapeName=dyn_arm_hook_02_capsule
   center=<1.971 -0.145 0.0729>
   rotation=<rot 0 -95.1395 -85.5176>
  }]
 dyn_arm_hook_03 [{
   shapeType=Capsule
   length=1.8119
   radius=2.4795
   shapeName=dyn_arm_hook_03_capsule
   center=<1.7779 -0.1239 0.0159>
   rotation=<rot 0 97.8143 -87.8202>
  }]
 dyn_arm_hook_04 [{
   shapeType=Capsule
   length=1.7463
   radius=2.4223
   shapeName=dyn_arm_hook_04_capsule
   center=<1.8209 -0.0475 0.0074>
   rotation=<rot 0 -92.6559 -84.0622>
  }]
 dyn_arm_hook_05 [{
   shapeType=Capsule
   length=1.7559
   radius=2.4297
   shapeName=dyn_arm_hook_05_capsule
   center=<1.7663 0.0067 -0.0814>
   rotation=<rot 0 96.5513 -87.8596>
  }]
 dyn_arm_hook_06 [{
   shapeType=Capsule
   length=1.7713
   radius=2.4609
   shapeName=dyn_arm_hook_06_capsule
   center=<1.8265 0.0203 -0.0022>
   rotation=<rot 0 -93.743 -88.9789>
  }]
 dyn_arm_hook_07 [{
   shapeType=Capsule
   length=1.7595
   radius=2.4646
   shapeName=dyn_arm_hook_07_capsule
   center=<1.7626 0.0671 0.0321>
   rotation=<rot 0 96.9454 -87.624>
  }]
 dyn_arm_hook_08 [{
   shapeType=Capsule
   length=3.7006
   radius=6.6036
   shapeName=dyn_arm_hook_08_capsule
   center=<5.7208 0.0753 -0.4472>
   rotation=<rot 0 -90.9639 -85.8998>
  }]
 dyn_arm_hook_MSTR [{shapeType=Capsule, length=0.505, radius=0.505, shapeName=dyn_arm_hook_MSTR_capsule, rotation=<rot 0 -90 0>}]
 dyn_jacketBack_01_l [{shapeType=Capsule, length=12, radius=12, shapeName=dyn_jacketBack_01_l_capsule, center=<8 0 -5>, rotation=<rot 90 0 0>}]
 dyn_jacketBack_01_r [{shapeType=Capsule, length=12, radius=12, shapeName=dyn_jacketBack_01_r_capsule, center=<8 0 -5>, rotation=<rot 90 0 0>}]
 dyn_jacketBack_02_l [{shapeType=Capsule, length=12, radius=12, shapeName=dyn_jacketBack_02_l_capsule, center=<8 0 -5>, rotation=<rot 90 0 0>}]
 dyn_jacketBack_02_r [{shapeType=Capsule, length=12, radius=12, shapeName=dyn_jacketBack_02_r_capsule, center=<8 0 -5>, rotation=<rot 90 0 0>}]
 dyn_jacketBack_03_l [{shapeType=Capsule, length=12, radius=12, shapeName=dyn_jacketBack_03_l_capsule, center=<8 0 -5>, rotation=<rot 90 0 0>}]
 dyn_jacketBack_03_r [{shapeType=Capsule, length=12, radius=12, shapeName=dyn_jacketBack_03_r_capsule, center=<8 0 -5>, rotation=<rot 90 0 0>}]
 dyn_jacketBack_04_l [{shapeType=Capsule, length=12, radius=12, shapeName=dyn_jacketBack_04_l_capsule, center=<8 0 -5>, rotation=<rot 90 0 0>}]
 dyn_jacketBack_04_r [{shapeType=Capsule, length=12, radius=12, shapeName=dyn_jacketBack_04_r_capsule, center=<8 0 -5>, rotation=<rot 90 0 0>}]
 dyn_jacketBack_05_l [{shapeType=Capsule, length=12, radius=12, shapeName=dyn_jacketBack_05_l_capsule, center=<8 0 -5>, rotation=<rot 90 0 0>}]
 dyn_jacketBack_mstr_l [{shapeType=Capsule, length=12, radius=12, shapeName=dyn_jacketBack_mstr_l_capsule, center=<8 0 -5>, rotation=<rot 90 0 0>}]
 dyn_jacketBack_mstr_r [{shapeType=Capsule, length=12, radius=12, shapeName=dyn_jacketBack_mstr_r_capsule, center=<8 0 -5>, rotation=<rot 90 0 0>}]
 dyn_jacketSide_01_l [{shapeType=Capsule, length=12, radius=12, shapeName=dyn_jacketSide_01_l_capsule, center=<8 0 -5>, rotation=<rot 90 0 0>}]
 dyn_jacketSide_01_r [{shapeType=Capsule, length=12, radius=12, shapeName=dyn_jacketSide_01_r_capsule, center=<8 0 -5>, rotation=<rot 90 0 0>}]
 dyn_jacketSide_02_l [{shapeType=Capsule, length=12, radius=12, shapeName=dyn_jacketSide_02_l_capsule, center=<8 0 -5>, rotation=<rot 90 0 0>}]
 dyn_jacketSide_02_r [{shapeType=Capsule, length=12, radius=12, shapeName=dyn_jacketSide_02_r_capsule, center=<8 0 -5>, rotation=<rot 90 0 0>}]
 dyn_jacketSide_03_l [{shapeType=Capsule, length=12, radius=12, shapeName=dyn_jacketSide_03_l_capsule, center=<8 0 -5>, rotation=<rot 90 0 0>}]
 dyn_jacketSide_03_r [{shapeType=Capsule, length=12, radius=12, shapeName=dyn_jacketSide_03_r_capsule, center=<8 0 -5>, rotation=<rot 90 0 0>}]
 dyn_jacketSide_04_l [{shapeType=Capsule, length=12, radius=12, shapeName=dyn_jacketSide_04_l_capsule, center=<8 0 -5>, rotation=<rot 90 0 0>}]
 dyn_jacketSide_04_r [{shapeType=Capsule, length=12, radius=12, shapeName=dyn_jacketSide_04_r_capsule, center=<8 0 -5>, rotation=<rot 90 0 0>}]
 dyn_jacketSide_05_l [{shapeType=Capsule, length=12, radius=12, shapeName=dyn_jacketSide_05_l_capsule, center=<8 0 -5>, rotation=<rot 90 0 0>}]
 dyn_jacketSide_mstr_l [{shapeType=Capsule, length=12, radius=12, shapeName=dyn_jacketSide_mstr_l_capsule, center=<8 0 -5>, rotation=<rot 90 0 0>}]
 dyn_jacketSide_mstr_r [{shapeType=Capsule, length=12, radius=12, shapeName=dyn_jacketSide_mstr_r_capsule, center=<8 0 -5>, rotation=<rot 90 0 0>}]
 foot_l [{shapeType=Capsule, length=12.5982, radius=6.5127, center=<-2.2072 6.0534 -0.3367>, rotation=<rot 0.0388 1.4701 -89.3982>}]
 foot_r [{shapeType=Capsule, length=12.5982, radius=6.5127, center=<2.2072 -6.0534 0.3367>, rotation=<rot 0.0388 1.4701 90.6018>}]
 head [{shapeType=Capsule, length=0, radius=18, center=<9.0697 0.0746 0>, rotation=<rot 0 -90 0>}]
 lowerarm_l []
 lowerarm_r []
 neck_02 []
 pelvis [{shapeType=Capsule, length=9.4028, radius=10.6596, center=<-1.9544 -2.2029 -0.1981>}]
 spine_01 [{shapeType=Capsule, length=9.0171, radius=9.7354, center=<-2.8183 -1.0873 0>}]
 spine_02 [
  {shapeType=Capsule, length=3.2984, radius=8.4854, center=<1.179 2.2974 0>}
  {shapeType=Capsule, length=20.6875, radius=9.3125, center=<-1.8317 2.606 0>, rotation=<rot 90 0 0>}
 ]
 spine_03 [
  {shapeType=Capsule, length=3.2984, radius=8.4854, center=<-0.0747 2.3725 0>}
  {shapeType=Capsule, length=20.7188, radius=9.2812, center=<-1.2326 2.6059 0>, rotation=<rot 90 0 0>}
 ]
 spine_04 [
  {shapeType=Capsule, length=5.0796, radius=11.1416, center=<5.5182 1.4112 0>}
  {shapeType=Capsule, length=8.4062, radius=10.9062, center=<-1.2326 1.308 0>, rotation=<rot 90 0 0>}
 ]
 spine_05 [{shapeType=Capsule, length=10.9859, radius=10.0791, center=<5.6377 0.3227 0>}]
 thigh_l []
 thigh_r []
 upperarm_l []
 upperarm_r []
}
constraints [
 {
  bone1Name=spine_01
  bone2Name=pelvis
  swing1LimitDegrees=45
  swing1Motion=Free
  swing2LimitDegrees=45
  swing2Motion=Free
  twistLimitDegrees=45
  twistMotion=Free
 }
 {
  bone1Name=spine_02
  bone2Name=spine_01
  swing1LimitDegrees=45
  swing1Motion=Free
  swing2LimitDegrees=45
  swing2Motion=Free
  twistLimitDegrees=45
  twistMotion=Free
 }
 {
  bone1Name=spine_03
  bone2Name=spine_02
  swing1LimitDegrees=45
  swing1Motion=Free
  swing2LimitDegrees=45
  swing2Motion=Free
  twistLimitDegrees=45
  twistMotion=Free
 }
 {
  bone1Name=spine_04
  bone2Name=spine_03
  swing1LimitDegrees=45
  swing1Motion=Free
  swing2LimitDegrees=45
  swing2Motion=Free
  twistLimitDegrees=45
  twistMotion=Free
 }
 {
  bone1Name=spine_05
  bone2Name=spine_04
  swing1LimitDegrees=45
  swing1Motion=Free
  swing2LimitDegrees=45
  swing2Motion=Free
  twistLimitDegrees=45
  twistMotion=Free
 }
 {
  bone1Name=clavicle_l
  bone2Name=spine_05
  swing1LimitDegrees=45
  swing1Motion=Free
  swing2LimitDegrees=45
  swing2Motion=Free
  twistLimitDegrees=45
  twistMotion=Free
 }
 {
  bone1Name=upperarm_l
  bone2Name=clavicle_l
  swing1LimitDegrees=45
  swing1Motion=Free
  swing2LimitDegrees=45
  swing2Motion=Free
  twistLimitDegrees=45
  twistMotion=Free
 }
 {
  bone1Name=lowerarm_l
  bone2Name=upperarm_l
  swing1LimitDegrees=45
  swing1Motion=Free
  swing2LimitDegrees=45
  swing2Motion=Free
  twistLimitDegrees=45
  twistMotion=Free
 }
 {
  bone1Name=neck_02
  bone2Name=spine_05
  swing1LimitDegrees=45
  swing1Motion=Free
  swing2LimitDegrees=45
  swing2Motion=Free
  twistLimitDegrees=45
  twistMotion=Free
 }
 {
  bone1Name=head
  bone2Name=neck_02
  swing1LimitDegrees=45
  swing1Motion=Free
  swing2LimitDegrees=45
  swing2Motion=Free
  twistLimitDegrees=45
  twistMotion=Free
 }
 {
  bone1Name=calf_l
  bone2Name=thigh_l
  swing1LimitDegrees=45
  swing1Motion=Free
  swing2LimitDegrees=45
  swing2Motion=Free
  twistLimitDegrees=45
  twistMotion=Free
 }
 {
  bone1Name=thigh_l
  bone2Name=pelvis
  swing1LimitDegrees=45
  swing1Motion=Free
  swing2LimitDegrees=45
  swing2Motion=Free
  twistLimitDegrees=45
  twistMotion=Free
 }
 {
  bone1Name=foot_l
  bone2Name=calf_l
  swing1LimitDegrees=45
  swing1Motion=Free
  swing2LimitDegrees=45
  swing2Motion=Free
  twistLimitDegrees=45
  twistMotion=Free
 }
 {
  bone1Name=clavicle_r
  bone2Name=spine_05
  swing1LimitDegrees=45
  swing1Motion=Free
  swing2LimitDegrees=45
  swing2Motion=Free
  twistLimitDegrees=45
  twistMotion=Free
 }
 {
  bone1Name=lowerarm_r
  bone2Name=upperarm_r
  swing1LimitDegrees=45
  swing1Motion=Free
  swing2LimitDegrees=45
  swing2Motion=Free
  twistLimitDegrees=45
  twistMotion=Free
 }
 {
  bone1Name=thigh_r
  bone2Name=pelvis
  swing1LimitDegrees=45
  swing1Motion=Free
  swing2LimitDegrees=45
  swing2Motion=Free
  twistLimitDegrees=45
  twistMotion=Free
 }
 {
  bone1Name=calf_r
  bone2Name=thigh_r
  swing1LimitDegrees=45
  swing1Motion=Free
  swing2LimitDegrees=45
  swing2Motion=Free
  twistLimitDegrees=45
  twistMotion=Free
 }
 {
  bone1Name=foot_r
  bone2Name=calf_r
  swing1LimitDegrees=45
  swing1Motion=Free
  swing2LimitDegrees=45
  swing2Motion=Free
  twistLimitDegrees=45
  twistMotion=Free
 }
 {
  bone1Name=upperarm_r
  bone2Name=clavicle_r
  swing1LimitDegrees=45
  swing1Motion=Free
  swing2LimitDegrees=45
  swing2Motion=Free
  twistLimitDegrees=45
  twistMotion=Free
 }
 {
  bone1Name=dyn_jacketBack_mstr_l
  bone2Name=pelvis
  swing1LimitDegrees=45
  swing1Motion=Free
  swing2LimitDegrees=45
  swing2Motion=Free
  twistLimitDegrees=45
  twistMotion=Free
 }
 {
  bone1Name=dyn_jacketBack_01_l
  bone2Name=dyn_jacketBack_mstr_l
  swing1LimitDegrees=45
  swing1Motion=Free
  swing2LimitDegrees=45
  swing2Motion=Free
  twistLimitDegrees=45
  twistMotion=Free
 }
 {
  bone1Name=dyn_jacketBack_02_l
  bone2Name=dyn_jacketBack_01_l
  swing1LimitDegrees=45
  swing1Motion=Free
  swing2LimitDegrees=45
  swing2Motion=Free
  twistLimitDegrees=45
  twistMotion=Free
 }
 {
  bone1Name=dyn_jacketBack_03_l
  bone2Name=dyn_jacketBack_02_l
  swing1LimitDegrees=45
  swing1Motion=Free
  swing2LimitDegrees=45
  swing2Motion=Free
  twistLimitDegrees=45
  twistMotion=Free
 }
 {
  bone1Name=dyn_jacketBack_04_l
  bone2Name=dyn_jacketBack_03_l
  swing1LimitDegrees=45
  swing1Motion=Free
  swing2LimitDegrees=45
  swing2Motion=Free
  twistLimitDegrees=45
  twistMotion=Free
 }
 {
  bone1Name=dyn_jacketBack_05_l
  bone2Name=dyn_jacketBack_04_l
  swing1LimitDegrees=45
  swing1Motion=Free
  swing2LimitDegrees=45
  swing2Motion=Free
  twistLimitDegrees=45
  twistMotion=Free
 }
 {
  bone1Name=dyn_jacketBack_mstr_r
  bone2Name=pelvis
  swing1LimitDegrees=45
  swing1Motion=Free
  swing2LimitDegrees=45
  swing2Motion=Free
  twistLimitDegrees=45
  twistMotion=Free
 }
 {
  bone1Name=dyn_jacketBack_01_r
  bone2Name=dyn_jacketBack_mstr_r
  swing1LimitDegrees=45
  swing1Motion=Free
  swing2LimitDegrees=45
  swing2Motion=Free
  twistLimitDegrees=45
  twistMotion=Free
 }
 {
  bone1Name=dyn_jacketBack_02_r
  bone2Name=dyn_jacketBack_01_r
  swing1LimitDegrees=45
  swing1Motion=Free
  swing2LimitDegrees=45
  swing2Motion=Free
  twistLimitDegrees=45
  twistMotion=Free
 }
 {
  bone1Name=dyn_jacketBack_03_r
  bone2Name=dyn_jacketBack_02_r
  swing1LimitDegrees=45
  swing1Motion=Free
  swing2LimitDegrees=45
  swing2Motion=Free
  twistLimitDegrees=45
  twistMotion=Free
 }
 {
  bone1Name=dyn_jacketBack_04_r
  bone2Name=dyn_jacketBack_03_r
  swing1LimitDegrees=45
  swing1Motion=Free
  swing2LimitDegrees=45
  swing2Motion=Free
  twistLimitDegrees=45
  twistMotion=Free
 }
 {
  bone1Name=dyn_jacketSide_mstr_l
  bone2Name=pelvis
  swing1LimitDegrees=45
  swing1Motion=Free
  swing2LimitDegrees=45
  swing2Motion=Free
  twistLimitDegrees=45
  twistMotion=Free
 }
 {
  bone1Name=dyn_jacketSide_01_l
  bone2Name=dyn_jacketSide_mstr_l
  swing1LimitDegrees=45
  swing1Motion=Free
  swing2LimitDegrees=45
  swing2Motion=Free
  twistLimitDegrees=45
  twistMotion=Free
 }
 {
  bone1Name=dyn_jacketSide_02_l
  bone2Name=dyn_jacketSide_01_l
  swing1LimitDegrees=45
  swing1Motion=Free
  swing2LimitDegrees=45
  swing2Motion=Free
  twistLimitDegrees=45
  twistMotion=Free
 }
 {
  bone1Name=dyn_jacketSide_03_l
  bone2Name=dyn_jacketSide_02_l
  swing1LimitDegrees=45
  swing1Motion=Free
  swing2LimitDegrees=45
  swing2Motion=Free
  twistLimitDegrees=45
  twistMotion=Free
 }
 {
  bone1Name=dyn_jacketSide_04_l
  bone2Name=dyn_jacketSide_03_l
  swing1LimitDegrees=45
  swing1Motion=Free
  swing2LimitDegrees=45
  swing2Motion=Free
  twistLimitDegrees=45
  twistMotion=Free
 }
 {
  bone1Name=dyn_jacketSide_05_l
  bone2Name=dyn_jacketSide_04_l
  swing1LimitDegrees=45
  swing1Motion=Free
  swing2LimitDegrees=45
  swing2Motion=Free
  twistLimitDegrees=45
  twistMotion=Free
 }
 {
  bone1Name=dyn_jacketSide_mstr_r
  bone2Name=pelvis
  swing1LimitDegrees=45
  swing1Motion=Free
  swing2LimitDegrees=45
  swing2Motion=Free
  twistLimitDegrees=45
  twistMotion=Free
 }
 {
  bone1Name=dyn_jacketSide_01_r
  bone2Name=dyn_jacketSide_mstr_r
  swing1LimitDegrees=45
  swing1Motion=Free
  swing2LimitDegrees=45
  swing2Motion=Free
  twistLimitDegrees=45
  twistMotion=Free
 }
 {
  bone1Name=dyn_jacketSide_02_r
  bone2Name=dyn_jacketSide_01_r
  swing1LimitDegrees=45
  swing1Motion=Free
  swing2LimitDegrees=45
  swing2Motion=Free
  twistLimitDegrees=45
  twistMotion=Free
 }
 {
  bone1Name=dyn_jacketSide_03_r
  bone2Name=dyn_jacketSide_02_r
  swing1LimitDegrees=45
  swing1Motion=Free
  swing2LimitDegrees=45
  swing2Motion=Free
  twistLimitDegrees=45
  twistMotion=Free
 }
 {
  bone1Name=dyn_jacketSide_04_r
  bone2Name=dyn_jacketSide_03_r
  swing1LimitDegrees=45
  swing1Motion=Free
  swing2LimitDegrees=45
  swing2Motion=Free
  twistLimitDegrees=45
  twistMotion=Free
 }
 {
  bone1Name=dyn_arm_hook_MSTR
  bone2Name=clavicle_r
  swing1LimitDegrees=45
  swing1Motion=Free
  swing2LimitDegrees=45
  swing2Motion=Free
  twistLimitDegrees=45
  twistMotion=Free
 }
 {
  bone1Name=dyn_arm_hook_01
  bone2Name=dyn_arm_hook_MSTR
  swing1LimitDegrees=45
  swing1Motion=Free
  swing2LimitDegrees=45
  swing2Motion=Free
  twistLimitDegrees=45
  twistMotion=Free
 }
 {
  bone1Name=dyn_arm_hook_02
  bone2Name=dyn_arm_hook_01
  swing1LimitDegrees=45
  swing1Motion=Free
  swing2LimitDegrees=45
  swing2Motion=Free
  twistLimitDegrees=45
  twistMotion=Free
 }
 {
  bone1Name=dyn_arm_hook_03
  bone2Name=dyn_arm_hook_02
  swing1LimitDegrees=45
  swing1Motion=Free
  swing2LimitDegrees=45
  swing2Motion=Free
  twistLimitDegrees=45
  twistMotion=Free
 }
 {
  bone1Name=dyn_arm_hook_04
  bone2Name=dyn_arm_hook_03
  swing1LimitDegrees=45
  swing1Motion=Free
  swing2LimitDegrees=45
  swing2Motion=Free
  twistLimitDegrees=45
  twistMotion=Free
 }
 {
  bone1Name=dyn_arm_hook_05
  bone2Name=dyn_arm_hook_04
  swing1LimitDegrees=45
  swing1Motion=Free
  swing2LimitDegrees=45
  swing2Motion=Free
  twistLimitDegrees=45
  twistMotion=Free
 }
 {
  bone1Name=dyn_arm_hook_06
  bone2Name=dyn_arm_hook_05
  swing1LimitDegrees=45
  swing1Motion=Free
  swing2LimitDegrees=45
  swing2Motion=Free
  twistLimitDegrees=45
  twistMotion=Free
 }
 {
  bone1Name=dyn_arm_hook_07
  bone2Name=dyn_arm_hook_06
  swing1LimitDegrees=45
  swing1Motion=Free
  swing2LimitDegrees=45
  swing2Motion=Free
  twistLimitDegrees=45
  twistMotion=Free
 }
 {
  bone1Name=dyn_arm_hook_08
  bone2Name=dyn_arm_hook_07
  swing1LimitDegrees=45
  swing1Motion=Free
  swing2LimitDegrees=45
  swing2Motion=Free
  twistLimitDegrees=45
  twistMotion=Free
 }
]
