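# Physics asset for the front face-accessory mesh part of the medium female Ruins body, covering left and right front dynamic hair chains.
# Capsule bodies on the upper spine, clavicles and upper arms form the torso the chains hang from and can collide with.
# Each chain runs from a master bone constrained to spine_05 through numbered segments.
# Body and chain-root joints leave rotation free, while the lower hair joints share tight swing and twist limits.
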PhysicsAsset /BRCosmetics/Characters/Player/Female/Medium/Bodies/F_MED_Ruins/Meshes/Parts/F_MED_Ruins_Front_FaceAcc_Physics
bodies {
 clavicle_l [{
   shapeType=Capsule
   length=4.7081
   radius=5.3397
   shapeName=clavicle_l_capsule
   center=<6.5243 -0.6953 0.104>
   rotation=<rot 16.9211 87.287 -90>
  }]
 clavicle_r [{
   shapeType=Capsule
   length=5.6159
   radius=5.6487
   shapeName=clavicle_r_capsule
   center=<-6.4341 0.3237 0.8599>
   rotation=<rot 3.2384 98.0696 -90>
  }]
 dyn_hair_l_Front_00 [{shapeType=Capsule, length=0.505, radius=0.505, shapeName=dyn_hair_l_Front_00_capsule, rotation=<rot 0 -90 0>}]
 dyn_hair_l_Front_01 [{shapeType=Capsule, length=2.75, radius=2.25, shapeName=dyn_hair_l_Front_01_capsule, center=<3.6882 0 0>, rotation=<rot 90 0 0>}]
 dyn_hair_l_Front_02 [{shapeType=Capsule, length=3.3313, radius=2.5686, shapeName=dyn_hair_l_Front_02_capsule, center=<3.5 0 0>, rotation=<rot 90 0 0>}]
 dyn_hair_l_Front_03 [{shapeType=Capsule, length=3.4468, radius=2.3364, shapeName=dyn_hair_l_Front_03_capsule, center=<3.5 0 0>, rotation=<rot 90 0 0>}]
 dyn_hair_l_Front_04 [{shapeType=Capsule, length=3.875, radius=2.25, shapeName=dyn_hair_l_Front_04_capsule, center=<3.5 0 0>, rotation=<rot 90 0 0>}]
 dyn_hair_l_Front_MSTR [{shapeType=Capsule, length=0.505, radius=0.505, shapeName=dyn_hair_l_Front_MSTR_capsule, rotation=<rot 0 -90 0>}]
 dyn_hair_r_Front_00 [{shapeType=Capsule, length=0.505, radius=0.505, shapeName=dyn_hair_r_Front_00_capsule, rotation=<rot 0 -90 0>}]
 dyn_hair_r_Front_01 [{shapeType=Capsule, length=1.875, radius=3.125, shapeName=dyn_hair_r_Front_01_capsule, center=<3 0 0>, rotation=<rot 90 0 0>}]
 dyn_hair_r_Front_02 [{shapeType=Capsule, length=2.375, radius=2.625, shapeName=dyn_hair_r_Front_02_capsule, center=<3.5 0 0>, rotation=<rot 90 0 0>}]
 dyn_hair_r_Front_03 [{shapeType=Capsule, length=2.625, radius=2.125, shapeName=dyn_hair_r_Front_03_capsule, center=<3.5 0 0>, rotation=<rot 90 0 0>}]
 dyn_hair_r_Front_04 [{shapeType=Capsule, length=4.75, radius=2, shapeName=dyn_hair_r_Front_04_capsule, center=<3.5 0 0>, rotation=<rot 90 0 0>}]
 dyn_hair_r_Front_MSTR [{shapeType=Capsule, length=0.505, radius=0.505, shapeName=dyn_hair_r_Front_MSTR_capsule, rotation=<rot 0 -90 0>}]
 spine_04 [{shapeType=Capsule, length=4.4066, radius=10.2626, shapeName=spine_04_capsule, center=<0.306 0.5327 0>, rotation=<rot 0 -92.6847 0>}]
 spine_05 [{shapeType=Capsule, length=12.9411, radius=8.1451, shapeName=spine_05_capsule, center=<0.357 -2.1578 -0.1551>}]
 upperarm_l [{shapeType=Capsule, length=13.4229, radius=4.1429, shapeName=upperarm_l_capsule, center=<5 -1.25 0.0001>, rotation=<rot 0 -90 90>}]
 upperarm_r [{shapeType=Capsule, length=13.4229, radius=4.1429, shapeName=upperarm_r_capsule, center=<-5 1.25 0.0001>, rotation=<rot 0 -90 90>}]
}
constraints [
 {
  bone1Name=spine_05
  bone2Name=spine_04
  swing1LimitDegrees=45
  swing1Motion=Free
  swing2LimitDegrees=45
  swing2Motion=Free
  twistLimitDegrees=45
  twistMotion=Free
 }
 {
  bone1Name=upperarm_r
  bone2Name=clavicle_r
  swing1LimitDegrees=45
  swing1Motion=Free
  swing2LimitDegrees=45
  swing2Motion=Free
  twistLimitDegrees=45
  twistMotion=Free
 }
 {
  bone1Name=upperarm_l
  bone2Name=clavicle_l
  swing1LimitDegrees=45
  swing1Motion=Free
  swing2LimitDegrees=45
  swing2Motion=Free
  twistLimitDegrees=45
  twistMotion=Free
 }
 {
  bone1Name=clavicle_r
  bone2Name=spine_05
  swing1LimitDegrees=45
  swing1Motion=Free
  swing2LimitDegrees=45
  swing2Motion=Free
  twistLimitDegrees=45
  twistMotion=Free
 }
 {
  bone1Name=clavicle_l
  bone2Name=spine_05
  swing1LimitDegrees=45
  swing1Motion=Free
  swing2LimitDegrees=45
  swing2Motion=Free
  twistLimitDegrees=45
  twistMotion=Free
 }
 {
  bone1Name=dyn_hair_r_Front_MSTR
  bone2Name=spine_05
  swing1LimitDegrees=45
  swing1Motion=Free
  swing2LimitDegrees=45
  swing2Motion=Free
  twistLimitDegrees=45
  twistMotion=Free
 }
 {
  bone1Name=dyn_hair_r_Front_00
  bone2Name=dyn_hair_r_Front_MSTR
  swing1LimitDegrees=45
  swing1Motion=Free
  swing2LimitDegrees=45
  swing2Motion=Free
  twistLimitDegrees=45
  twistMotion=Free
 }
 {bone1Name=dyn_hair_r_Front_01, bone2Name=dyn_hair_r_Front_00, swing1LimitDegrees=35, swing2LimitDegrees=15, twistLimitDegrees=5}
 {bone1Name=dyn_hair_r_Front_02, bone2Name=dyn_hair_r_Front_01, swing1LimitDegrees=35, swing2LimitDegrees=15, twistLimitDegrees=5}
 {bone1Name=dyn_hair_r_Front_03, bone2Name=dyn_hair_r_Front_02, swing1LimitDegrees=35, swing2LimitDegrees=15, twistLimitDegrees=5}
 {bone1Name=dyn_hair_r_Front_04, bone2Name=dyn_hair_r_Front_03, swing1LimitDegrees=35, swing2LimitDegrees=15, twistLimitDegrees=5}
 {
  bone1Name=dyn_hair_l_Front_MSTR
  bone2Name=spine_05
  swing1LimitDegrees=45
  swing1Motion=Free
  swing2LimitDegrees=45
  swing2Motion=Free
  twistLimitDegrees=45
  twistMotion=Free
 }
 {
  bone1Name=dyn_hair_l_Front_00
  bone2Name=dyn_hair_l_Front_MSTR
  swing1LimitDegrees=45
  swing1Motion=Free
  swing2LimitDegrees=45
  swing2Motion=Free
  twistLimitDegrees=45
  twistMotion=Free
 }
 {bone1Name=dyn_hair_l_Front_01, bone2Name=dyn_hair_l_Front_00, swing1LimitDegrees=35, swing2LimitDegrees=15, twistLimitDegrees=5}
 {bone1Name=dyn_hair_l_Front_02, bone2Name=dyn_hair_l_Front_01, swing1LimitDegrees=35, swing2LimitDegrees=15, twistLimitDegrees=5}
 {bone1Name=dyn_hair_l_Front_03, bone2Name=dyn_hair_l_Front_02, swing1LimitDegrees=35, swing2LimitDegrees=15, twistLimitDegrees=5}
 {bone1Name=dyn_hair_l_Front_04, bone2Name=dyn_hair_l_Front_03, swing1LimitDegrees=35, swing2LimitDegrees=15, twistLimitDegrees=5}
]
