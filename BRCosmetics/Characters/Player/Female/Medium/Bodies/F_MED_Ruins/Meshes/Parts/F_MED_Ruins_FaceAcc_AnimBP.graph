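# Post-body Animation Blueprint for the F_MED_Ruins face accessory skeleton, adding secondary motion to the character's hair.
# Copied body poses feed constraints, look-ats and arm/clavicle-driven corrective offsets, then a RigidBody pass using the front-hair physics asset.
# AnimDynamics chains add wind-driven back and side hair with spine collision spheres, backpack alphas pin those to spine_05, and bangs simulate last.
# Several older node chains remain unconnected.
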AnimBlueprint /BRCosmetics/Characters/Player/Female/Medium/Bodies/F_MED_Ruins/Meshes/Parts/F_MED_Ruins_FaceAcc_AnimBP
target_skeleton=&/F_MED_Ruins_FaceAcc_Skeleton
graphs [
 AnimGraph {nodes [
   n1 = Anim Dynamics - Bone: dyn_hair_r_Front_01(Alpha=1, ExternalForce="0,0,0") {
    angularDampingOverride=0.9999
    angularSpringConstant=100
    bGravityOverrideInSimSpace=true
    bOverrideAngularDamping=true
    simulationSpace=RootRelative
    alphaBoolBlend {blendOption=Linear}
    boundBone {boneName=dyn_hair_r_Front_01}
    componentAppliedLinearAccClamp=<3500 3500 3500>
    componentLinearAccScale=<0.6 0.6 0.8>
    componentLinearVelScale=<0.6 0.6 0.8>
    physicsBodyDefinitions [{
      sphereCollisionRadius=0
      boundBone {boneName=dyn_hair_r_Front_01}
      boxExtents=<20 20 10>
      constraintSetup {angularLimitsMax=<1 25 35>, angularLimitsMin=<-1 -25 0>, angularTarget=<0 0 0>}
      localJointOffset=<1 0 0>
     }]
    relativeSpaceBone {boneName=root}
   }
   n2 = Anim Dynamics - Bone: dyn_hair_r_Front_02(ComponentPose=@n1, Alpha=1, ExternalForce="0,0,0") {
    angularDampingOverride=0.9999
    angularSpringConstant=100
    bGravityOverrideInSimSpace=true
    bOverrideAngularDamping=true
    simulationSpace=RootRelative
    alphaBoolBlend {blendOption=Linear}
    boundBone {boneName=dyn_hair_r_Front_02}
    componentAppliedLinearAccClamp=<3500 3500 3500>
    componentLinearAccScale=<0.6 0.6 0.8>
    componentLinearVelScale=<0.6 0.6 0.8>
    physicsBodyDefinitions [{
      sphereCollisionRadius=0
      boundBone {boneName=dyn_hair_r_Front_02}
      boxExtents=<20 20 10>
      constraintSetup {angularLimitsMax=<1 25 35>, angularLimitsMin=<-1 -25 3>, angularTarget=<0 0 0>}
      localJointOffset=<1 0 0>
     }]
    relativeSpaceBone {boneName=root}
   }
   n3 = Anim Dynamics - Bone: dyn_hair_r_Front_03(ComponentPose=@n2, Alpha=1, ExternalForce="0,0,0") {
    angularDampingOverride=0.9999
    angularSpringConstant=100
    bGravityOverrideInSimSpace=true
    bOverrideAngularDamping=true
    simulationSpace=RootRelative
    alphaBoolBlend {blendOption=Linear}
    boundBone {boneName=dyn_hair_r_Front_03}
    componentAppliedLinearAccClamp=<3500 3500 3500>
    componentLinearAccScale=<0.6 0.6 0.8>
    componentLinearVelScale=<0.6 0.6 0.8>
    physicsBodyDefinitions [{
      sphereCollisionRadius=0
      boundBone {boneName=dyn_hair_r_Front_03}
      boxExtents=<20 20 10>
      constraintSetup {angularLimitsMax=<1 25 35>, angularLimitsMin=<-1 -25 3>, angularTarget=<0 0 0>}
      localJointOffset=<1 0 0>
     }]
    relativeSpaceBone {boneName=root}
   }
   n4 = Anim Dynamics - Bone: dyn_hair_r_Front_04(ComponentPose=@n3, Alpha=1, ExternalForce="0,0,0") {
    angularDampingOverride=0.9999
    angularSpringConstant=100
    bGravityOverrideInSimSpace=true
    bOverrideAngularDamping=true
    simulationSpace=RootRelative
    alphaBoolBlend {blendOption=Linear}
    boundBone {boneName=dyn_hair_r_Front_04}
    componentAppliedLinearAccClamp=<3500 3500 3500>
    componentLinearAccScale=<0.6 0.6 0.8>
    componentLinearVelScale=<0.6 0.6 0.8>
    physicsBodyDefinitions [{
      sphereCollisionRadius=0
      boundBone {boneName=dyn_hair_r_Front_04}
      boxExtents=<20 20 10>
      constraintSetup {angularLimitsMax=<1 25 35>, angularLimitsMin=<-1 -25 0>, angularTarget=<0 0 0>}
      localJointOffset=<1 0 0>
     }]
    relativeSpaceBone {boneName=root}
   }
   n5 = Anim Dynamics - Bone: dyn_hair_l_Front_01(ComponentPose=@n4, Alpha=1, ExternalForce="0,0,0") {
    angularDampingOverride=0.9999
    angularSpringConstant=100
    bGravityOverrideInSimSpace=true
    bOverrideAngularDamping=true
    simulationSpace=RootRelative
    alphaBoolBlend {blendOption=Linear}
    boundBone {boneName=dyn_hair_l_Front_01}
    componentAppliedLinearAccClamp=<3500 3500 3500>
    componentLinearAccScale=<0.6 0.6 0.8>
    componentLinearVelScale=<0.6 0.6 0.8>
    physicsBodyDefinitions [{
      sphereCollisionRadius=0
      boundBone {boneName=dyn_hair_l_Front_01}
      boxExtents=<20 20 10>
      constraintSetup {angularLimitsMax=<1 25 35>, angularLimitsMin=<-1 -25 0>, angularTarget=<0 0 0>}
      localJointOffset=<1 0 0>
     }]
    relativeSpaceBone {boneName=root}
   }
   n6 = Anim Dynamics - Bone: dyn_hair_l_Front_02(ComponentPose=@n5, Alpha=1, ExternalForce="0,0,0") {
    angularDampingOverride=0.9999
    angularSpringConstant=100
    bGravityOverrideInSimSpace=true
    bOverrideAngularDamping=true
    simulationSpace=RootRelative
    alphaBoolBlend {blendOption=Linear}
    boundBone {boneName=dyn_hair_l_Front_02}
    componentAppliedLinearAccClamp=<3500 3500 3500>
    componentLinearAccScale=<0.6 0.6 0.8>
    componentLinearVelScale=<0.6 0.6 0.8>
    physicsBodyDefinitions [{
      sphereCollisionRadius=0
      boundBone {boneName=dyn_hair_l_Front_02}
      boxExtents=<20 20 10>
      constraintSetup {angularLimitsMax=<1 25 35>, angularLimitsMin=<-1 -25 3>, angularTarget=<0 0 0>}
      localJointOffset=<1 0 0>
     }]
    relativeSpaceBone {boneName=root}
   }
   n7 = Anim Dynamics - Bone: dyn_hair_l_Front_03(ComponentPose=@n6, Alpha=1, ExternalForce="0,0,0") {
    angularDampingOverride=0.9999
    angularSpringConstant=100
    bGravityOverrideInSimSpace=true
    bOverrideAngularDamping=true
    simulationSpace=RootRelative
    alphaBoolBlend {blendOption=Linear}
    boundBone {boneName=dyn_hair_l_Front_03}
    componentAppliedLinearAccClamp=<3500 3500 3500>
    componentLinearAccScale=<0.6 0.6 0.8>
    componentLinearVelScale=<0.6 0.6 0.8>
    physicsBodyDefinitions [{
      sphereCollisionRadius=0
      boundBone {boneName=dyn_hair_l_Front_03}
      boxExtents=<20 20 10>
      constraintSetup {angularLimitsMax=<1 25 35>, angularLimitsMin=<-1 -25 3>, angularTarget=<0 0 0>}
      localJointOffset=<1 0 0>
     }]
    relativeSpaceBone {boneName=root}
   }
   Anim Dynamics - Bone: dyn_hair_l_Front_04(ComponentPose=@n7, Alpha=1, ExternalForce="0,0,0") {
    angularDampingOverride=0.9999
    angularSpringConstant=100
    bGravityOverrideInSimSpace=true
    bOverrideAngularDamping=true
    simulationSpace=RootRelative
    alphaBoolBlend {blendOption=Linear}
    boundBone {boneName=dyn_hair_l_Front_04}
    componentAppliedLinearAccClamp=<3500 3500 3500>
    componentLinearAccScale=<0.6 0.6 0.8>
    componentLinearVelScale=<0.6 0.6 0.8>
    physicsBodyDefinitions [{
      sphereCollisionRadius=0
      boundBone {boneName=dyn_hair_l_Front_04}
      boxExtents=<20 20 10>
      constraintSetup {angularLimitsMax=<1 25 35>, angularLimitsMin=<-1 -25 0>, angularTarget=<0 0 0>}
      localJointOffset=<1 0 0>
     }]
    relativeSpaceBone {boneName=root}
   }
   Anim Dynamics - Bone: dyn_hair_r_01(Alpha=1, ExternalForce="0,0,0") {
    angularDampingOverride=0.98
    angularSpringConstant=50
    bAngularSpring=true
    bGravityOverrideInSimSpace=true
    bOverrideAngularDamping=true
    simulationSpace=RootRelative
    alphaBoolBlend {blendOption=Linear}
    boundBone {boneName=dyn_hair_r_01}
    componentAppliedLinearAccClamp=<3500 3500 3500>
    componentLinearAccScale=<0.6 0.6 1.2>
    componentLinearVelScale=<0.6 0.6 1.2>
    physicsBodyDefinitions [{
      sphereCollisionRadius=0
      boundBone {boneName=dyn_hair_r_01}
      boxExtents=<30 30 15>
      constraintSetup {angularLimitsMax=<1 10 0>, angularLimitsMin=<-1 -10 -30>, angularTarget=<1 0 0>}
      localJointOffset=<2 0 0>
     }]
    relativeSpaceBone {boneName=root}
   }
   Anim Dynamics - Bone: dyn_hair_bk_01(Alpha=1, ExternalForce="0,0,0") {
    angularDampingOverride=0.98
    angularSpringConstant=50
    bAngularSpring=true
    bGravityOverrideInSimSpace=true
    bOverrideAngularDamping=true
    simulationSpace=RootRelative
    alphaBoolBlend {blendOption=Linear}
    boundBone {boneName=dyn_hair_bk_01}
    componentAppliedLinearAccClamp=<3500 3500 3500>
    componentLinearAccScale=<0.6 0.6 1.2>
    componentLinearVelScale=<0.6 0.6 1.2>
    physicsBodyDefinitions [{
      sphereCollisionRadius=0
      boundBone {boneName=dyn_hair_bk_01}
      boxExtents=<30 30 15>
      constraintSetup {angularLimitsMax=<1 10 0>, angularLimitsMin=<-1 -10 -30>, angularTarget=<1 0 0>}
      localJointOffset=<2 0 0>
     }]
    relativeSpaceBone {boneName=root}
   }
   Anim Dynamics - Bone: dyn_hair_l_01(Alpha=1, ExternalForce="0,0,0") {
    angularDampingOverride=0.98
    angularSpringConstant=50
    bAngularSpring=true
    bGravityOverrideInSimSpace=true
    bOverrideAngularDamping=true
    simulationSpace=RootRelative
    alphaBoolBlend {blendOption=Linear}
    boundBone {boneName=dyn_hair_l_01}
    componentAppliedLinearAccClamp=<3500 3500 3500>
    componentLinearAccScale=<0.6 0.6 1.2>
    componentLinearVelScale=<0.6 0.6 1.2>
    physicsBodyDefinitions [{
      sphereCollisionRadius=0
      boundBone {boneName=dyn_hair_l_01}
      boxExtents=<30 30 15>
      constraintSetup {angularLimitsMax=<1 10 0>, angularLimitsMin=<-1 -10 -30>, angularTarget=<1 0 0>}
      localJointOffset=<2 0 0>
     }]
    relativeSpaceBone {boneName=root}
   }
   n8 = Anim Dynamics - Bone: dyn_bangs_r_01(Alpha=1, ExternalForce="0,0,0") {
    angularDampingOverride=0.9999
    angularSpringConstant=75
    bAngularSpring=true
    bGravityOverrideInSimSpace=true
    bOverrideAngularDamping=true
    simulationSpace=RootRelative
    alphaBoolBlend {blendOption=Linear}
    boundBone {boneName=dyn_bangs_r_01}
    componentAppliedLinearAccClamp=<3500 3500 3500>
    componentLinearAccScale=<0.9 0.9 0.9>
    componentLinearVelScale=<0.9 0.9 0.9>
    physicsBodyDefinitions [{
      sphereCollisionRadius=0
      boundBone {boneName=dyn_bangs_r_01}
      boxExtents=<15 10 10>
      constraintSetup {angularLimitsMax=<1 5 15>, angularLimitsMin=<-1 -45 -15>, angularTarget=<0 0 0>}
      localJointOffset=<2 0 0>
     }]
    relativeSpaceBone {boneName=root}
   }
   n9 = Anim Dynamics - Bone: dyn_bangs_l_01(ComponentPose=@n8, Alpha=1, ExternalForce="0,0,0") {
    angularDampingOverride=0.9999
    angularSpringConstant=75
    bAngularSpring=true
    bGravityOverrideInSimSpace=true
    bOverrideAngularDamping=true
    simulationSpace=RootRelative
    alphaBoolBlend {blendOption=Linear}
    boundBone {boneName=dyn_bangs_l_01}
    componentAppliedLinearAccClamp=<3500 3500 3500>
    componentLinearAccScale=<0.9 0.9 0.9>
    componentLinearVelScale=<0.9 0.9 0.9>
    physicsBodyDefinitions [{
      sphereCollisionRadius=0
      boundBone {boneName=dyn_bangs_l_01}
      boxExtents=<15 10 10>
      constraintSetup {angularLimitsMax=<1 5 15>, angularLimitsMin=<-1 -45 -15>, angularTarget=<0 0 0>}
      localJointOffset=<2 0 0>
     }]
    relativeSpaceBone {boneName=root}
   }
   n10 = Anim Dynamics - Bone: dyn_hair_r_Front_01_a(ComponentPose=@n9, Alpha=1, ExternalForce="0,0,0") {
    angularDampingOverride=0.9999
    angularSpringConstant=75
    bAngularSpring=true
    bGravityOverrideInSimSpace=true
    bOverrideAngularDamping=true
    simulationSpace=RootRelative
    alphaBoolBlend {blendOption=Linear}
    boundBone {boneName=dyn_hair_r_Front_01_a}
    componentAppliedLinearAccClamp=<3500 3500 3500>
    componentLinearAccScale=<0.9 0.9 0.9>
    componentLinearVelScale=<0.9 0.9 0.9>
    physicsBodyDefinitions [{
      sphereCollisionRadius=0
      boundBone {boneName=dyn_hair_r_Front_01_a}
      boxExtents=<15 10 10>
      constraintSetup {angularLimitsMax=<1 5 15>, angularLimitsMin=<-1 -45 -15>, angularTarget=<0 0 0>}
      localJointOffset=<2 0 0>
     }]
    relativeSpaceBone {boneName=root}
   }
   Anim Dynamics - Bone: dyn_hair_l_Front_00_a(ComponentPose=@n10, Alpha=1, ExternalForce="0,0,0") {
    angularDampingOverride=0.9999
    angularSpringConstant=75
    bAngularSpring=true
    bGravityOverrideInSimSpace=true
    bOverrideAngularDamping=true
    simulationSpace=RootRelative
    alphaBoolBlend {blendOption=Linear}
    boundBone {boneName=dyn_hair_l_Front_00_a}
    componentAppliedLinearAccClamp=<3500 3500 3500>
    componentLinearAccScale=<0.9 0.9 0.9>
    componentLinearVelScale=<0.9 0.9 0.9>
    physicsBodyDefinitions [{
      sphereCollisionRadius=0
      boundBone {boneName=dyn_hair_l_Front_00_a}
      boxExtents=<15 10 10>
      constraintSetup {angularLimitsMax=<1 5 15>, angularLimitsMin=<-1 -45 -15>, angularTarget=<0 0 0>}
      localJointOffset=<2 0 0>
     }]
    relativeSpaceBone {boneName=root}
   }
   n11 = "Constraint - dyn_hair_bk_00 (P)"(Alpha=1, ConstraintWeights_0=1) {
    alphaBoolBlend {blendOption=Linear}
    boneToModify {boneName=dyn_hair_bk_00}
    constraintSetup [{targetBone {boneName=spine_05}}]
    constraintWeights [1]
   }
   n12 = "Constraint - dyn_hair_l_00 (P)"(ComponentPose=@n11, Alpha=1, ConstraintWeights_0=1) {
    alphaBoolBlend {blendOption=Linear}
    boneToModify {boneName=dyn_hair_l_00}
    constraintSetup [{targetBone {boneName=spine_05}}]
    constraintWeights [1]
   }
   "Constraint - dyn_hair_r_00 (P)"(ComponentPose=@n12, Alpha=1, ConstraintWeights_0=1) {
    alphaBoolBlend {blendOption=Linear}
    boneToModify {boneName=dyn_hair_r_00}
    constraintSetup [{targetBone {boneName=spine_05}}]
    constraintWeights [1]
   }
   n13 = Look At - Bone: dyn_hair_r_Front_MSTR(Alpha=1) {
    alphaBoolBlend {blendOption=Linear}
    boneToModify {boneName=dyn_hair_r_Front_MSTR}
    lookAtLocation {x=0}
    lookAtTarget {boneReference {boneName=dyn_front_hair_TGT}}
    lookAt_Axis {axis=<1 0>}
   }
   Look At - Bone: dyn_hair_l_Front_MSTR(ComponentPose=@n13, Alpha=1) {
    alphaBoolBlend {blendOption=Linear}
    boneToModify {boneName=dyn_hair_l_Front_MSTR}
    lookAtLocation {x=0}
    lookAtTarget {boneReference {boneName=dyn_front_hair_TGT}}
    lookAt_Axis {axis=<1 0>}
   }
   n14 = "Transform (Modify) Bone - Bone: dyn_hair_l_03"(Alpha=1, Translation="0,0,0", Rotation="0,-30,0", Scale="1,1,1") {
    rotationMode=BMM_Additive
    rotationSpace=BCS_BoneSpace
    translationMode=BMM_Additive
    translationSpace=BCS_BoneSpace
    alphaBoolBlend {blendOption=Linear}
    boneToModify {boneName=dyn_hair_l_03}
    rotation {yaw=-30}
   }
   n15 = "Transform (Modify) Bone - Bone: dyn_hair_bk_03"(ComponentPose=@n14, Alpha=1, Translation="0,0,0", Rotation="0.000,-30.000,-0.000", Scale="1,1,1") {
    rotationMode=BMM_Additive
    rotationSpace=BCS_BoneSpace
    translationMode=BMM_Additive
    translationSpace=BCS_BoneSpace
    alphaBoolBlend {blendOption=Linear}
    boneToModify {boneName=dyn_hair_bk_03}
    rotation {yaw=-30}
   }
   "Transform (Modify) Bone - Bone: dyn_hair_r_03"(ComponentPose=@n15, Alpha=1, Translation="0,0,0", Rotation="0,-30,0", Scale="1,1,1") {
    rotationMode=BMM_Additive
    rotationSpace=BCS_BoneSpace
    translationMode=BMM_Additive
    translationSpace=BCS_BoneSpace
    alphaBoolBlend {blendOption=Linear}
    boneToModify {boneName=dyn_hair_r_03}
    rotation {yaw=-30}
   }
   n16 = Copy Pose From Mesh(SourceMeshComponent=Get Skel_Mesh()) {bUseAttachedParent=true}
   n17 = Copy Pose From Mesh() {bUseAttachedParent=true}
   n18 = "Constraint - dyn_hair_l_MSTR (P)"(
    ComponentPose=Local To Component(LocalPose=Layered blend per bone(BasePose=@n17, BlendPoses_0=@n16) {layerSetup [{branchFilters [{boneName=upperarm_l}, {boneName=upperarm_r}]}]})
    Alpha=1
    ConstraintWeights_0=1
   ) {
    alphaBoolBlend {blendOption=Linear}
    boneToModify {boneName=dyn_hair_l_MSTR}
    constraintSetup [{targetBone {boneName=spine_05}}]
    constraintWeights [1]
   }
   n19 = "Constraint - dyn_hair_r_MSTR (P)"(ComponentPose=@n18, Alpha=1, ConstraintWeights_0=1) {
    alphaBoolBlend {blendOption=Linear}
    boneToModify {boneName=dyn_hair_r_MSTR}
    constraintSetup [{targetBone {boneName=spine_05}}]
    constraintWeights [1]
   }
   n20 = Look At - Bone: dyn_hair_l_00(ComponentPose=@n19, Alpha=1) {
    alphaBoolBlend {blendOption=Linear}
    boneToModify {boneName=dyn_hair_l_00}
    lookAtLocation {x=0}
    lookAtTarget {boneReference {boneName=dyn_hair_l_02}}
    lookAt_Axis {axis=<1 0>}
   }
   n21 = Look At - Bone: dyn_hair_bk_00(ComponentPose=@n20, Alpha=1) {
    alphaBoolBlend {blendOption=Linear}
    boneToModify {boneName=dyn_hair_bk_00}
    lookAtLocation {x=0}
    lookAtTarget {boneReference {boneName=dyn_hair_bk_02}}
    lookAt_Axis {axis=<1 0>}
   }
   n22 = Look At - Bone: dyn_hair_r_00(ComponentPose=@n21, Alpha=1) {
    alphaBoolBlend {blendOption=Linear}
    boneToModify {boneName=dyn_hair_r_00}
    lookAtLocation {x=0}
    lookAtTarget {boneReference {boneName=dyn_hair_r_02}}
    lookAt_Axis {axis=<1 0>}
   }
   n23 = "Constraint - dyn_hair_side_l (PPPP)"!(
    ComponentPose=@n22
    Alpha=1
    ConstraintWeights_0=0.4002
    ConstraintWeights_1:float=0.2614
    ConstraintWeights_2:float=0.2097
    ConstraintWeights_3:float=0.1286
   ) -> (Pose) {
    alphaBoolBlend {blendOption=Linear}
    boneToModify {boneName=dyn_hair_side_l}
    constraintSetup [{targetBone {boneName=neck_02}}, {targetBone {boneName=head}}, {targetBone {boneName=neck_01}}, {targetBone {boneName=spine_05}}]
    constraintWeights [1,1,1,1]
   }
   n24 = "Constraint - dyn_hair_side_r (PPPPP)"!(
    ComponentPose=@n23
    Alpha=1
    ConstraintWeights_0=0.001
    ConstraintWeights_1:float=0.2893
    ConstraintWeights_2:float=0.1648
    ConstraintWeights_3:float=0.1691
    ConstraintWeights_4:float=0.3758
   ) -> (Pose) {
    alphaBoolBlend {blendOption=Linear}
    boneToModify {boneName=dyn_hair_side_r}
    constraintSetup [
     {targetBone {boneName=clavicle_r}}
     {targetBone {boneName=neck_02}}
     {targetBone {boneName=head}}
     {targetBone {boneName=neck_01}}
     {targetBone {boneName=spine_05}}
    ]
    constraintWeights [1,1,1,1,1]
   }
   n25 = "Constraint - dyn_hair_side_r (P)"(ComponentPose=@n24, Alpha=1, ConstraintWeights_0=1) {
    alphaBoolBlend {blendOption=Linear}
    boneToModify {boneName=dyn_hair_side_r}
    constraintSetup [{targetBone {boneName=clavicle_r}}]
    constraintWeights [1]
   }
   n26 = "Constraint - dyn_hair_side_l (P)"(ComponentPose=@n25, Alpha=1, ConstraintWeights_0=1) {
    alphaBoolBlend {blendOption=Linear}
    boneToModify {boneName=dyn_hair_side_l}
    constraintSetup [{targetBone {boneName=clavicle_l}}]
    constraintWeights [1]
   }
   n27 = "Transform (Modify) Bone - Bone: dyn_hair_side_r"(ComponentPose=@n26, Alpha=Get Arm R rot() -> (Arm R rot:double), Translation="0,0,-8", Rotation="0,0,0", Scale="1,1,1") {
    translationMode=BMM_Additive
    translationSpace=BCS_BoneSpace
    alphaBoolBlend {blendOption=Linear}
    boneToModify {boneName=dyn_hair_side_r}
    translation {z=-8}
   }
   n28 = "Transform (Modify) Bone - Bone: dyn_hair_side_l"(ComponentPose=@n27, Alpha=Get Arm L rot() -> (Arm L rot:double), Translation="1,0,8", Rotation="0,0,0", Scale="1,1,1") {
    translationMode=BMM_Additive
    translationSpace=BCS_BoneSpace
    alphaBoolBlend {blendOption=Linear}
    boneToModify {boneName=dyn_hair_side_l}
    translation {x=1, z=8}
   }
   n29 = "Transform (Modify) Bone - Bone: dyn_hair_side_l"(ComponentPose=@n28, Alpha=Get clav L rot() -> (clav L rot:double), Translation="4,-0.5,2", Rotation="0,15,10", Scale="1,1,1") {
    rotationMode=BMM_Additive
    rotationSpace=BCS_BoneSpace
    translationMode=BMM_Additive
    translationSpace=BCS_BoneSpace
    alphaBoolBlend {blendOption=Linear}
    boneToModify {boneName=dyn_hair_side_l}
    rotation {roll=10, yaw=15}
    translation=<4 -0.5 2>
   }
   n30 = "Constraint - dyn_hair_r_Front_MSTR (PP)"!(ComponentPose=@n29, Alpha=1, ConstraintWeights_0=0.6, ConstraintWeights_1:float=0.4) -> (Pose) {
    alphaBoolBlend {blendOption=Linear}
    boneToModify {boneName=dyn_hair_r_Front_MSTR}
    constraintSetup [{targetBone {boneName=clavicle_r}}, {targetBone {boneName=head}}]
    constraintWeights [1,1]
   }
   n31 = "Constraint - dyn_hair_l_Front_MSTR (PP)"!(ComponentPose=@n30, Alpha=1, ConstraintWeights_0=0.6, ConstraintWeights_1:float=0.4) -> (Pose) {
    alphaBoolBlend {blendOption=Linear}
    boneToModify {boneName=dyn_hair_l_Front_MSTR}
    constraintSetup [{targetBone {boneName=clavicle_l}}, {targetBone {boneName=head}}]
    constraintWeights [1,1]
   }
   n32 = RigidBody!(ComponentPose=@n31, Alpha, OverrideWorldGravity:Vector="0,0,-980", ExternalForce) -> (Pose) {
    bClampLinearTranslationLimitToRefPose=true
    bOverrideWorldGravity=true
    overridePhysicsAsset=&/F_MED_Ruins_Front_FaceAcc_Physics
    simulationSpace=BaseBoneSpace
    alphaBoolBlend {blendOption=Linear}
    baseBoneRef {boneName=spine_05}
    componentAppliedLinearAccClamp=<1000 1000 1000>
    componentLinearAccScale=<0.2 0.2 0.3>
    componentLinearVelScale=<0.1 0.1 0.2>
   }
   n33 = Get HairRotator_01() -> (HairRotator_01:Rotator)
   n34 = "Transform (Modify) Bone - Bone: dyn_hair_l_02"(ComponentPose=@n32, Alpha=1, Translation="0,0,0", Rotation=@n33, Scale="1,1,1") {
    rotationMode=BMM_Additive
    rotationSpace=BCS_BoneSpace
    translationMode=BMM_Additive
    translationSpace=BCS_BoneSpace
    alphaBoolBlend {blendOption=Linear}
    boneToModify {boneName=dyn_hair_l_02}
    rotation {yaw=-30}
   }
   n35 = "Transform (Modify) Bone - Bone: dyn_hair_bk_02"(ComponentPose=@n34, Alpha=1, Translation="0,0,0", Rotation=@n33, Scale="1,1,1") {
    rotationMode=BMM_Additive
    rotationSpace=BCS_BoneSpace
    translationMode=BMM_Additive
    translationSpace=BCS_BoneSpace
    alphaBoolBlend {blendOption=Linear}
    boneToModify {boneName=dyn_hair_bk_02}
    rotation=<rot 0 -30 0>
   }
   n36 = "Transform (Modify) Bone - Bone: dyn_hair_r_02"(ComponentPose=@n35, Alpha=1, Translation="0,0,0", Rotation=@n33, Scale="1,1,1") {
    rotationMode=BMM_Additive
    rotationSpace=BCS_BoneSpace
    translationMode=BMM_Additive
    translationSpace=BCS_BoneSpace
    alphaBoolBlend {blendOption=Linear}
    boneToModify {boneName=dyn_hair_r_02}
    rotation {yaw=-30}
   }
   n37 = Get winds() -> (winds:Vector)
   n38 = Anim Dynamics - Bone: dyn_hair_bk_02(ComponentPose=@n36, Alpha=1, ExternalForce=@n37) {
    angularDampingOverride=0.9999
    angularSpringConstant=75
    bAngularSpring=true
    bGravityOverrideInSimSpace=true
    bOverrideAngularDamping=true
    bUseSphericalLimits=true
    simulationSpace=RootRelative
    alphaBoolBlend {blendOption=Linear}
    boundBone {boneName=dyn_hair_bk_02}
    componentAppliedLinearAccClamp=<3500 3500 3500>
    componentLinearAccScale=<1 1 1>
    componentLinearVelScale=<1 1 1>
    physicsBodyDefinitions [{
      collisionType=CustomSphere
      sphereCollisionRadius=2
      boundBone {boneName=dyn_hair_bk_02}
      boxExtents=<20 20 12>
      constraintSetup {angularLimitsMax=<15 10 2>, angularLimitsMin=<-15 -10 -45>, angularTarget=<1 0 0>}
      localJointOffset=<2 0 0>
     }]
    relativeSpaceBone {boneName=root}
    sphericalLimits [
     {limitRadius=14, drivingBone {boneName=spine_04}, sphereLocalOffset=<2 7 0>}
     {limitRadius=14.5, drivingBone {boneName=spine_03}, sphereLocalOffset=<2 7 0>}
     {limitRadius=14, drivingBone {boneName=spine_05}, sphereLocalOffset=<3 7.5 0>}
    ]
   }
   n39 = Anim Dynamics - Bone: dyn_hair_bk_03(ComponentPose=@n38, Alpha=1, ExternalForce=@n37) {
    angularDampingOverride=0.9999
    angularSpringConstant=75
    bAngularSpring=true
    bGravityOverrideInSimSpace=true
    bOverrideAngularDamping=true
    bUseSphericalLimits=true
    simulationSpace=RootRelative
    alphaBoolBlend {blendOption=Linear}
    boundBone {boneName=dyn_hair_bk_03}
    componentAppliedLinearAccClamp=<3500 3500 3500>
    componentLinearAccScale=<1 1 1>
    componentLinearVelScale=<1 1 1>
    physicsBodyDefinitions [{
      collisionType=CustomSphere
      sphereCollisionRadius=2
      boundBone {boneName=dyn_hair_bk_03}
      boxExtents=<20 20 12>
      constraintSetup {angularLimitsMax=<15 10 2>, angularLimitsMin=<-15 -10 -45>, angularTarget=<1 0 0>}
      localJointOffset=<2 0 0>
     }]
    relativeSpaceBone {boneName=root}
    sphericalLimits [
     {limitRadius=14, drivingBone {boneName=spine_04}, sphereLocalOffset=<2 7 0>}
     {limitRadius=14.5, drivingBone {boneName=spine_03}, sphereLocalOffset=<2 7 0>}
     {limitRadius=14, drivingBone {boneName=spine_05}, sphereLocalOffset=<3 7.5 0>}
    ]
   }
   n40 = Anim Dynamics - Bone: dyn_hair_bk_04(ComponentPose=@n39, Alpha=1, ExternalForce=@n37) {
    angularDampingOverride=0.9999
    angularSpringConstant=75
    bAngularSpring=true
    bGravityOverrideInSimSpace=true
    bOverrideAngularDamping=true
    bUseSphericalLimits=true
    simulationSpace=RootRelative
    alphaBoolBlend {blendOption=Linear}
    boundBone {boneName=dyn_hair_bk_04}
    componentAppliedLinearAccClamp=<3500 3500 3500>
    componentLinearAccScale=<1 1 1>
    componentLinearVelScale=<1 1 1>
    physicsBodyDefinitions [{
      collisionType=CustomSphere
      sphereCollisionRadius=2
      boundBone {boneName=dyn_hair_bk_04}
      boxExtents=<20 20 12>
      constraintSetup {angularLimitsMax=<15 10 2>, angularLimitsMin=<-15 -10 -45>, angularTarget=<1 0 0>}
      localJointOffset=<2 0 0>
     }]
    relativeSpaceBone {boneName=root}
    sphericalLimits [
     {limitRadius=14, drivingBone {boneName=spine_04}, sphereLocalOffset=<2 7 0>}
     {limitRadius=14.5, drivingBone {boneName=spine_03}, sphereLocalOffset=<2 7 0>}
     {limitRadius=14, drivingBone {boneName=spine_05}, sphereLocalOffset=<3 7.5 0>}
    ]
   }
   n41 = Anim Dynamics - Bone: dyn_hair_r_02(ComponentPose=@n40, Alpha=1, ExternalForce=@n37) {
    angularDampingOverride=0.9999
    angularSpringConstant=75
    bAngularSpring=true
    bGravityOverrideInSimSpace=true
    bOverrideAngularDamping=true
    bUseSphericalLimits=true
    simulationSpace=RootRelative
    alphaBoolBlend {blendOption=Linear}
    boundBone {boneName=dyn_hair_r_02}
    componentAppliedLinearAccClamp=<3500 3500 3500>
    componentLinearAccScale=<1 1 1>
    componentLinearVelScale=<1 1 1>
    physicsBodyDefinitions [{
      collisionType=CustomSphere
      sphereCollisionRadius=2
      boundBone {boneName=dyn_hair_r_02}
      boxExtents=<20 20 12>
      constraintSetup {angularLimitsMax=<15 10 2>, angularLimitsMin=<-15 -10 -45>, angularTarget=<1 0 0>}
      localJointOffset=<2 0 0>
     }]
    relativeSpaceBone {boneName=root}
    sphericalLimits [
     {limitRadius=14, drivingBone {boneName=spine_04}, sphereLocalOffset=<2 7 0>}
     {limitRadius=14.5, drivingBone {boneName=spine_03}, sphereLocalOffset=<2 7 0>}
     {limitRadius=14, drivingBone {boneName=spine_05}, sphereLocalOffset=<3 7.5 0>}
    ]
   }
   n42 = Anim Dynamics - Bone: dyn_hair_r_03(ComponentPose=@n41, Alpha=1, ExternalForce=@n37) {
    angularDampingOverride=0.9999
    angularSpringConstant=75
    bAngularSpring=true
    bGravityOverrideInSimSpace=true
    bOverrideAngularDamping=true
    bUseSphericalLimits=true
    simulationSpace=RootRelative
    alphaBoolBlend {blendOption=Linear}
    boundBone {boneName=dyn_hair_r_03}
    componentAppliedLinearAccClamp=<3500 3500 3500>
    componentLinearAccScale=<1 1 1>
    componentLinearVelScale=<1 1 1>
    physicsBodyDefinitions [{
      collisionType=CustomSphere
      sphereCollisionRadius=2
      boundBone {boneName=dyn_hair_r_03}
      boxExtents=<20 20 12>
      constraintSetup {angularLimitsMax=<15 10 2>, angularLimitsMin=<-15 -10 -45>, angularTarget=<1 0 0>}
      localJointOffset=<2 0 0>
     }]
    relativeSpaceBone {boneName=root}
    sphericalLimits [
     {limitRadius=14, drivingBone {boneName=spine_04}, sphereLocalOffset=<2 7 0>}
     {limitRadius=14.5, drivingBone {boneName=spine_03}, sphereLocalOffset=<2 7 0>}
     {limitRadius=14, drivingBone {boneName=spine_05}, sphereLocalOffset=<3 7.5 0>}
    ]
   }
   n43 = Anim Dynamics - Bone: dyn_hair_r_04(ComponentPose=@n42, Alpha=1, ExternalForce=@n37) {
    angularDampingOverride=0.9999
    angularSpringConstant=75
    bAngularSpring=true
    bGravityOverrideInSimSpace=true
    bOverrideAngularDamping=true
    bUseSphericalLimits=true
    simulationSpace=RootRelative
    alphaBoolBlend {blendOption=Linear}
    boundBone {boneName=dyn_hair_r_04}
    componentAppliedLinearAccClamp=<3500 3500 3500>
    componentLinearAccScale=<1 1 1>
    componentLinearVelScale=<1 1 1>
    physicsBodyDefinitions [{
      collisionType=CustomSphere
      sphereCollisionRadius=2
      boundBone {boneName=dyn_hair_r_04}
      boxExtents=<20 20 12>
      constraintSetup {angularLimitsMax=<15 10 2>, angularLimitsMin=<-15 -10 -45>, angularTarget=<1 0 0>}
      localJointOffset=<2 0 0>
     }]
    relativeSpaceBone {boneName=root}
    sphericalLimits [
     {limitRadius=14, drivingBone {boneName=spine_04}, sphereLocalOffset=<2 7 0>}
     {limitRadius=14.5, drivingBone {boneName=spine_03}, sphereLocalOffset=<2 7 0>}
     {limitRadius=14, drivingBone {boneName=spine_05}, sphereLocalOffset=<3 7.5 0>}
    ]
   }
   n44 = Anim Dynamics - Bone: dyn_hair_l_02(ComponentPose=@n43, Alpha=1, ExternalForce=@n37) {
    angularDampingOverride=0.9999
    angularSpringConstant=75
    bAngularSpring=true
    bGravityOverrideInSimSpace=true
    bOverrideAngularDamping=true
    bUseSphericalLimits=true
    simulationSpace=RootRelative
    alphaBoolBlend {blendOption=Linear}
    boundBone {boneName=dyn_hair_l_02}
    componentAppliedLinearAccClamp=<3500 3500 3500>
    componentLinearAccScale=<1 1 1>
    componentLinearVelScale=<1 1 1>
    physicsBodyDefinitions [{
      collisionType=CustomSphere
      sphereCollisionRadius=2
      boundBone {boneName=dyn_hair_l_02}
      boxExtents=<20 20 12>
      constraintSetup {angularLimitsMax=<15 10 2>, angularLimitsMin=<-15 -10 -45>, angularTarget=<1 0 0>}
      localJointOffset=<2 0 0>
     }]
    relativeSpaceBone {boneName=root}
    sphericalLimits [
     {limitRadius=14, drivingBone {boneName=spine_04}, sphereLocalOffset=<2 7 0>}
     {limitRadius=14.5, drivingBone {boneName=spine_03}, sphereLocalOffset=<2 7 0>}
     {limitRadius=14, drivingBone {boneName=spine_05}, sphereLocalOffset=<3 7.5 0>}
    ]
   }
   n45 = Anim Dynamics - Bone: dyn_hair_l_03(ComponentPose=@n44, Alpha=1, ExternalForce=@n37) {
    angularDampingOverride=0.9999
    angularSpringConstant=75
    bAngularSpring=true
    bGravityOverrideInSimSpace=true
    bOverrideAngularDamping=true
    bUseSphericalLimits=true
    simulationSpace=RootRelative
    alphaBoolBlend {blendOption=Linear}
    boundBone {boneName=dyn_hair_l_03}
    componentAppliedLinearAccClamp=<3500 3500 3500>
    componentLinearAccScale=<1 1 1>
    componentLinearVelScale=<1 1 1>
    physicsBodyDefinitions [{
      collisionType=CustomSphere
      sphereCollisionRadius=2
      boundBone {boneName=dyn_hair_l_03}
      boxExtents=<20 20 12>
      constraintSetup {angularLimitsMax=<15 10 2>, angularLimitsMin=<-15 -10 -45>, angularTarget=<1 0 0>}
      localJointOffset=<2 0 0>
     }]
    relativeSpaceBone {boneName=root}
    sphericalLimits [
     {limitRadius=14, drivingBone {boneName=spine_04}, sphereLocalOffset=<2 7 0>}
     {limitRadius=14.5, drivingBone {boneName=spine_03}, sphereLocalOffset=<2 7 0>}
     {limitRadius=14, drivingBone {boneName=spine_05}, sphereLocalOffset=<3 7.5 0>}
    ]
   }
   n46 = Anim Dynamics - Bone: dyn_hair_l_04(ComponentPose=@n45, Alpha=1, ExternalForce=@n37) {
    angularDampingOverride=0.9999
    angularSpringConstant=75
    bAngularSpring=true
    bGravityOverrideInSimSpace=true
    bOverrideAngularDamping=true
    bUseSphericalLimits=true
    simulationSpace=RootRelative
    alphaBoolBlend {blendOption=Linear}
    boundBone {boneName=dyn_hair_l_04}
    componentAppliedLinearAccClamp=<3500 3500 3500>
    componentLinearAccScale=<1 1 1>
    componentLinearVelScale=<1 1 1>
    physicsBodyDefinitions [{
      collisionType=CustomSphere
      sphereCollisionRadius=2
      boundBone {boneName=dyn_hair_l_04}
      boxExtents=<20 20 12>
      constraintSetup {angularLimitsMax=<15 10 2>, angularLimitsMin=<-15 -10 -45>, angularTarget=<1 0 0>}
      localJointOffset=<2 0 0>
     }]
    relativeSpaceBone {boneName=root}
    sphericalLimits [
     {limitRadius=14, drivingBone {boneName=spine_04}, sphereLocalOffset=<2 7 0>}
     {limitRadius=14.5, drivingBone {boneName=spine_03}, sphereLocalOffset=<2 7 0>}
     {limitRadius=14, drivingBone {boneName=spine_05}, sphereLocalOffset=<3 7.5 0>}
    ]
   }
   n47 = Get backpack_A() -> (backpack_A:double)
   n48 = "Constraint - dyn_hair_bk_02 (P)"(ComponentPose=@n46, Alpha=@n47, ConstraintWeights_0=1) {
    alphaBoolBlend {blendOption=Linear}
    boneToModify {boneName=dyn_hair_bk_02}
    constraintSetup [{targetBone {boneName=spine_05}}]
    constraintWeights [1]
   }
   n49 = Get backpack_B() -> (backpack_B:double)
   n50 = "Constraint - dyn_hair_bk_02 (P)"(ComponentPose=@n48, Alpha=@n49, ConstraintWeights_0=1) {
    alphaBoolBlend {blendOption=Linear}
    boneToModify {boneName=dyn_hair_bk_02}
    constraintSetup [{targetBone {boneName=spine_05}}]
    constraintWeights [1]
   }
   n51 = Get backpack_C() -> (backpack_C:double)
   n52 = "Constraint - dyn_hair_bk_02 (P)"(ComponentPose=@n50, Alpha=@n51, ConstraintWeights_0=1) {
    alphaBoolBlend {blendOption=Linear}
    boneToModify {boneName=dyn_hair_bk_02}
    constraintSetup [{targetBone {boneName=spine_05}}]
    constraintWeights [1]
   }
   n53 = "Constraint - dyn_hair_l_02 (P)"(ComponentPose=@n52, Alpha=@n47, ConstraintWeights_0=1) {
    alphaBoolBlend {blendOption=Linear}
    boneToModify {boneName=dyn_hair_l_02}
    constraintSetup [{targetBone {boneName=spine_05}}]
    constraintWeights [1]
   }
   n54 = "Constraint - dyn_hair_l_03 (P)"(ComponentPose=@n53, Alpha=@n49, ConstraintWeights_0=1) {
    alphaBoolBlend {blendOption=Linear}
    boneToModify {boneName=dyn_hair_l_03}
    constraintSetup [{targetBone {boneName=spine_05}}]
    constraintWeights [1]
   }
   n55 = "Constraint - dyn_hair_l_04 (P)"(ComponentPose=@n54, Alpha=@n51, ConstraintWeights_0=1) {
    alphaBoolBlend {blendOption=Linear}
    boneToModify {boneName=dyn_hair_l_04}
    constraintSetup [{targetBone {boneName=spine_05}}]
    constraintWeights [1]
   }
   n56 = "Constraint - dyn_hair_r_02 (P)"(ComponentPose=@n55, Alpha=@n47, ConstraintWeights_0=1) {
    alphaBoolBlend {blendOption=Linear}
    boneToModify {boneName=dyn_hair_r_02}
    constraintSetup [{targetBone {boneName=spine_05}}]
    constraintWeights [1]
   }
   n57 = "Constraint - dyn_hair_r_03 (P)"(ComponentPose=@n56, Alpha=@n49, ConstraintWeights_0=1) {
    alphaBoolBlend {blendOption=Linear}
    boneToModify {boneName=dyn_hair_r_03}
    constraintSetup [{targetBone {boneName=spine_05}}]
    constraintWeights [1]
   }
   n58 = "Constraint - dyn_hair_r_04 (P)"(ComponentPose=@n57, Alpha=@n51, ConstraintWeights_0=1) {
    alphaBoolBlend {blendOption=Linear}
    boneToModify {boneName=dyn_hair_r_04}
    constraintSetup [{targetBone {boneName=spine_05}}]
    constraintWeights [1]
   }
   n59 = Anim Dynamics - Bone: dyn_bangs_r_01(ComponentPose=@n58, Alpha=1, ExternalForce="0,0,0") {
    angularDampingOverride=0.98
    angularSpringConstant=50
    bAngularSpring=true
    bGravityOverrideInSimSpace=true
    bOverrideAngularDamping=true
    simulationSpace=RootRelative
    alphaBoolBlend {blendOption=Linear}
    boundBone {boneName=dyn_bangs_r_01}
    componentAppliedLinearAccClamp=<3500 3500 3500>
    componentLinearAccScale=<0.6 0.6 1.2>
    componentLinearVelScale=<0.6 0.6 1.2>
    physicsBodyDefinitions [{
      sphereCollisionRadius=0
      boundBone {boneName=dyn_bangs_r_01}
      boxExtents=<25 20 20>
      constraintSetup {angularLimitsMax=<1 3 15>, angularLimitsMin=<-1 -15 -15>, angularTarget=<1 0 0>}
      localJointOffset=<2 0 0>
     }]
    relativeSpaceBone {boneName=root}
   }
   n60 = Anim Dynamics - Bone: dyn_bangs_r_02(ComponentPose=@n59, Alpha=1, ExternalForce="0,0,0") {
    angularDampingOverride=0.98
    angularSpringConstant=50
    bAngularSpring=true
    bGravityOverrideInSimSpace=true
    bOverrideAngularDamping=true
    simulationSpace=RootRelative
    alphaBoolBlend {blendOption=Linear}
    boundBone {boneName=dyn_bangs_r_02}
    componentAppliedLinearAccClamp=<3500 3500 3500>
    componentLinearAccScale=<0.6 0.6 1.2>
    componentLinearVelScale=<0.6 0.6 1.2>
    physicsBodyDefinitions [{
      sphereCollisionRadius=0
      boundBone {boneName=dyn_bangs_r_02}
      boxExtents=<25 20 20>
      constraintSetup {angularLimitsMax=<1 3 15>, angularLimitsMin=<-1 -15 -15>, angularTarget=<1 0 0>}
      localJointOffset=<2 0 0>
     }]
    relativeSpaceBone {boneName=root}
   }
   n61 = Anim Dynamics - Bone: dyn_bangs_l_01(ComponentPose=@n60, Alpha=1, ExternalForce="0,0,0") {
    angularDampingOverride=0.98
    angularSpringConstant=50
    bAngularSpring=true
    bGravityOverrideInSimSpace=true
    bOverrideAngularDamping=true
    simulationSpace=RootRelative
    alphaBoolBlend {blendOption=Linear}
    boundBone {boneName=dyn_bangs_l_01}
    componentAppliedLinearAccClamp=<3500 3500 3500>
    componentLinearAccScale=<0.6 0.6 1.2>
    componentLinearVelScale=<0.6 0.6 1.2>
    physicsBodyDefinitions [{
      sphereCollisionRadius=0
      boundBone {boneName=dyn_bangs_l_01}
      boxExtents=<20 15 15>
      constraintSetup {angularLimitsMax=<1 3 15>, angularLimitsMin=<-1 -15 -15>, angularTarget=<1 0 0>}
      localJointOffset=<2 0 0>
     }]
    relativeSpaceBone {boneName=root}
   }
   n62 = Anim Dynamics - Bone: dyn_bangs_l_02(ComponentPose=@n61, Alpha=1, ExternalForce="0,0,0") {
    angularDampingOverride=0.98
    angularSpringConstant=50
    bAngularSpring=true
    bGravityOverrideInSimSpace=true
    bOverrideAngularDamping=true
    simulationSpace=RootRelative
    alphaBoolBlend {blendOption=Linear}
    boundBone {boneName=dyn_bangs_l_02}
    componentAppliedLinearAccClamp=<3500 3500 3500>
    componentLinearAccScale=<0.6 0.6 1.2>
    componentLinearVelScale=<0.6 0.6 1.2>
    physicsBodyDefinitions [{
      sphereCollisionRadius=0
      boundBone {boneName=dyn_bangs_l_02}
      boxExtents=<20 15 15>
      constraintSetup {angularLimitsMax=<1 3 15>, angularLimitsMin=<-1 -15 -15>, angularTarget=<1 0 0>}
      localJointOffset=<2 0 0>
     }]
    relativeSpaceBone {boneName=root}
   }
   n63 = Anim Dynamics - Bone: dyn_hair_l_Front_00_a(ComponentPose=@n62, Alpha=1, ExternalForce="0,0,0") {
    angularDampingOverride=0.98
    angularSpringConstant=50
    bAngularSpring=true
    bGravityOverrideInSimSpace=true
    bOverrideAngularDamping=true
    simulationSpace=RootRelative
    alphaBoolBlend {blendOption=Linear}
    boundBone {boneName=dyn_hair_l_Front_00_a}
    componentAppliedLinearAccClamp=<3500 3500 3500>
    componentLinearAccScale=<0.6 0.6 1.2>
    componentLinearVelScale=<0.6 0.6 1.2>
    physicsBodyDefinitions [{
      sphereCollisionRadius=0
      boundBone {boneName=dyn_hair_l_Front_00_a}
      boxExtents=<20 15 15>
      constraintSetup {angularLimitsMax=<1 3 25>, angularLimitsMin=<-1 -15 -25>, angularTarget=<1 0 0>}
      localJointOffset=<1 0 0>
     }]
    relativeSpaceBone {boneName=root}
   }
   n64 = Anim Dynamics - Bone: dyn_hair_r_Front_01_a(ComponentPose=@n63, Alpha=1, ExternalForce="0,0,0") {
    angularDampingOverride=0.98
    angularSpringConstant=50
    bAngularSpring=true
    bGravityOverrideInSimSpace=true
    bOverrideAngularDamping=true
    simulationSpace=RootRelative
    alphaBoolBlend {blendOption=Linear}
    boundBone {boneName=dyn_hair_r_Front_01_a}
    componentAppliedLinearAccClamp=<3500 3500 3500>
    componentLinearAccScale=<0.6 0.6 1.2>
    componentLinearVelScale=<0.6 0.6 1.2>
    physicsBodyDefinitions [{
      sphereCollisionRadius=0
      boundBone {boneName=dyn_hair_r_Front_01_a}
      boxExtents=<20 15 15>
      constraintSetup {angularLimitsMax=<1 3 25>, angularLimitsMin=<-1 -15 -25>, angularTarget=<1 0 0>}
      localJointOffset=<1 0 0>
     }]
    relativeSpaceBone {boneName=root}
   }
   Output Pose(Result=Component To Local(ComponentPose=@n64)) {layerGroup=None}
  ]}
 EventGraph {nodes [
   Event BlueprintUpdateAnimation()
   > n1 = Sequence()
   > Set Arm L rot(Arm L rot:double=GetAlphaFromReferenceTransformDelta(SkeletalMeshComponent=Get Skel_Mesh(), BoneName=upperarm_l, BaseName=spine_05, InputRange=(X=40,Y=90), TwistAxis="1,0,0")) -> (Output_Get:double)
   > Set Arm R rot(Arm R rot:double=GetAlphaFromReferenceTransformDelta(SkeletalMeshComponent=Get Skel_Mesh(), BoneName=upperarm_r, BaseName=spine_05, InputRange=(X=40,Y=120), TwistAxis="1,0,0")) -> (Output_Get:double)
   > Set clav L rot(clav L rot:double=GetAlphaFromReferenceTransformDelta(
     SkeletalMeshComponent=Get Skel_Mesh()
     BoneName=clavicle_l
     BaseName=spine_05
     DeltaComponentType=EulerY
     InputRange=(X=0,Y=40)
     TwistAxis="1,0,0"
    )) -> (Output_Get:double)
   > n2 = Branch(Condition=Get bIsSkyDiving())
   n3 = GetGameTimeInSeconds()
   > n4 = GetRandomWindVector!(
    execute
    GameTime:float=@n3
    TimeOffset:float=0
    PerlinScale:float=0
    Frequency:Vector="7,6,8"
    Amplitude:Vector="-800,500,0"
    Offset:Vector="200,-300,0"
   ) -> (then, ReturnValue:Vector)
   > Set winds(winds:Vector=@n4) -> (Output_Get:Vector)
   > n5 = RandomRotatorFunction!(
    execute
    GameTimeInSeconds:double=@n3
    PerlinScale:double=1
    Frequency:Rotator="1.5,7,0"
    Amplitude:Rotator="5,15,0"
    Offset:Rotator="0,-25, 0"
    TimeOffset:double=0
   ) -> (then, RandomRotator:Rotator)
   > Set HairRotator_01(HairRotator_01:Rotator=@n5) -> (Output_Get:Rotator)
   Set winds(execute=@n2.else, winds="0, 0, 0")
   > Set HairRotator_01(HairRotator_01="0, 0, 0")
   n6 = Branch(execute=@n1.then_1, Condition=CheckForBackbling())
   > Set backpack_A(backpack_A:double=0.3) -> (Output_Get:double)
   > Set backpack_B(backpack_B:double=0.5) -> (Output_Get:double)
   > Set backpack_C(backpack_C:double=0.7) -> (Output_Get:double)
   Set backpack_A(execute=@n6.else, backpack_A=0)
   > Set backpack_B(backpack_B=0)
   > Set backpack_C(backpack_C=0)
   Event BlueprintInitializeAnimation()
   > Set Skel_Mesh(Skel_Mesh=GetOwnerSkeletalMeshComponent())
  ]}
]
variables [
 Arm L rot {type=number}
 Arm R rot {type=number}
 backpack_A {type=number}
 backpack_B {type=number}
 backpack_C {type=number}
 clav L rot {type=number}
 HairRotator_01 {type=Rotator}
 Skel_Mesh {type=SkeletalMeshComponent}
 winds {type=Vector}
]
